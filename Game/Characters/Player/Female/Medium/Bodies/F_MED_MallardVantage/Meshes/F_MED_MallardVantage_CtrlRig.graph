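# Control Rig for the F_MED_MallardVantage player body that drives corrective deform bones at joints including knees, elbows, biceps, shoulders, hips, pecs and wrists.
# Parent constraints blend base bones between neighboring bones, while toggleable branches gate most regions.
# Remapped rotation deltas from reference pose or spherical pose readers weight the corrective offsets, with a helper negating translation for right-side bones.
# Pecs also use aim math and position constraints toward the upper arms.
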ControlRigBlueprint /Game/Characters/Player/Female/Medium/Bodies/F_MED_MallardVantage/Meshes/F_MED_MallardVantage_CtrlRig
graphs [
 ContainedGraph {nodes [
   n1 = RigVMFunctionEntryNode()
   SequenceAggregate_2 = &RigVMFunction_Sequence(ExecuteContext=@n1.ExecuteContext)
   SequenceAggregate_2_1 = &RigVMFunction_Sequence(ExecuteContext=@SequenceAggregate_2.B)
   RigVMFunctionReturnNode(A=@SequenceAggregate_2.A, B=@SequenceAggregate_2_1.A, C=@SequenceAggregate_2_1.B)
  ]}
 ContainedGraph {nodes [
   n1 = RigVMFunctionEntryNode()
   n2 = &RigUnit_GetTransform(Space=LocalSpace, Item=@n1.Item)
   n3 = &RigUnit_GetTransform(Space=LocalSpace, bInitial=true, Item=@n1.Item)
   > RigVMFunctionReturnNode(OutTransform=/Script/RigVM.RigVMFunction_MathTransformToEulerTransform(Value=&RigVMFunction_MathTransformMakeRelative(Global=@n2.Transform, Parent=@n3.Transform).Local) {is=struct RigVMUnitNode, template="Cast::Execute(in Value,out Result)"}.Result)
  ]}
 ContainedGraph {nodes [
   n1 = RigVMFunctionEntryNode()
   n2 = &RigVMFunction_MathTransformToEulerTransform(Value=@n1.transform) {template="Cast::Execute(in Value,out Result)"}
   > RigVMFunctionReturnNode(
    mirrorTransform=(Rotation=(X=0,Y=0,Z=0,W=1),Translation=(X=0,Y=0,Z=0),Scale3D=(X=0,Y=0,Z=0))
    mirrorTransform.Rotation=/Script/RigVM.RigVMFunction_MathQuaternionFromRotatorV2(Value=@n2.Result.Rotation) {is=struct RigVMUnitNode, template="Cast::Execute(in Value,out Result)"}.Result
    mirrorTransform.Translation=/Script/RigVM.RigVMFunction_MathVectorScale(Factor=-1, Value=@n2.Result.Location) {is=struct RigVMUnitNode, template="Scale::Execute(in Value,in Factor,out Result)"}.Result
    mirrorTransform.Scale3D=@n2.Result.Scale
   )
  ]}
 ContainedGraph {nodes [
   n1 = RigVMFunctionEntryNode()
   > SequenceExecution_2 = RigVMAggregateNode()
   GetDeltaTransformFromRefPose_FNC = RigVMFunctionReferenceNode(Item=(Type=Bone,Name="dfrm_calf_l"), ExecuteContext=@SequenceExecution_2.A)
   GetDeltaTransformFromRefPose_FNC_1 = RigVMFunctionReferenceNode(Item=(Type=Bone,Name="dfrm_calf_r"), ExecuteContext=@SequenceExecution_2.B)
   SequenceExecution = RigVMAggregateNode(ExecuteContext=@SequenceExecution_2.C)
   n2 = &RigVMFunction_ControlFlowBranch(Condition=@n1.knee_in, ExecuteContext=@SequenceExecution.A)
   SequenceExecution_1 = RigVMAggregateNode(ExecuteContext=@n2.True)
   &RigUnit_ParentConstraint(
    AdvancedSettings=(RotationOrderForFilter=XZY)
    Child=(Type=Bone,Name="deform_knee_in_base_l")
    Parents=((Item=(Type=Bone,Name="dfrm_calf_l"),Weight=0.5),(Item=(Type=Bone,Name="dfrm_thigh_l"),Weight=0.5))
    Weight=1
    ExecutePin=@SequenceExecution_1.A
   )
   n3 = &RigVMFunction_MathFloatRemap(SourceMaximum=100, SourceMinimum=10, bClamp=true, Value=@GetDeltaTransformFromRefPose_FNC.OutTransform.Rotation.Yaw)
   > &RigUnit_ModifyTransforms(
    ItemToModify=((Item=(Type=Bone,Name="deform_knee_in_l"),Transform=(Rotation=(X=0,Y=0,Z=0,W=1),Translation=(X=0,Y=0,Z=0),Scale3D=(X=1,Y=1,Z=1))))
    Weight=@n3.Result
   )
   &RigUnit_ParentConstraint(
    AdvancedSettings=(RotationOrderForFilter=XZY)
    Child=(Type=Bone,Name="deform_knee_in_base_r")
    Parents=((Item=(Type=Bone,Name="dfrm_calf_r"),Weight=0.5),(Item=(Type=Bone,Name="dfrm_thigh_r"),Weight=0.5))
    Weight=1
    ExecutePin=@SequenceExecution_1.B
   )
   > Deform_GetMirrorTransform_FNC = RigVMFunctionReferenceNode(transform=(Rotation=(X=0,Y=0,Z=0,W=1),Translation=(X=0,Y=0,Z=0),Scale3D=(X=1,Y=1,Z=1)))
   n4 = &RigVMFunction_MathFloatRemap(SourceMaximum=100, SourceMinimum=10, bClamp=true, Value=@GetDeltaTransformFromRefPose_FNC_1.OutTransform.Rotation.Yaw)
   > &RigUnit_ModifyTransforms(
    ItemToModify=((Item=(Type=Bone,Name="deform_knee_in_r"),Transform=(Rotation=(X=0,Y=0,Z=0,W=1),Translation=(X=0,Y=0,Z=0),Scale3D=(X=1,Y=1,Z=1))))
    ItemToModify.0.Transform=@Deform_GetMirrorTransform_FNC.mirrorTransform
    Weight=@n4.Result
   )
   n5 = &RigVMFunction_ControlFlowBranch(Condition=@n1.knee_out, ExecuteContext=@SequenceExecution.B)
   SequenceExecution_1_1 = RigVMAggregateNode(ExecuteContext=@n5.True)
   &RigUnit_ParentConstraint(
    AdvancedSettings=(RotationOrderForFilter=XZY)
    Child=(Type=Bone,Name="deform_knee_out_base_l")
    Parents=((Item=(Type=Bone,Name="dfrm_calf_l"),Weight=0.1),(Item=(Type=Bone,Name="dfrm_thigh_l"),Weight=0.9))
    Weight=1
    ExecutePin=@SequenceExecution_1_1.A
   )
   > &RigUnit_ModifyTransforms(
    ItemToModify=((Item=(Type=Bone,Name="deform_knee_out_l"),Transform=(Rotation=(X=0,Y=0,Z=0,W=1),Translation=(X=0,Y=0,Z=0),Scale3D=(X=1,Y=1,Z=1))))
    Weight=@n3.Result
   )
   &RigUnit_ParentConstraint(
    AdvancedSettings=(RotationOrderForFilter=XZY)
    Child=(Type=Bone,Name="deform_knee_out_base_r")
    Parents=((Item=(Type=Bone,Name="dfrm_calf_r"),Weight=0.1),(Item=(Type=Bone,Name="dfrm_thigh_r"),Weight=0.9))
    Weight=1
    ExecutePin=@SequenceExecution_1_1.B
   )
   > Deform_GetMirrorTransform_FNC_1 = RigVMFunctionReferenceNode(transform=(Rotation=(X=0,Y=0,Z=0,W=1),Translation=(X=0,Y=0,Z=0),Scale3D=(X=1,Y=1,Z=1)))
   > &RigUnit_ModifyTransforms(
    ItemToModify=((Item=(Type=Bone,Name="deform_knee_out_r"),Transform=(Rotation=(X=0,Y=0,Z=0,W=1),Translation=(X=0,Y=0,Z=0),Scale3D=(X=1,Y=1,Z=1))))
    ItemToModify.0.Transform=@Deform_GetMirrorTransform_FNC_1.mirrorTransform
    Weight=@n4.Result
   )
   RigVMFunctionReturnNode(ExecuteContext=@SequenceExecution_2.D)
  ]}
 ContainedGraph {nodes [
   n1 = RigVMFunctionEntryNode()
   SequenceExecution_2 = &RigVMFunction_Sequence(ExecuteContext=@n1.ExecuteContext)
   SequenceExecution_2_1 = &RigVMFunction_Sequence(ExecuteContext=@SequenceExecution_2.B)
   SequenceExecution_2_2 = &RigVMFunction_Sequence(ExecuteContext=@SequenceExecution_2_1.B)
   RigVMFunctionReturnNode(A=@SequenceExecution_2.A, B=@SequenceExecution_2_1.A, C=@SequenceExecution_2_2.A, D=@SequenceExecution_2_2.B)
  ]}
 ContainedGraph {nodes [
   n1 = RigVMFunctionEntryNode()
   SequenceExecution = &RigVMFunction_Sequence(ExecuteContext=@n1.ExecuteContext)
   SequenceExecution_1 = &RigVMFunction_Sequence(ExecuteContext=@SequenceExecution.B)
   SequenceExecution_2 = &RigVMFunction_Sequence(ExecuteContext=@SequenceExecution_1.B)
   RigVMFunctionReturnNode(A=@SequenceExecution.A, B=@SequenceExecution_1.A, C=@SequenceExecution_2.A, D=@SequenceExecution_2.B)
  ]}
 ContainedGraph {nodes [
   n1 = RigVMFunctionEntryNode()
   SequenceExecution_1 = &RigVMFunction_Sequence(ExecuteContext=@n1.ExecuteContext)
   SequenceExecution_1_1 = &RigVMFunction_Sequence(ExecuteContext=@SequenceExecution_1.B)
   SequenceExecution_1_2 = &RigVMFunction_Sequence(ExecuteContext=@SequenceExecution_1_1.B)
   RigVMFunctionReturnNode(A=@SequenceExecution_1.A, B=@SequenceExecution_1_1.A, C=@SequenceExecution_1_2.A, D=@SequenceExecution_1_2.B)
  ]}
 ContainedGraph {nodes [
   n1 = RigVMFunctionEntryNode()
   SequenceExecution_1_1 = &RigVMFunction_Sequence(ExecuteContext=@n1.ExecuteContext)
   SequenceExecution_1_1_1 = &RigVMFunction_Sequence(ExecuteContext=@SequenceExecution_1_1.B)
   SequenceExecution_1_1_2 = &RigVMFunction_Sequence(ExecuteContext=@SequenceExecution_1_1_1.B)
   RigVMFunctionReturnNode(A=@SequenceExecution_1_1.A, B=@SequenceExecution_1_1_1.A, C=@SequenceExecution_1_1_2.A, D=@SequenceExecution_1_1_2.B)
  ]}
 ContainedGraph {nodes [
   n1 = RigVMFunctionEntryNode()
   n2 = &RigVMFunction_MathTransformToEulerTransform(Value=@n1.transform) {template="Cast::Execute(in Value,out Result)"}
   > RigVMFunctionReturnNode(
    mirrorTransform=(Rotation=(X=0,Y=0,Z=0,W=1),Translation=(X=0,Y=0,Z=0),Scale3D=(X=0,Y=0,Z=0))
    mirrorTransform.Rotation=&RigVMFunction_MathQuaternionFromRotatorV2(Value=@n2.Result.Rotation) {template="Cast::Execute(in Value,out Result)"}.Result
    mirrorTransform.Translation=&RigVMFunction_MathVectorScale(Factor=-1, Value=@n2.Result.Location) {template="Scale::Execute(in Value,in Factor,out Result)"}.Result
    mirrorTransform.Scale3D=@n2.Result.Scale
   )
  ]}
 ContainedGraph {nodes [
   n1 = RigVMFunctionEntryNode()
   n2 = &RigUnit_GetTransform(Space=LocalSpace, Item=@n1.Item)
   n3 = &RigUnit_GetTransform(Space=LocalSpace, bInitial=true, Item=@n1.Item)
   > RigVMFunctionReturnNode(OutTransform=&RigVMFunction_MathTransformToEulerTransform(Value=&RigVMFunction_MathTransformMakeRelative(Global=@n2.Transform, Parent=@n3.Transform).Local) {template="Cast::Execute(in Value,out Result)"}.Result)
  ]}
 ContainedGraph {nodes [
   n1 = RigVMFunctionEntryNode()
   > SequenceExecution_2 = RigVMAggregateNode()
   GetDeltaTransformFromRefPose_FNC = RigVMFunctionReferenceNode(Item=(Type=Bone,Name="dfrm_lowerarm_l"), ExecuteContext=@SequenceExecution_2.A)
   GetDeltaTransformFromRefPose_FNC_1 = RigVMFunctionReferenceNode(Item=(Type=Bone,Name="dfrm_lowerarm_r"), ExecuteContext=@SequenceExecution_2.B)
   SequenceExecution = RigVMAggregateNode(ExecuteContext=@SequenceExecution_2.C)
   n2 = &RigVMFunction_ControlFlowBranch(ExecuteContext=@SequenceExecution.A, Condition=@n1.elbow_in)
   SequenceExecution_1 = RigVMAggregateNode(ExecuteContext=@n2.True)
   &RigUnit_ParentConstraint(
    AdvancedSettings=(RotationOrderForFilter=XZY)
    Child=(Type=Bone,Name="deform_elbow_in_base_l")
    Parents=((Item=(Type=Bone,Name="dfrm_lowerarm_l"),Weight=0.4),(Item=(Type=Bone,Name="dfrm_upperarm_l"),Weight=0.6))
    Weight=1
    ExecutePin=@SequenceExecution_1.A
   )
   n3 = &RigVMFunction_MathFloatRemap(SourceMaximum=90, SourceMinimum=15, bClamp=true, Value=@GetDeltaTransformFromRefPose_FNC.OutTransform.Rotation.Yaw)
   > &RigUnit_ModifyTransforms(
    ItemToModify=((Item=(Type=Bone,Name="deform_elbow_in_l"),Transform=(Rotation=(X=0,Y=0,Z=0,W=1),Translation=(X=0,Y=0,Z=0),Scale3D=(X=1,Y=1,Z=1))))
    Weight=@n3.Result
   )
   &RigUnit_ParentConstraint(
    AdvancedSettings=(RotationOrderForFilter=XZY)
    Child=(Type=Bone,Name="deform_elbow_in_base_r")
    Parents=((Item=(Type=Bone,Name="dfrm_lowerarm_r"),Weight=0.4),(Item=(Type=Bone,Name="dfrm_upperarm_r"),Weight=0.6))
    Weight=1
    ExecutePin=@SequenceExecution_1.B
   )
   > Deform_GetMirrorTransform_FNC_1 = RigVMFunctionReferenceNode(transform=(Rotation=(X=0,Y=0,Z=0,W=1),Translation=(X=0,Y=0,Z=0),Scale3D=(X=1,Y=1,Z=1)))
   n4 = &RigVMFunction_MathFloatRemap(SourceMaximum=90, SourceMinimum=15, bClamp=true, Value=@GetDeltaTransformFromRefPose_FNC_1.OutTransform.Rotation.Yaw)
   > &RigUnit_ModifyTransforms(
    ItemToModify=((Item=(Type=Bone,Name="deform_elbow_in_r"),Transform=(Rotation=(X=0,Y=0,Z=0,W=1),Translation=(X=0,Y=0,Z=0),Scale3D=(X=1,Y=1,Z=1))))
    ItemToModify.0.Transform=@Deform_GetMirrorTransform_FNC_1.mirrorTransform
    Weight=@n4.Result
   )
   n5 = &RigVMFunction_ControlFlowBranch(ExecuteContext=@SequenceExecution.C, Condition=@n1.bicep)
   SequenceExecution_1_1_1 = RigVMAggregateNode(ExecuteContext=@n5.True)
   &RigUnit_ModifyTransforms(
    ItemToModify=((Item=(Type=Bone,Name="deform_bicep_l"),Transform=(Rotation=(X=0,Y=0,Z=0,W=1),Translation=(X=0,Y=0,Z=0),Scale3D=(X=1,Y=1,Z=1))))
    ExecutePin=@SequenceExecution_1_1_1.A
    Weight=@n3.Result
   )
   Deform_GetMirrorTransform_FNC = RigVMFunctionReferenceNode(transform=(Rotation=(X=0,Y=0,Z=0,W=1),Translation=(X=0,Y=0,Z=0),Scale3D=(X=1,Y=1,Z=1)), ExecuteContext=@SequenceExecution_1_1_1.B)
   > &RigUnit_ModifyTransforms(
    ItemToModify=((Item=(Type=Bone,Name="deform_bicep_r"),Transform=(Rotation=(X=0,Y=0,Z=0,W=1),Translation=(X=0,Y=0,Z=0),Scale3D=(X=1,Y=1,Z=1))))
    ItemToModify.0.Transform=@Deform_GetMirrorTransform_FNC.mirrorTransform
    Weight=@n4.Result
   )
   n6 = &RigVMFunction_ControlFlowBranch(Condition=@n1.elbow_out, ExecuteContext=@SequenceExecution.B)
   SequenceExecution_1_1 = RigVMAggregateNode(ExecuteContext=@n6.True)
   &RigUnit_ParentConstraint(
    AdvancedSettings=(RotationOrderForFilter=XZY)
    Child=(Type=Bone,Name="deform_elbow_out_base_l")
    Parents=((Item=(Type=Bone,Name="dfrm_lowerarm_l"),Weight=0.3),(Item=(Type=Bone,Name="dfrm_upperarm_l"),Weight=0.7))
    Weight=1
    ExecutePin=@SequenceExecution_1_1.A
   )
   > &RigUnit_ModifyTransforms(
    ItemToModify=((Item=(Type=Bone,Name="deform_elbow_out_l"),Transform=(Rotation=(X=0,Y=0,Z=0,W=1),Translation=(X=0,Y=0,Z=0),Scale3D=(X=1,Y=1,Z=1))))
    Weight=@n3.Result
   )
   &RigUnit_ParentConstraint(
    AdvancedSettings=(RotationOrderForFilter=XZY)
    Child=(Type=Bone,Name="deform_elbow_out_base_r")
    Parents=((Item=(Type=Bone,Name="dfrm_lowerarm_r"),Weight=0.3),(Item=(Type=Bone,Name="dfrm_upperarm_r"),Weight=0.7))
    Weight=1
    ExecutePin=@SequenceExecution_1_1.B
   )
   > Deform_GetMirrorTransform_FNC_2 = RigVMFunctionReferenceNode(transform=(Rotation=(X=0,Y=0,Z=0,W=1),Translation=(X=0,Y=0,Z=0),Scale3D=(X=1,Y=1,Z=1)))
   > &RigUnit_ModifyTransforms(
    ItemToModify=((Item=(Type=Bone,Name="deform_elbow_out_r"),Transform=(Rotation=(X=0,Y=0,Z=0,W=1),Translation=(X=0,Y=0,Z=0),Scale3D=(X=1,Y=1,Z=1))))
    ItemToModify.0.Transform=@Deform_GetMirrorTransform_FNC_2.mirrorTransform
    Weight=@n4.Result
   )
   RigVMFunctionReturnNode(ExecuteContext=@SequenceExecution_2.D)
  ]}
 ContainedGraph {nodes [
   n1 = RigVMFunctionEntryNode()
   SequenceExecution_2 = &RigVMFunction_Sequence(ExecuteContext=@n1.ExecuteContext)
   SequenceExecution_2_1 = &RigVMFunction_Sequence(ExecuteContext=@SequenceExecution_2.B)
   SequenceExecution_2_2 = &RigVMFunction_Sequence(ExecuteContext=@SequenceExecution_2_1.B)
   RigVMFunctionReturnNode(A=@SequenceExecution_2.A, B=@SequenceExecution_2_1.A, C=@SequenceExecution_2_2.A, D=@SequenceExecution_2_2.B)
  ]}
 ContainedGraph {nodes [
   n1 = RigVMFunctionEntryNode()
   SequenceExecution = &RigVMFunction_Sequence(ExecuteContext=@n1.ExecuteContext)
   SequenceExecution_1 = &RigVMFunction_Sequence(ExecuteContext=@SequenceExecution.B)
   SequenceExecution_2 = &RigVMFunction_Sequence(ExecuteContext=@SequenceExecution_1.B)
   RigVMFunctionReturnNode(A=@SequenceExecution.A, B=@SequenceExecution_1.A, C=@SequenceExecution_2.A, D=@SequenceExecution_2.B)
  ]}
 ContainedGraph {nodes [
   n1 = RigVMFunctionEntryNode()
   SequenceExecution_1 = &RigVMFunction_Sequence(ExecuteContext=@n1.ExecuteContext)
   SequenceExecution_1_1 = &RigVMFunction_Sequence(ExecuteContext=@SequenceExecution_1.B)
   SequenceExecution_1_2 = &RigVMFunction_Sequence(ExecuteContext=@SequenceExecution_1_1.B)
   RigVMFunctionReturnNode(A=@SequenceExecution_1.A, B=@SequenceExecution_1_1.A, C=@SequenceExecution_1_2.A, D=@SequenceExecution_1_2.B)
  ]}
 ContainedGraph {nodes [
   n1 = RigVMFunctionEntryNode()
   SequenceExecution_1_1 = &RigVMFunction_Sequence(ExecuteContext=@n1.ExecuteContext)
   SequenceExecution_1_1_1 = &RigVMFunction_Sequence(ExecuteContext=@SequenceExecution_1_1.B)
   SequenceExecution_1_1_2 = &RigVMFunction_Sequence(ExecuteContext=@SequenceExecution_1_1_1.B)
   RigVMFunctionReturnNode(A=@SequenceExecution_1_1.A, B=@SequenceExecution_1_1_1.A, C=@SequenceExecution_1_1_2.A, D=@SequenceExecution_1_1_2.B)
  ]}
 ContainedGraph {nodes [
   n1 = RigVMFunctionEntryNode()
   SequenceExecution_1_1_1 = &RigVMFunction_Sequence(ExecuteContext=@n1.ExecuteContext)
   SequenceExecution_1_1_1_1 = &RigVMFunction_Sequence(ExecuteContext=@SequenceExecution_1_1_1.B)
   SequenceExecution_1_1_1_2 = &RigVMFunction_Sequence(ExecuteContext=@SequenceExecution_1_1_1_1.B)
   RigVMFunctionReturnNode(A=@SequenceExecution_1_1_1.A, B=@SequenceExecution_1_1_1_1.A, C=@SequenceExecution_1_1_1_2.A, D=@SequenceExecution_1_1_1_2.B)
  ]}
 ContainedGraph {nodes [
   n1 = RigVMFunctionEntryNode()
   n2 = &RigVMFunction_MathTransformToEulerTransform(Value=@n1.transform) {template="Cast::Execute(in Value,out Result)"}
   > RigVMFunctionReturnNode(
    mirrorTransform=(Rotation=(X=0,Y=0,Z=0,W=1),Translation=(X=0,Y=0,Z=0),Scale3D=(X=0,Y=0,Z=0))
    mirrorTransform.Rotation=&RigVMFunction_MathQuaternionFromRotatorV2(Value=@n2.Result.Rotation) {template="Cast::Execute(in Value,out Result)"}.Result
    mirrorTransform.Translation=&RigVMFunction_MathVectorScale(Factor=-1, Value=@n2.Result.Location) {template="Scale::Execute(in Value,in Factor,out Result)"}.Result
    mirrorTransform.Scale3D=@n2.Result.Scale
   )
  ]}
 ContainedGraph {nodes [
   n1 = RigVMFunctionEntryNode()
   > SequenceExecution_2 = RigVMAggregateNode()
   n2 = &RigVMFunction_ControlFlowBranch(Condition=@n1.upperarm_up, ExecuteContext=@SequenceExecution_2.A)
   SequenceExecution_3 = RigVMAggregateNode(ExecuteContext=@n2.True)
   &RigUnit_ParentConstraint(
    AdvancedSettings=(RotationOrderForFilter=XZY)
    Child=(Type=Bone,Name="deform_upperarm_up_base_l")
    Parents=((Item=(Type=Bone,Name="dfrm_upperarm_twist_01_l"),Weight=0.5),(Item=(Type=Bone,Name="clavicle_l"),Weight=0.5))
    Weight=1
    ExecutePin=@SequenceExecution_3.A
   )
   > n3 = &RigUnit_SphericalPoseReader(
    ActiveRegionScaleFactors=(PositiveWidth=2.5,NegativeWidth=3,PositiveHeight=3,NegativeHeight=3)
    ActiveRegionSize=0.1
    Debug=(bDrawDebug=false,DebugScale=25,DebugSegments=20,DebugThickness=0.25)
    DriverAxis=(X=1,Y=0,Z=0)
    DriverItem=(Type=Bone,Name="dfrm_upperarm_l")
    FalloffRegionScaleFactors=(PositiveWidth=1,NegativeWidth=1,PositiveHeight=1,NegativeHeight=1)
    FalloffSize=0.5
    OptionalParentItem=(Type=Bone)
    RotationOffset=(X=0,Y=0,Z=0)
    Debug.bDrawDebug=@n1.debug_up
   )
   > &RigUnit_ModifyTransforms(
    ItemToModify=((Item=(Type=Bone,Name="deform_upperarm_up_l"),Transform=(Rotation=(X=0,Y=0,Z=0,W=1),Translation=(X=0,Y=0,Z=5),Scale3D=(X=1,Y=1,Z=1))))
    Weight=@n3.OutputParam
   )
   &RigUnit_ParentConstraint(
    AdvancedSettings=(RotationOrderForFilter=XZY)
    Child=(Type=Bone,Name="deform_upperarm_up_base_r")
    Parents=((Item=(Type=Bone,Name="dfrm_upperarm_twist_01_r"),Weight=0.5),(Item=(Type=Bone,Name="clavicle_r"),Weight=0.5))
    Weight=1
    ExecutePin=@SequenceExecution_3.B
   )
   > n4 = &RigUnit_SphericalPoseReader(
    ActiveRegionScaleFactors=(PositiveWidth=2.5,NegativeWidth=3,PositiveHeight=3,NegativeHeight=3)
    ActiveRegionSize=0.1
    Debug=(bDrawDebug=false,DebugScale=25,DebugSegments=20,DebugThickness=0.25)
    DriverAxis=(X=-1,Y=0,Z=0)
    DriverItem=(Type=Bone,Name="dfrm_upperarm_r")
    FalloffRegionScaleFactors=(PositiveWidth=1,NegativeWidth=1,PositiveHeight=1,NegativeHeight=1)
    FalloffSize=0.5
    FlipHeightScaling=true
    OptionalParentItem=(Type=Bone)
    RotationOffset=(X=180,Y=0,Z=180)
    Debug.bDrawDebug=@n1.debug_up
   )
   > Deform_GetMirrorTransform_FNC = RigVMFunctionReferenceNode(transform=(Rotation=(X=0,Y=0,Z=0,W=1),Translation=(X=0,Y=0,Z=0),Scale3D=(X=1,Y=1,Z=1)))
   > &RigUnit_ModifyTransforms(
    ItemToModify=((Item=(Type=Bone,Name="deform_upperarm_up_r"),Transform=(Rotation=(X=0,Y=0,Z=0,W=1),Translation=(X=0,Y=0,Z=0),Scale3D=(X=1,Y=1,Z=1))))
    ItemToModify.0.Transform=@Deform_GetMirrorTransform_FNC.mirrorTransform
    Weight=@n4.OutputParam
   )
   n5 = &RigVMFunction_ControlFlowBranch(Condition=@n1.upperarm_fr, ExecuteContext=@SequenceExecution_2.B)
   SequenceExecution_3_1 = RigVMAggregateNode(ExecuteContext=@n5.True)
   &RigUnit_ParentConstraint(
    AdvancedSettings=(RotationOrderForFilter=XZY)
    Child=(Type=Bone,Name="deform_upperarm_fr_base_l")
    Parents=((Item=(Type=Bone,Name="dfrm_upperarm_twist_01_l"),Weight=0.3),(Item=(Type=Bone,Name="clavicle_l"),Weight=0.7))
    Weight=1
    ExecutePin=@SequenceExecution_3_1.A
   )
   > n6 = &RigUnit_SphericalPoseReader(
    ActiveRegionScaleFactors=(PositiveWidth=0.5,NegativeWidth=3,PositiveHeight=1.5,NegativeHeight=1.5)
    ActiveRegionSize=0.1
    Debug=(bDrawDebug=false,DebugScale=25,DebugSegments=20,DebugThickness=0.25)
    DriverAxis=(X=1,Y=0,Z=0)
    DriverItem=(Type=Bone,Name="dfrm_upperarm_l")
    FalloffRegionScaleFactors=(PositiveWidth=1,NegativeWidth=1,PositiveHeight=1,NegativeHeight=1)
    FalloffSize=0.4
    OptionalParentItem=(Type=Bone)
    RotationOffset=(X=90,Y=0,Z=0)
    Debug.bDrawDebug=@n1.debug_fr
   )
   > &RigUnit_ModifyTransforms(
    ItemToModify=((Item=(Type=Bone,Name="deform_upperarm_fr_l"),Transform=(Rotation=(X=0,Y=0,Z=0,W=1),Translation=(X=0,Y=0,Z=0),Scale3D=(X=1,Y=1,Z=1))))
    Weight=@n6.OutputParam
   )
   > &RigUnit_ModifyTransforms(
    ItemToModify=((Item=(Type=Bone,Name="deform_upperarm_bk_l"),Transform=(Rotation=(X=0,Y=0,Z=0,W=1),Translation=(X=0,Y=0,Z=0),Scale3D=(X=1,Y=1,Z=1))))
    Weight=@n6.OutputParam
   )
   &RigUnit_ParentConstraint(
    AdvancedSettings=(RotationOrderForFilter=XZY)
    Child=(Type=Bone,Name="deform_upperarm_fr_base_r")
    Parents=((Item=(Type=Bone,Name="dfrm_upperarm_twist_01_r"),Weight=0.3),(Item=(Type=Bone,Name="clavicle_r"),Weight=0.7))
    Weight=1
    ExecutePin=@SequenceExecution_3_1.B
   )
   > n7 = &RigUnit_SphericalPoseReader(
    ActiveRegionScaleFactors=(PositiveWidth=0.5,NegativeWidth=3,PositiveHeight=1.5,NegativeHeight=1.5)
    ActiveRegionSize=0.1
    Debug=(bDrawDebug=false,DebugScale=25,DebugSegments=20,DebugThickness=0.25)
    DriverAxis=(X=-1,Y=0,Z=0)
    DriverItem=(Type=Bone,Name="dfrm_upperarm_r")
    FalloffRegionScaleFactors=(PositiveWidth=1,NegativeWidth=1,PositiveHeight=1,NegativeHeight=1)
    FalloffSize=0.4
    FlipWidthScaling=true
    OptionalParentItem=(Type=Bone)
    RotationOffset=(X=90,Y=180,Z=180)
    Debug.bDrawDebug=@n1.debug_fr
   )
   > Deform_GetMirrorTransform_FNC_2 = RigVMFunctionReferenceNode(transform=(Rotation=(X=0,Y=0,Z=0,W=1),Translation=(X=0,Y=0,Z=0),Scale3D=(X=1,Y=1,Z=1)))
   > &RigUnit_ModifyTransforms(
    ItemToModify=((Item=(Type=Bone,Name="deform_upperarm_fr_r"),Transform=(Rotation=(X=0,Y=0,Z=0,W=1),Translation=(X=0,Y=0,Z=0),Scale3D=(X=1,Y=1,Z=1))))
    ItemToModify.0.Transform=@Deform_GetMirrorTransform_FNC_2.mirrorTransform
    Weight=@n7.OutputParam
   )
   > Deform_GetMirrorTransform_FNC_1 = RigVMFunctionReferenceNode(transform=(Rotation=(X=0,Y=0,Z=0,W=1),Translation=(X=0,Y=0,Z=0),Scale3D=(X=1,Y=1,Z=1)))
   > &RigUnit_ModifyTransforms(
    ItemToModify=((Item=(Type=Bone,Name="deform_upperarm_bk_r"),Transform=(Rotation=(X=0,Y=0,Z=0,W=1),Translation=(X=0,Y=0,Z=0),Scale3D=(X=1,Y=1,Z=1))))
    ItemToModify.0.Transform=@Deform_GetMirrorTransform_FNC_1.mirrorTransform
    Weight=@n7.OutputParam
   )
   n8 = &RigVMFunction_ControlFlowBranch(Condition=@n1.upperarm_bk, ExecuteContext=@SequenceExecution_2.C)
   SequenceExecution_3_1_1 = RigVMAggregateNode(ExecuteContext=@n8.True)
   &RigUnit_ParentConstraint(
    AdvancedSettings=(RotationOrderForFilter=XZY)
    Child=(Type=Bone,Name="deform_upperarm_bk_base_l")
    Parents=((Item=(Type=Bone,Name="dfrm_upperarm_twist_01_l"),Weight=0.3),(Item=(Type=Bone,Name="clavicle_l"),Weight=0.7))
    Weight=1
    ExecutePin=@SequenceExecution_3_1_1.A
   )
   > n9 = &RigUnit_SphericalPoseReader(
    ActiveRegionScaleFactors=(PositiveWidth=3,NegativeWidth=1,PositiveHeight=1.5,NegativeHeight=1.5)
    ActiveRegionSize=0.1
    Debug=(bDrawDebug=false,DebugScale=25,DebugSegments=20,DebugThickness=0.25)
    DriverAxis=(X=1,Y=0,Z=0)
    DriverItem=(Type=Bone,Name="dfrm_upperarm_l")
    FalloffRegionScaleFactors=(PositiveWidth=1,NegativeWidth=1,PositiveHeight=1,NegativeHeight=1)
    FalloffSize=0.4
    OptionalParentItem=(Type=Bone)
    RotationOffset=(X=90,Y=0,Z=180)
    Debug.bDrawDebug=@n1.debug_bk
   )
   > &RigUnit_ModifyTransforms(
    ItemToModify=((Item=(Type=Bone,Name="deform_upperarm_bk_l"),Transform=(Rotation=(X=0,Y=0,Z=0,W=1),Translation=(X=0,Y=0,Z=0),Scale3D=(X=1,Y=1,Z=1))))
    Weight=@n9.OutputParam
   )
   &RigUnit_ParentConstraint(
    AdvancedSettings=(RotationOrderForFilter=XZY)
    Child=(Type=Bone,Name="deform_upperarm_bk_base_r")
    Parents=((Item=(Type=Bone,Name="dfrm_upperarm_twist_01_r"),Weight=0.3),(Item=(Type=Bone,Name="clavicle_r"),Weight=0.7))
    Weight=1
    ExecutePin=@SequenceExecution_3_1_1.B
   )
   > n10 = &RigUnit_SphericalPoseReader(
    ActiveRegionScaleFactors=(PositiveWidth=3,NegativeWidth=1,PositiveHeight=1.5,NegativeHeight=1.5)
    ActiveRegionSize=0.1
    Debug=(bDrawDebug=false,DebugScale=25,DebugSegments=20,DebugThickness=0.25)
    DriverAxis=(X=-1,Y=0,Z=0)
    DriverItem=(Type=Bone,Name="dfrm_upperarm_r")
    FalloffRegionScaleFactors=(PositiveWidth=1,NegativeWidth=1,PositiveHeight=1,NegativeHeight=1)
    FalloffSize=0.4
    FlipHeightScaling=true
    FlipWidthScaling=false
    OptionalParentItem=(Type=Bone)
    RotationOffset=(X=90,Y=0,Z=0)
    Debug.bDrawDebug=@n1.debug_bk
   )
   > Deform_GetMirrorTransform_FNC_3 = RigVMFunctionReferenceNode(transform=(Rotation=(X=0,Y=0,Z=0,W=1),Translation=(X=0,Y=0,Z=0),Scale3D=(X=1,Y=1,Z=1)))
   > &RigUnit_ModifyTransforms(
    ItemToModify=((Item=(Type=Bone,Name="deform_upperarm_bk_r"),Transform=(Rotation=(X=0,Y=0,Z=0,W=1),Translation=(X=0,Y=0,Z=0),Scale3D=(X=1,Y=1,Z=1))))
    ItemToModify.0.Transform=@Deform_GetMirrorTransform_FNC_3.mirrorTransform
    Weight=@n10.OutputParam
   )
   RigVMFunctionReturnNode(ExecuteContext=@SequenceExecution_2.D)
  ]}
 ContainedGraph {nodes [
   n1 = RigVMFunctionEntryNode()
   SequenceExecution_2 = &RigVMFunction_Sequence(ExecuteContext=@n1.ExecuteContext)
   SequenceExecution_2_1 = &RigVMFunction_Sequence(ExecuteContext=@SequenceExecution_2.B)
   SequenceExecution_2_2 = &RigVMFunction_Sequence(ExecuteContext=@SequenceExecution_2_1.B)
   RigVMFunctionReturnNode(A=@SequenceExecution_2.A, B=@SequenceExecution_2_1.A, C=@SequenceExecution_2_2.A, D=@SequenceExecution_2_2.B)
  ]}
 ContainedGraph {nodes [
   n1 = RigVMFunctionEntryNode()
   SequenceExecution_3 = &RigVMFunction_Sequence(ExecuteContext=@n1.ExecuteContext)
   SequenceExecution_3_1 = &RigVMFunction_Sequence(ExecuteContext=@SequenceExecution_3.B)
   SequenceExecution_3_2 = &RigVMFunction_Sequence(ExecuteContext=@SequenceExecution_3_1.B)
   RigVMFunctionReturnNode(A=@SequenceExecution_3.A, B=@SequenceExecution_3_1.A, C=@SequenceExecution_3_2.A, D=@SequenceExecution_3_2.B)
  ]}
 ContainedGraph {nodes [
   n1 = RigVMFunctionEntryNode()
   SequenceExecution_3_1 = &RigVMFunction_Sequence(ExecuteContext=@n1.ExecuteContext)
   SequenceExecution_3_1_1 = &RigVMFunction_Sequence(ExecuteContext=@SequenceExecution_3_1.B)
   SequenceExecution_3_1_2 = &RigVMFunction_Sequence(ExecuteContext=@SequenceExecution_3_1_1.B)
   RigVMFunctionReturnNode(A=@SequenceExecution_3_1.A, B=@SequenceExecution_3_1_1.A, C=@SequenceExecution_3_1_2.A, D=@SequenceExecution_3_1_2.B)
  ]}
 ContainedGraph {nodes [
   n1 = RigVMFunctionEntryNode()
   SequenceExecution_3_1_1 = &RigVMFunction_Sequence(ExecuteContext=@n1.ExecuteContext)
   SequenceExecution_3_1_1_1 = &RigVMFunction_Sequence(ExecuteContext=@SequenceExecution_3_1_1.B)
   SequenceExecution_3_1_1_2 = &RigVMFunction_Sequence(ExecuteContext=@SequenceExecution_3_1_1_1.B)
   RigVMFunctionReturnNode(A=@SequenceExecution_3_1_1.A, B=@SequenceExecution_3_1_1_1.A, C=@SequenceExecution_3_1_1_2.A, D=@SequenceExecution_3_1_1_2.B)
  ]}
 ContainedGraph {nodes [
   n1 = RigVMFunctionEntryNode()
   n2 = &RigVMFunction_MathTransformToEulerTransform(Value=@n1.transform) {template="Cast::Execute(in Value,out Result)"}
   > RigVMFunctionReturnNode(
    mirrorTransform=(Rotation=(X=0,Y=0,Z=0,W=1),Translation=(X=0,Y=0,Z=0),Scale3D=(X=0,Y=0,Z=0))
    mirrorTransform.Rotation=&RigVMFunction_MathQuaternionFromRotatorV2(Value=@n2.Result.Rotation) {template="Cast::Execute(in Value,out Result)"}.Result
    mirrorTransform.Translation=&RigVMFunction_MathVectorScale(Factor=-1, Value=@n2.Result.Location) {template="Scale::Execute(in Value,in Factor,out Result)"}.Result
    mirrorTransform.Scale3D=@n2.Result.Scale
   )
  ]}
 ContainedGraph {nodes [
   n1 = RigVMFunctionEntryNode()
   n2 = &RigVMFunction_ControlFlowBranch(ExecuteContext=@n1.ExecuteContext, Condition=&RigVMFunction_MathBoolOr(B=@n1.glute, A=@n1.groin).Result)
   RigVMFunctionReturnNode(ExecuteContext=@n2.Completed)
   SequenceExecution = RigVMAggregateNode(ExecuteContext=@n2.True)
   n3 = &RigUnit_SphericalPoseReader(
    ActiveRegionScaleFactors=(PositiveWidth=3.5,NegativeWidth=0.5,PositiveHeight=3,NegativeHeight=3)
    ActiveRegionSize=0.1
    Debug=(bDrawDebug=false,DebugScale=25,DebugSegments=20,DebugThickness=0.25)
    DriverAxis=(X=-1,Y=0,Z=0)
    DriverItem=(Type=Bone,Name="dfrm_thigh_l")
    FalloffRegionScaleFactors=(PositiveWidth=1,NegativeWidth=1,PositiveHeight=1,NegativeHeight=1)
    FalloffSize=0.45
    OptionalParentItem=(Type=Bone)
    RotationOffset=(X=90,Y=0,Z=0)
    ExecutePin=@SequenceExecution.A
    Debug.bDrawDebug=@n1.debug
   )
   > n4 = &RigUnit_SphericalPoseReader(
    ActiveRegionScaleFactors=(PositiveWidth=3.5,NegativeWidth=0.5,PositiveHeight=3,NegativeHeight=3)
    ActiveRegionSize=0.1
    Debug=(bDrawDebug=false,DebugScale=25,DebugSegments=20,DebugThickness=0.25)
    DriverAxis=(X=1,Y=0,Z=0)
    DriverItem=(Type=Bone,Name="dfrm_thigh_r")
    FalloffRegionScaleFactors=(PositiveWidth=1,NegativeWidth=1,PositiveHeight=1,NegativeHeight=1)
    FalloffSize=0.45
    FlipHeightScaling=false
    FlipWidthScaling=true
    OptionalParentItem=(Type=Bone)
    RotationOffset=(X=-90,Y=0,Z=0)
    Debug.bDrawDebug=@n1.debug
   )
   n5 = &RigVMFunction_ControlFlowBranch(ExecuteContext=@SequenceExecution.B, Condition=@n1.groin)
   SequenceExecution_1 = RigVMAggregateNode(ExecuteContext=@n5.True)
   &RigUnit_ParentConstraint(
    AdvancedSettings=(RotationOrderForFilter=XZY)
    Child=(Type=Bone,Name="deform_groin_base_l")
    Parents=((Item=(Type=Bone,Name="dfrm_thigh_l"),Weight=0.7),(Item=(Type=Bone,Name="dfrm_pelvis"),Weight=0.3))
    Weight=1
    ExecutePin=@SequenceExecution_1.A
   )
   > &RigUnit_ModifyTransforms(
    ItemToModify=((Item=(Type=Bone,Name="deform_groin_l"),Transform=(Rotation=(X=0,Y=0,Z=0,W=1),Translation=(X=-3,Y=12,Z=0),Scale3D=(X=1,Y=1,Z=1))))
    Weight=@n3.OutputParam
   )
   &RigUnit_ParentConstraint(
    AdvancedSettings=(RotationOrderForFilter=XZY)
    Child=(Type=Bone,Name="deform_groin_base_r")
    Parents=((Item=(Type=Bone,Name="dfrm_thigh_r"),Weight=0.7),(Item=(Type=Bone,Name="dfrm_pelvis"),Weight=0.3))
    Weight=1
    ExecutePin=@SequenceExecution_1.B
   )
   > &RigUnit_ModifyTransforms(
    ItemToModify=((Item=(Type=Bone,Name="deform_groin_r"),Transform=(Rotation=(X=0,Y=0,Z=0,W=1),Translation=(X=3,Y=-12,Z=0),Scale3D=(X=1,Y=1,Z=1))))
    Weight=@n4.OutputParam
   )
   n6 = &RigVMFunction_ControlFlowBranch(ExecuteContext=@SequenceExecution.C, Condition=@n1.glute)
   SequenceExecution_1_1 = RigVMAggregateNode(ExecuteContext=@n6.True)
   &RigUnit_ParentConstraint(
    AdvancedSettings=(RotationOrderForFilter=XZY)
    Child=(Type=Bone,Name="deform_glute_base_l")
    Parents=((Item=(Type=Bone,Name="dfrm_thigh_l"),Weight=0.25),(Item=(Type=Bone,Name="dfrm_pelvis"),Weight=0.75))
    Weight=1
    ExecutePin=@SequenceExecution_1_1.A
   )
   > &RigUnit_ModifyTransforms(
    ItemToModify=((Item=(Type=Bone,Name="deform_glute_l"),Transform=(Rotation=(X=0,Y=0,Z=0,W=1),Translation=(X=0,Y=0,Z=0),Scale3D=(X=1,Y=1,Z=1))))
    Weight=@n3.OutputParam
   )
   &RigUnit_ParentConstraint(
    AdvancedSettings=(RotationOrderForFilter=XZY)
    Child=(Type=Bone,Name="deform_glute_base_r")
    Parents=((Item=(Type=Bone,Name="dfrm_thigh_r"),Weight=0.25),(Item=(Type=Bone,Name="dfrm_pelvis"),Weight=0.75))
    Weight=1
    ExecutePin=@SequenceExecution_1_1.B
   )
   > Deform_GetMirrorTransform_FNC_1 = RigVMFunctionReferenceNode(transform=(Rotation=(X=0,Y=0,Z=0,W=1),Translation=(X=0,Y=0,Z=0),Scale3D=(X=1,Y=1,Z=1)))
   > &RigUnit_ModifyTransforms(
    ItemToModify=((Item=(Type=Bone,Name="deform_glute_r"),Transform=(Rotation=(X=0,Y=0,Z=0,W=1),Translation=(X=0,Y=0,Z=0),Scale3D=(X=1,Y=1,Z=1))))
    ItemToModify.0.Transform=@Deform_GetMirrorTransform_FNC_1.mirrorTransform
    Weight=@n4.OutputParam
   )
  ]}
 ContainedGraph {nodes [
   n1 = RigVMFunctionEntryNode()
   SequenceExecution = &RigVMFunction_Sequence(ExecuteContext=@n1.ExecuteContext)
   SequenceExecution_1 = &RigVMFunction_Sequence(ExecuteContext=@SequenceExecution.B)
   SequenceExecution_2 = &RigVMFunction_Sequence(ExecuteContext=@SequenceExecution_1.B)
   RigVMFunctionReturnNode(A=@SequenceExecution.A, B=@SequenceExecution_1.A, C=@SequenceExecution_2.A, D=@SequenceExecution_2.B)
  ]}
 ContainedGraph {nodes [
   n1 = RigVMFunctionEntryNode()
   SequenceExecution_1 = &RigVMFunction_Sequence(ExecuteContext=@n1.ExecuteContext)
   SequenceExecution_1_1 = &RigVMFunction_Sequence(ExecuteContext=@SequenceExecution_1.B)
   SequenceExecution_1_2 = &RigVMFunction_Sequence(ExecuteContext=@SequenceExecution_1_1.B)
   RigVMFunctionReturnNode(A=@SequenceExecution_1.A, B=@SequenceExecution_1_1.A, C=@SequenceExecution_1_2.A, D=@SequenceExecution_1_2.B)
  ]}
 ContainedGraph {nodes [
   n1 = RigVMFunctionEntryNode()
   SequenceExecution_1_1 = &RigVMFunction_Sequence(ExecuteContext=@n1.ExecuteContext)
   SequenceExecution_1_1_1 = &RigVMFunction_Sequence(ExecuteContext=@SequenceExecution_1_1.B)
   SequenceExecution_1_1_2 = &RigVMFunction_Sequence(ExecuteContext=@SequenceExecution_1_1_1.B)
   RigVMFunctionReturnNode(A=@SequenceExecution_1_1.A, B=@SequenceExecution_1_1_1.A, C=@SequenceExecution_1_1_2.A, D=@SequenceExecution_1_1_2.B)
  ]}
 ContainedGraph {nodes [
   n1 = RigVMFunctionEntryNode()
   n2 = &RigVMFunction_MathTransformToEulerTransform(Value=@n1.transform) {template="Cast::Execute(in Value,out Result)"}
   > RigVMFunctionReturnNode(
    mirrorTransform=(Rotation=(X=0,Y=0,Z=0,W=1),Translation=(X=0,Y=0,Z=0),Scale3D=(X=0,Y=0,Z=0))
    mirrorTransform.Rotation=&RigVMFunction_MathQuaternionFromRotatorV2(Value=@n2.Result.Rotation) {template="Cast::Execute(in Value,out Result)"}.Result
    mirrorTransform.Translation=&RigVMFunction_MathVectorScale(Factor=-1, Value=@n2.Result.Location) {template="Scale::Execute(in Value,in Factor,out Result)"}.Result
    mirrorTransform.Scale3D=@n2.Result.Scale
   )
  ]}
 ContainedGraph {nodes [
   n1 = RigVMFunctionEntryNode()
   n2 = &RigUnit_GetTransform(Space=LocalSpace, Item=@n1.Item)
   n3 = &RigUnit_GetTransform(Space=LocalSpace, bInitial=true, Item=@n1.Item)
   > RigVMFunctionReturnNode(OutTransform=&RigVMFunction_MathTransformToEulerTransform(Value=&RigVMFunction_MathTransformMakeRelative(Global=@n2.Transform, Parent=@n3.Transform).Local) {template="Cast::Execute(in Value,out Result)"}.Result)
  ]}
 ContainedGraph {nodes [
   RigVMFunctionEntryNode()
   > SequenceExecution = RigVMAggregateNode()
   RigVMFunctionReturnNode(ExecuteContext=@SequenceExecution.D)
   n1 = /Script/ControlRig.RigUnit_AimBoneMath(
    DebugSettings=(Scale=10,WorldOffset=(Rotation=(X=0,Y=0,Z=0,W=1),Translation=(X=0,Y=0,Z=0),Scale3D=(X=1,Y=1,Z=1)))
    Primary=(Weight=1,Axis=(X=-1,Y=0,Z=0),Target=(X=0,Y=0,Z=0),Kind=Location,Space=(Type=Bone,Name="upperarm_r"))
    Secondary=(Weight=1,Axis=(X=0,Y=1,Z=0),Target=(X=0,Y=0,Z=0),Kind=Location,Space=(Type=Bone,Name="spine_04"))
    Weight=1
    InputTransform=&RigUnit_GetTransform(Item=(Type=Bone,Name="deform_pec_aim_r")).Transform
   ) {is=struct RigVMUnitNode}
   &RigUnit_SetTransform(Item=(Type=Bone,Name="deform_pec_aim_r"), Weight=1, bInitial=false, Value=@n1.Result, ExecutePin=@SequenceExecution.B)
   > /Script/ControlRig.RigUnit_PositionConstraintLocalSpaceOffset(
    Child=(Type=Bone,Name="deform_pec_r")
    Filter=(bX=True,bY=false,bZ=false)
    Parents=((Item=(Type=Bone,Name="deform_pec_base_r"),Weight=0),(Item=(Type=Bone,Name="upperarm_r"),Weight=1))
    Weight=1
    bMaintainOffset=True
   ) {is=struct RigVMUnitNode}
   > GetDeltaTransformFromRefPose_FNC_1 = RigVMFunctionReferenceNode(Item=(Type=Bone,Name="clavicle_r"))
   > Deform_GetMirrorTransform_FNC = RigVMFunctionReferenceNode(transform=(Rotation=(X=0,Y=0,Z=0,W=1),Translation=(X=0,Y=0,Z=0),Scale3D=(X=1,Y=1,Z=1)))
   > &RigUnit_ModifyTransforms(
    ItemToModify=((Item=(Type=Bone,Name="deform_pec_r"),Transform=(Rotation=(X=0,Y=0,Z=0,W=1),Translation=(X=0,Y=0,Z=0),Scale3D=(X=1,Y=1,Z=1))))
    ItemToModify.0.Transform=@Deform_GetMirrorTransform_FNC.mirrorTransform
    Weight=&RigVMFunction_MathFloatRemap(SourceMaximum=30, bClamp=true, Value=@GetDeltaTransformFromRefPose_FNC_1.OutTransform.Rotation.Pitch).Result
   )
   n2 = &RigUnit_AimBoneMath(
    DebugSettings=(Scale=10,WorldOffset=(Rotation=(X=0,Y=0,Z=0,W=1),Translation=(X=0,Y=0,Z=0),Scale3D=(X=1,Y=1,Z=1)))
    Primary=(Weight=1,Axis=(X=1,Y=0,Z=0),Target=(X=0,Y=0,Z=0),Kind=Location,Space=(Type=Bone,Name="upperarm_l"))
    Secondary=(Weight=1,Axis=(X=0,Y=-1,Z=0),Target=(X=0,Y=0,Z=0),Kind=Location,Space=(Type=Bone,Name="spine_04"))
    Weight=1
    InputTransform=&RigUnit_GetTransform(Item=(Type=Bone,Name="deform_pec_aim_l")).Transform
   )
   &RigUnit_SetTransform(Item=(Type=Bone,Name="deform_pec_aim_l"), Weight=1, bInitial=false, Value=@n2.Result, ExecutePin=@SequenceExecution.A)
   > &RigUnit_PositionConstraintLocalSpaceOffset(
    Child=(Type=Bone,Name="deform_pec_l")
    Filter=(bX=True,bY=false,bZ=false)
    Parents=((Item=(Type=Bone,Name="deform_pec_base_l"),Weight=0),(Item=(Type=Bone,Name="upperarm_l"),Weight=1))
    Weight=1
    bMaintainOffset=True
   )
   > GetDeltaTransformFromRefPose_FNC = RigVMFunctionReferenceNode(Item=(Type=Bone,Name="clavicle_l"))
   > &RigUnit_ModifyTransforms(
    ItemToModify=((Item=(Type=Bone,Name="deform_pec_l"),Transform=(Rotation=(X=0,Y=0,Z=0,W=1),Translation=(X=0,Y=0,Z=0),Scale3D=(X=1,Y=1,Z=1))))
    Weight=&RigVMFunction_MathFloatRemap(SourceMaximum=30, bClamp=true, Value=@GetDeltaTransformFromRefPose_FNC.OutTransform.Rotation.Pitch).Result
   )
   &RigUnit_ParentConstraint(
    AdvancedSettings=(RotationOrderForFilter=XZY)
    Child=(Type=Bone,Name="deform_pec_aim_base_l")
    ExecutePin=()
    Parents=((Item=(Type=Bone,Name="spine_03"),Weight=0.25),(Item=(Type=Bone,Name="spine_04"),Weight=0.55),(Item=(Type=Bone,Name="spine_05"),Weight=0.2))
    Weight=1
   )
   &RigUnit_ParentConstraint(
    AdvancedSettings=(RotationOrderForFilter=XZY)
    Child=(Type=Bone,Name="deform_pec_aim_base_r")
    ExecutePin=()
    Parents=((Item=(Type=Bone,Name="spine_03"),Weight=0.25),(Item=(Type=Bone,Name="spine_04"),Weight=0.55),(Item=(Type=Bone,Name="spine_05"),Weight=0.2))
    Weight=1
   )
  ]}
 ContainedGraph {nodes [
   n1 = RigVMFunctionEntryNode()
   SequenceExecution = &RigVMFunction_Sequence(ExecuteContext=@n1.ExecuteContext)
   SequenceExecution_1 = &RigVMFunction_Sequence(ExecuteContext=@SequenceExecution.B)
   SequenceExecution_2 = &RigVMFunction_Sequence(ExecuteContext=@SequenceExecution_1.B)
   RigVMFunctionReturnNode(A=@SequenceExecution.A, B=@SequenceExecution_1.A, C=@SequenceExecution_2.A, D=@SequenceExecution_2.B)
  ]}
 ContainedGraph {nodes [
   n1 = RigVMFunctionEntryNode()
   n2 = &RigVMFunction_MathTransformToEulerTransform(Value=@n1.transform) {template="Cast::Execute(in Value,out Result)"}
   > RigVMFunctionReturnNode(
    mirrorTransform=(Rotation=(X=0,Y=0,Z=0,W=1),Translation=(X=0,Y=0,Z=0),Scale3D=(X=0,Y=0,Z=0))
    mirrorTransform.Rotation=&RigVMFunction_MathQuaternionFromRotatorV2(Value=@n2.Result.Rotation) {template="Cast::Execute(in Value,out Result)"}.Result
    mirrorTransform.Translation=&RigVMFunction_MathVectorScale(Factor=-1, Value=@n2.Result.Location) {template="Scale::Execute(in Value,in Factor,out Result)"}.Result
    mirrorTransform.Scale3D=@n2.Result.Scale
   )
  ]}
 ContainedGraph {nodes [
   n1 = RigVMFunctionEntryNode()
   > SequenceExecution = RigVMAggregateNode()
   n2 = &RigVMFunction_ControlFlowBranch(Condition=@n1.wrist_up, ExecuteContext=@SequenceExecution.A)
   SequenceExecution_1 = RigVMAggregateNode(ExecuteContext=@n2.True)
   &RigUnit_ParentConstraint(
    AdvancedSettings=(RotationOrderForFilter=XZY)
    Child=(Type=Bone,Name="deform_wrist_up_base_l")
    Parents=((Item=(Type=Bone,Name="hand_l"),Weight=0.4),(Item=(Type=Bone,Name="lowerarm_twist_01_l"),Weight=0.6))
    Weight=1
    ExecutePin=@SequenceExecution_1.A
   )
   > n3 = &RigUnit_SphericalPoseReader(
    ActiveRegionScaleFactors=(PositiveWidth=0.5,NegativeWidth=3.5,PositiveHeight=3,NegativeHeight=3)
    ActiveRegionSize=0.1
    Debug=(bDrawDebug=false,DebugScale=25,DebugSegments=20,DebugThickness=0.25)
    DriverAxis=(X=1,Y=0,Z=0)
    DriverItem=(Type=Bone,Name="hand_l")
    FalloffRegionScaleFactors=(PositiveWidth=1,NegativeWidth=1,PositiveHeight=1,NegativeHeight=1)
    FalloffSize=0.45
    OptionalParentItem=(Type=Bone,Name="lowerarm_twist_01_l")
    RotationOffset=(X=90,Y=0,Z=0)
    Debug.bDrawDebug=@n1.debug_up
   )
   > &RigUnit_ModifyTransforms(
    ItemToModify=((Item=(Type=Bone,Name="deform_wrist_up_l"),Transform=(Rotation=(X=0,Y=0,Z=0,W=1),Translation=(X=0,Y=0,Z=0),Scale3D=(X=1,Y=1,Z=1))))
    Weight=@n3.OutputParam
   )
   &RigUnit_ParentConstraint(
    AdvancedSettings=(RotationOrderForFilter=XZY)
    Child=(Type=Bone,Name="deform_wrist_up_base_r")
    Parents=((Item=(Type=Bone,Name="hand_r"),Weight=0.4),(Item=(Type=Bone,Name="lowerarm_twist_01_r"),Weight=0.6))
    Weight=1
    ExecutePin=@SequenceExecution_1.B
   )
   > n4 = &RigUnit_SphericalPoseReader(
    ActiveRegionScaleFactors=(PositiveWidth=0.5,NegativeWidth=3.5,PositiveHeight=3,NegativeHeight=3)
    ActiveRegionSize=0.1
    Debug=(bDrawDebug=false,DebugScale=25,DebugSegments=20,DebugThickness=0.25)
    DriverAxis=(X=-1,Y=0,Z=0)
    DriverItem=(Type=Bone,Name="hand_r")
    FalloffRegionScaleFactors=(PositiveWidth=1,NegativeWidth=1,PositiveHeight=1,NegativeHeight=1)
    FalloffSize=0.45
    FlipHeightScaling=true
    OptionalParentItem=(Type=Bone,Name="lowerarm_twist_01_r")
    RotationOffset=(X=90,Y=0,Z=180)
    Debug.bDrawDebug=@n1.debug_up
   )
   > Deform_GetMirrorTransform_FNC = RigVMFunctionReferenceNode(transform=(Rotation=(X=0,Y=0,Z=0,W=1),Translation=(X=0,Y=0,Z=0),Scale3D=(X=1,Y=1,Z=1)))
   > &RigUnit_ModifyTransforms(
    ItemToModify=((Item=(Type=Bone,Name="deform_wrist_up_r"),Transform=(Rotation=(X=0,Y=0,Z=0,W=1),Translation=(X=0,Y=0,Z=0),Scale3D=(X=1,Y=1,Z=1))))
    ItemToModify.0.Transform=@Deform_GetMirrorTransform_FNC.mirrorTransform
    Weight=@n4.OutputParam
   )
   n5 = &RigVMFunction_ControlFlowBranch(Condition=@n1.wrist_dn, ExecuteContext=@SequenceExecution.B)
   SequenceExecution_1_1 = RigVMAggregateNode(ExecuteContext=@n5.True)
   &RigUnit_ParentConstraint(
    AdvancedSettings=(RotationOrderForFilter=XZY)
    Child=(Type=Bone,Name="deform_wrist_dn_base_l")
    Parents=((Item=(Type=Bone,Name="hand_l"),Weight=0.4),(Item=(Type=Bone,Name="lowerarm_twist_01_l"),Weight=0.6))
    Weight=1
    ExecutePin=@SequenceExecution_1_1.A
   )
   > n6 = &RigUnit_SphericalPoseReader(
    ActiveRegionScaleFactors=(PositiveWidth=1,NegativeWidth=4,PositiveHeight=3,NegativeHeight=3)
    ActiveRegionSize=0.1
    Debug=(bDrawDebug=false,DebugScale=25,DebugSegments=20,DebugThickness=0.25)
    DriverAxis=(X=1,Y=0,Z=0)
    DriverItem=(Type=Bone,Name="hand_l")
    FalloffRegionScaleFactors=(PositiveWidth=1,NegativeWidth=1,PositiveHeight=1,NegativeHeight=1)
    FalloffSize=0.4
    OptionalParentItem=(Type=Bone,Name="lowerarm_twist_01_l")
    RotationOffset=(X=-90,Y=0,Z=0)
    Debug.bDrawDebug=@n1.debug_dn
   )
   > &RigUnit_ModifyTransforms(
    ItemToModify=((Item=(Type=Bone,Name="deform_wrist_dn_l"),Transform=(Rotation=(X=0,Y=0,Z=0,W=1),Translation=(X=0,Y=0,Z=0),Scale3D=(X=1,Y=1,Z=1))))
    Weight=@n6.OutputParam
   )
   &RigUnit_ParentConstraint(
    AdvancedSettings=(RotationOrderForFilter=XZY)
    Child=(Type=Bone,Name="deform_wrist_dn_base_r")
    Parents=((Item=(Type=Bone,Name="hand_r"),Weight=0.4),(Item=(Type=Bone,Name="lowerarm_twist_01_r"),Weight=0.6))
    Weight=1
    ExecutePin=@SequenceExecution_1_1.B
   )
   > n7 = &RigUnit_SphericalPoseReader(
    ActiveRegionScaleFactors=(PositiveWidth=1,NegativeWidth=4,PositiveHeight=3,NegativeHeight=3)
    ActiveRegionSize=0.1
    Debug=(bDrawDebug=false,DebugScale=25,DebugSegments=20,DebugThickness=0.25)
    DriverAxis=(X=-1,Y=0,Z=0)
    DriverItem=(Type=Bone,Name="hand_r")
    FalloffRegionScaleFactors=(PositiveWidth=1,NegativeWidth=1,PositiveHeight=1,NegativeHeight=1)
    FalloffSize=0.4
    FlipHeightScaling=true
    OptionalParentItem=(Type=Bone,Name="lowerarm_twist_01_r")
    RotationOffset=(X=-90,Y=0,Z=190)
    Debug.bDrawDebug=@n1.debug_dn
   )
   > Deform_GetMirrorTransform_FNC_1 = RigVMFunctionReferenceNode(transform=(Rotation=(X=0,Y=0,Z=0,W=1),Translation=(X=0,Y=0,Z=0),Scale3D=(X=1,Y=1,Z=1)))
   > &RigUnit_ModifyTransforms(
    ItemToModify=((Item=(Type=Bone,Name="deform_wrist_dn_r"),Transform=(Rotation=(X=0,Y=0,Z=0,W=1),Translation=(X=0,Y=0,Z=0),Scale3D=(X=1,Y=1,Z=1))))
    ItemToModify.0.Transform=@Deform_GetMirrorTransform_FNC_1.mirrorTransform
    Weight=@n7.OutputParam
   )
   RigVMFunctionReturnNode(ExecuteContext=@SequenceExecution.D)
  ]}
 ContainedGraph {nodes [
   n1 = RigVMFunctionEntryNode()
   SequenceExecution = &RigVMFunction_Sequence(ExecuteContext=@n1.ExecuteContext)
   SequenceExecution_1 = &RigVMFunction_Sequence(ExecuteContext=@SequenceExecution.B)
   SequenceExecution_2 = &RigVMFunction_Sequence(ExecuteContext=@SequenceExecution_1.B)
   RigVMFunctionReturnNode(A=@SequenceExecution.A, B=@SequenceExecution_1.A, C=@SequenceExecution_2.A, D=@SequenceExecution_2.B)
  ]}
 ContainedGraph {nodes [
   n1 = RigVMFunctionEntryNode()
   SequenceExecution_1 = &RigVMFunction_Sequence(ExecuteContext=@n1.ExecuteContext)
   SequenceExecution_1_1 = &RigVMFunction_Sequence(ExecuteContext=@SequenceExecution_1.B)
   SequenceExecution_1_2 = &RigVMFunction_Sequence(ExecuteContext=@SequenceExecution_1_1.B)
   RigVMFunctionReturnNode(A=@SequenceExecution_1.A, B=@SequenceExecution_1_1.A, C=@SequenceExecution_1_2.A, D=@SequenceExecution_1_2.B)
  ]}
 ContainedGraph {nodes [
   n1 = RigVMFunctionEntryNode()
   SequenceExecution_1_1 = &RigVMFunction_Sequence(ExecuteContext=@n1.ExecuteContext)
   SequenceExecution_1_1_1 = &RigVMFunction_Sequence(ExecuteContext=@SequenceExecution_1_1.B)
   SequenceExecution_1_1_2 = &RigVMFunction_Sequence(ExecuteContext=@SequenceExecution_1_1_1.B)
   RigVMFunctionReturnNode(A=@SequenceExecution_1_1.A, B=@SequenceExecution_1_1_1.A, C=@SequenceExecution_1_1_2.A, D=@SequenceExecution_1_1_2.B)
  ]}
 ContainedGraph {nodes [
   n1 = RigVMFunctionEntryNode()
   n2 = &RigVMFunction_MathTransformToEulerTransform(Value=@n1.transform) {template="Cast::Execute(in Value,out Result)"}
   > RigVMFunctionReturnNode(
    mirrorTransform=(Rotation=(X=0,Y=0,Z=0,W=1),Translation=(X=0,Y=0,Z=0),Scale3D=(X=0,Y=0,Z=0))
    mirrorTransform.Rotation=&RigVMFunction_MathQuaternionFromRotatorV2(Value=@n2.Result.Rotation) {template="Cast::Execute(in Value,out Result)"}.Result
    mirrorTransform.Translation=&RigVMFunction_MathVectorScale(Factor=-1, Value=@n2.Result.Location) {template="Scale::Execute(in Value,in Factor,out Result)"}.Result
    mirrorTransform.Scale3D=@n2.Result.Scale
   )
  ]}
 ContainedGraph {nodes [
   n1 = RigVMFunctionEntryNode()
   > SequenceExecution = RigVMAggregateNode()
   n2 = &RigUnit_SphericalPoseReader(
    ActiveRegionScaleFactors=(PositiveWidth=4,NegativeWidth=0.5,PositiveHeight=1.5,NegativeHeight=4)
    ActiveRegionSize=0.05
    Debug=(bDrawDebug=false,DebugScale=25,DebugSegments=20,DebugThickness=0.25)
    DriverAxis=(X=1,Y=0,Z=0)
    DriverItem=(Type=Bone,Name="upperarm_l")
    FalloffRegionScaleFactors=(PositiveWidth=1,NegativeWidth=1,PositiveHeight=1,NegativeHeight=1)
    FalloffSize=0.2
    OptionalParentItem=(Type=Bone,Name="spine_03")
    RotationOffset=(X=0,Y=-150,Z=0)
    Debug.bDrawDebug=@n1.debug
    ExecutePin=@SequenceExecution.A
   )
   > &RigUnit_ModifyTransforms(
    ItemToModify=((Item=(Type=Bone,Name="deform_lat_l"),Transform=(Rotation=(X=0,Y=0,Z=0,W=1),Translation=(X=0,Y=0,Z=0),Scale3D=(X=1,Y=1,Z=1))))
    Weight=@n2.OutputParam
   )
   n3 = &RigUnit_SphericalPoseReader(
    ActiveRegionScaleFactors=(PositiveWidth=4,NegativeWidth=0.5,PositiveHeight=1.5,NegativeHeight=4)
    ActiveRegionSize=0.05
    Debug=(bDrawDebug=false,DebugScale=25,DebugSegments=20,DebugThickness=0.25)
    DriverAxis=(X=-1,Y=0,Z=0)
    DriverItem=(Type=Bone,Name="upperarm_r")
    FalloffRegionScaleFactors=(PositiveWidth=1,NegativeWidth=1,PositiveHeight=1,NegativeHeight=1)
    FalloffSize=0.2
    FlipHeightScaling=true
    OptionalParentItem=(Type=Bone,Name="spine_03")
    RotationOffset=(X=0,Y=30,Z=0)
    Debug.bDrawDebug=@n1.debug
    ExecutePin=@SequenceExecution.B
   )
   > Deform_GetMirrorTransform_FNC = RigVMFunctionReferenceNode(transform=(Rotation=(X=0,Y=0,Z=0,W=1),Translation=(X=0,Y=0,Z=0),Scale3D=(X=1,Y=1,Z=1)))
   > &RigUnit_ModifyTransforms(
    ItemToModify=((Item=(Type=Bone,Name="deform_lat_r"),Transform=(Rotation=(X=0,Y=0,Z=0,W=1),Translation=(X=0,Y=0,Z=0),Scale3D=(X=1,Y=1,Z=1))))
    ItemToModify.0.Transform=@Deform_GetMirrorTransform_FNC.mirrorTransform
    Weight=@n3.OutputParam
   )
   RigVMFunctionReturnNode(ExecuteContext=@SequenceExecution.C)
  ]}
 ContainedGraph {nodes [
   n1 = RigVMFunctionEntryNode()
   SequenceExecution = &RigVMFunction_Sequence(ExecuteContext=@n1.ExecuteContext)
   SequenceExecution_1 = &RigVMFunction_Sequence(ExecuteContext=@SequenceExecution.B)
   SequenceExecution_2 = &RigVMFunction_Sequence(ExecuteContext=@SequenceExecution_1.B)
   RigVMFunctionReturnNode(A=@SequenceExecution.A, B=@SequenceExecution_1.A, C=@SequenceExecution_2.A, D=@SequenceExecution_2.B)
  ]}
 ContainedGraph {nodes [
   n1 = RigVMFunctionEntryNode()
   n2 = &RigUnit_GetTransform(Space=LocalSpace, Item=@n1.Item)
   n3 = &RigUnit_GetTransform(Space=LocalSpace, bInitial=true, Item=@n1.Item)
   > RigVMFunctionReturnNode(OutTransform=&RigVMFunction_MathTransformToEulerTransform(Value=&RigVMFunction_MathTransformMakeRelative(Global=@n2.Transform, Parent=@n3.Transform).Local) {template="Cast::Execute(in Value,out Result)"}.Result)
  ]}
 ContainedGraph {nodes [
   n1 = RigVMFunctionEntryNode()
   n2 = &RigVMFunction_MathTransformToEulerTransform(Value=@n1.transform) {template="Cast::Execute(in Value,out Result)"}
   > RigVMFunctionReturnNode(
    mirrorTransform=(Rotation=(X=0,Y=0,Z=0,W=1),Translation=(X=0,Y=0,Z=0),Scale3D=(X=0,Y=0,Z=0))
    mirrorTransform.Rotation=&RigVMFunction_MathQuaternionFromRotatorV2(Value=@n2.Result.Rotation) {template="Cast::Execute(in Value,out Result)"}.Result
    mirrorTransform.Translation=&RigVMFunction_MathVectorScale(Factor=-1, Value=@n2.Result.Location) {template="Scale::Execute(in Value,in Factor,out Result)"}.Result
    mirrorTransform.Scale3D=@n2.Result.Scale
   )
  ]}
 ContainedGraph {nodes [
   n1 = RigVMFunctionEntryNode()
   > SequenceExecution_2 = RigVMAggregateNode()
   GetDeltaTransformFromRefPose_FNC = RigVMFunctionReferenceNode(Item=(Type=Bone,Name="dfrm_calf_l"), ExecuteContext=@SequenceExecution_2.A)
   GetDeltaTransformFromRefPose_FNC_1 = RigVMFunctionReferenceNode(Item=(Type=Bone,Name="dfrm_calf_r"), ExecuteContext=@SequenceExecution_2.B)
   SequenceExecution = RigVMAggregateNode(ExecuteContext=@SequenceExecution_2.C)
   n2 = &RigVMFunction_ControlFlowBranch(Condition=@n1.knee_in, ExecuteContext=@SequenceExecution.A)
   SequenceExecution_1 = RigVMAggregateNode(ExecuteContext=@n2.True)
   &RigUnit_ParentConstraint(
    AdvancedSettings=(RotationOrderForFilter=XZY)
    Child=(Type=Bone,Name="deform_knee_in_base_l")
    Parents=((Item=(Type=Bone,Name="dfrm_calf_l"),Weight=0.5),(Item=(Type=Bone,Name="dfrm_thigh_l"),Weight=0.5))
    Weight=1
    ExecutePin=@SequenceExecution_1.A
   )
   n3 = &RigVMFunction_MathFloatRemap(SourceMaximum=100, SourceMinimum=10, bClamp=true, Value=@GetDeltaTransformFromRefPose_FNC.OutTransform.Rotation.Yaw)
   > &RigUnit_ModifyTransforms(
    ItemToModify=((Item=(Type=Bone,Name="deform_knee_in_l"),Transform=(Rotation=(X=0,Y=0,Z=0,W=1),Translation=(X=0,Y=0,Z=0),Scale3D=(X=1,Y=1,Z=1))))
    Weight=@n3.Result
   )
   &RigUnit_ParentConstraint(
    AdvancedSettings=(RotationOrderForFilter=XZY)
    Child=(Type=Bone,Name="deform_knee_in_base_r")
    Parents=((Item=(Type=Bone,Name="dfrm_calf_r"),Weight=0.5),(Item=(Type=Bone,Name="dfrm_thigh_r"),Weight=0.5))
    Weight=1
    ExecutePin=@SequenceExecution_1.B
   )
   > Deform_GetMirrorTransform_FNC = RigVMFunctionReferenceNode(transform=(Rotation=(X=0,Y=0,Z=0,W=1),Translation=(X=0,Y=0,Z=0),Scale3D=(X=1,Y=1,Z=1)))
   n4 = &RigVMFunction_MathFloatRemap(SourceMaximum=100, SourceMinimum=10, bClamp=true, Value=@GetDeltaTransformFromRefPose_FNC_1.OutTransform.Rotation.Yaw)
   > &RigUnit_ModifyTransforms(
    ItemToModify=((Item=(Type=Bone,Name="deform_knee_in_r"),Transform=(Rotation=(X=0,Y=0,Z=0,W=1),Translation=(X=0,Y=0,Z=0),Scale3D=(X=1,Y=1,Z=1))))
    ItemToModify.0.Transform=@Deform_GetMirrorTransform_FNC.mirrorTransform
    Weight=@n4.Result
   )
   n5 = &RigVMFunction_ControlFlowBranch(Condition=@n1.knee_out, ExecuteContext=@SequenceExecution.B)
   SequenceExecution_1_1 = RigVMAggregateNode(ExecuteContext=@n5.True)
   &RigUnit_ParentConstraint(
    AdvancedSettings=(RotationOrderForFilter=XZY)
    Child=(Type=Bone,Name="deform_knee_out_base_l")
    Parents=((Item=(Type=Bone,Name="dfrm_calf_l"),Weight=0.1),(Item=(Type=Bone,Name="dfrm_thigh_l"),Weight=0.9))
    Weight=1
    ExecutePin=@SequenceExecution_1_1.A
   )
   > &RigUnit_ModifyTransforms(
    ItemToModify=((Item=(Type=Bone,Name="deform_knee_out_l"),Transform=(Rotation=(X=0,Y=0,Z=0,W=1),Translation=(X=0,Y=0,Z=0),Scale3D=(X=1,Y=1,Z=1))))
    Weight=@n3.Result
   )
   &RigUnit_ParentConstraint(
    AdvancedSettings=(RotationOrderForFilter=XZY)
    Child=(Type=Bone,Name="deform_knee_out_base_r")
    Parents=((Item=(Type=Bone,Name="dfrm_calf_r"),Weight=0.1),(Item=(Type=Bone,Name="dfrm_thigh_r"),Weight=0.9))
    Weight=1
    ExecutePin=@SequenceExecution_1_1.B
   )
   > Deform_GetMirrorTransform_FNC_1 = RigVMFunctionReferenceNode(transform=(Rotation=(X=0,Y=0,Z=0,W=1),Translation=(X=0,Y=0,Z=0),Scale3D=(X=1,Y=1,Z=1)))
   > &RigUnit_ModifyTransforms(
    ItemToModify=((Item=(Type=Bone,Name="deform_knee_out_r"),Transform=(Rotation=(X=0,Y=0,Z=0,W=1),Translation=(X=0,Y=0,Z=0),Scale3D=(X=1,Y=1,Z=1))))
    ItemToModify.0.Transform=@Deform_GetMirrorTransform_FNC_1.mirrorTransform
    Weight=@n4.Result
   )
   RigVMFunctionReturnNode(ExecuteContext=@SequenceExecution_2.D)
  ]}
 ContainedGraph {nodes [
   n1 = RigVMFunctionEntryNode()
   SequenceExecution_2 = &RigVMFunction_Sequence(ExecuteContext=@n1.ExecuteContext)
   SequenceExecution_2_1 = &RigVMFunction_Sequence(ExecuteContext=@SequenceExecution_2.B)
   SequenceExecution_2_2 = &RigVMFunction_Sequence(ExecuteContext=@SequenceExecution_2_1.B)
   RigVMFunctionReturnNode(A=@SequenceExecution_2.A, B=@SequenceExecution_2_1.A, C=@SequenceExecution_2_2.A, D=@SequenceExecution_2_2.B)
  ]}
 ContainedGraph {nodes [
   n1 = RigVMFunctionEntryNode()
   SequenceExecution = &RigVMFunction_Sequence(ExecuteContext=@n1.ExecuteContext)
   SequenceExecution_1 = &RigVMFunction_Sequence(ExecuteContext=@SequenceExecution.B)
   SequenceExecution_2 = &RigVMFunction_Sequence(ExecuteContext=@SequenceExecution_1.B)
   RigVMFunctionReturnNode(A=@SequenceExecution.A, B=@SequenceExecution_1.A, C=@SequenceExecution_2.A, D=@SequenceExecution_2.B)
  ]}
 ContainedGraph {nodes [
   n1 = RigVMFunctionEntryNode()
   SequenceExecution_1 = &RigVMFunction_Sequence(ExecuteContext=@n1.ExecuteContext)
   SequenceExecution_1_1 = &RigVMFunction_Sequence(ExecuteContext=@SequenceExecution_1.B)
   SequenceExecution_1_2 = &RigVMFunction_Sequence(ExecuteContext=@SequenceExecution_1_1.B)
   RigVMFunctionReturnNode(A=@SequenceExecution_1.A, B=@SequenceExecution_1_1.A, C=@SequenceExecution_1_2.A, D=@SequenceExecution_1_2.B)
  ]}
 ContainedGraph {nodes [
   n1 = RigVMFunctionEntryNode()
   SequenceExecution_1_1 = &RigVMFunction_Sequence(ExecuteContext=@n1.ExecuteContext)
   SequenceExecution_1_1_1 = &RigVMFunction_Sequence(ExecuteContext=@SequenceExecution_1_1.B)
   SequenceExecution_1_1_2 = &RigVMFunction_Sequence(ExecuteContext=@SequenceExecution_1_1_1.B)
   RigVMFunctionReturnNode(A=@SequenceExecution_1_1.A, B=@SequenceExecution_1_1_1.A, C=@SequenceExecution_1_1_2.A, D=@SequenceExecution_1_1_2.B)
  ]}
 ContainedGraph {nodes [
   n1 = RigVMFunctionEntryNode()
   n2 = &RigVMFunction_MathTransformToEulerTransform(Value=@n1.transform) {template="Cast::Execute(in Value,out Result)"}
   > RigVMFunctionReturnNode(
    mirrorTransform=(Rotation=(X=0,Y=0,Z=0,W=1),Translation=(X=0,Y=0,Z=0),Scale3D=(X=0,Y=0,Z=0))
    mirrorTransform.Rotation=&RigVMFunction_MathQuaternionFromRotatorV2(Value=@n2.Result.Rotation) {template="Cast::Execute(in Value,out Result)"}.Result
    mirrorTransform.Translation=&RigVMFunction_MathVectorScale(Factor=-1, Value=@n2.Result.Location) {template="Scale::Execute(in Value,in Factor,out Result)"}.Result
    mirrorTransform.Scale3D=@n2.Result.Scale
   )
  ]}
 ContainedGraph {nodes [
   n1 = RigVMFunctionEntryNode()
   n2 = &RigVMFunction_ControlFlowBranch(ExecuteContext=@n1.ExecuteContext, Condition=&RigVMFunction_MathBoolOr(B=@n1.glute, A=@n1.groin).Result)
   RigVMFunctionReturnNode(ExecuteContext=@n2.Completed)
   SequenceExecution = RigVMAggregateNode(ExecuteContext=@n2.True)
   n3 = &RigUnit_SphericalPoseReader(
    ActiveRegionScaleFactors=(PositiveWidth=3.5,NegativeWidth=0.5,PositiveHeight=3,NegativeHeight=3)
    ActiveRegionSize=0.1
    Debug=(bDrawDebug=false,DebugScale=25,DebugSegments=20,DebugThickness=0.25)
    DriverAxis=(X=-1,Y=0,Z=0)
    DriverItem=(Type=Bone,Name="dfrm_thigh_l")
    FalloffRegionScaleFactors=(PositiveWidth=1,NegativeWidth=1,PositiveHeight=1,NegativeHeight=1)
    FalloffSize=0.45
    OptionalParentItem=(Type=Bone)
    RotationOffset=(X=90,Y=0,Z=0)
    ExecutePin=@SequenceExecution.A
    Debug.bDrawDebug=@n1.debug
   )
   > n4 = &RigUnit_SphericalPoseReader(
    ActiveRegionScaleFactors=(PositiveWidth=3.5,NegativeWidth=0.5,PositiveHeight=3,NegativeHeight=3)
    ActiveRegionSize=0.1
    Debug=(bDrawDebug=false,DebugScale=25,DebugSegments=20,DebugThickness=0.25)
    DriverAxis=(X=1,Y=0,Z=0)
    DriverItem=(Type=Bone,Name="dfrm_thigh_r")
    FalloffRegionScaleFactors=(PositiveWidth=1,NegativeWidth=1,PositiveHeight=1,NegativeHeight=1)
    FalloffSize=0.45
    FlipHeightScaling=false
    FlipWidthScaling=true
    OptionalParentItem=(Type=Bone)
    RotationOffset=(X=-90,Y=0,Z=0)
    Debug.bDrawDebug=@n1.debug
   )
   n5 = &RigVMFunction_ControlFlowBranch(ExecuteContext=@SequenceExecution.B, Condition=@n1.groin)
   SequenceExecution_1 = RigVMAggregateNode(ExecuteContext=@n5.True)
   &RigUnit_ParentConstraint(
    AdvancedSettings=(RotationOrderForFilter=XZY)
    Child=(Type=Bone,Name="deform_groin_base_l")
    Parents=((Item=(Type=Bone,Name="dfrm_thigh_l"),Weight=0.7),(Item=(Type=Bone,Name="dfrm_pelvis"),Weight=0.3))
    Weight=1
    ExecutePin=@SequenceExecution_1.A
   )
   > &RigUnit_ModifyTransforms(
    ItemToModify=((Item=(Type=Bone,Name="deform_groin_l"),Transform=(Rotation=(X=0,Y=0,Z=0,W=1),Translation=(X=-3,Y=12,Z=0),Scale3D=(X=1,Y=1,Z=1))))
    Weight=@n3.OutputParam
   )
   &RigUnit_ParentConstraint(
    AdvancedSettings=(RotationOrderForFilter=XZY)
    Child=(Type=Bone,Name="deform_groin_base_r")
    Parents=((Item=(Type=Bone,Name="dfrm_thigh_r"),Weight=0.7),(Item=(Type=Bone,Name="dfrm_pelvis"),Weight=0.3))
    Weight=1
    ExecutePin=@SequenceExecution_1.B
   )
   > &RigUnit_ModifyTransforms(
    ItemToModify=((Item=(Type=Bone,Name="deform_groin_r"),Transform=(Rotation=(X=0,Y=0,Z=0,W=1),Translation=(X=3,Y=-12,Z=0),Scale3D=(X=1,Y=1,Z=1))))
    Weight=@n4.OutputParam
   )
   n6 = &RigVMFunction_ControlFlowBranch(ExecuteContext=@SequenceExecution.C, Condition=@n1.glute)
   SequenceExecution_1_1 = RigVMAggregateNode(ExecuteContext=@n6.True)
   &RigUnit_ParentConstraint(
    AdvancedSettings=(RotationOrderForFilter=XZY)
    Child=(Type=Bone,Name="deform_glute_base_l")
    Parents=((Item=(Type=Bone,Name="dfrm_thigh_l"),Weight=0.25),(Item=(Type=Bone,Name="dfrm_pelvis"),Weight=0.75))
    Weight=1
    ExecutePin=@SequenceExecution_1_1.A
   )
   > &RigUnit_ModifyTransforms(
    ItemToModify=((Item=(Type=Bone,Name="deform_glute_l"),Transform=(Rotation=(X=0,Y=0,Z=0,W=1),Translation=(X=0,Y=0,Z=0),Scale3D=(X=1,Y=1,Z=1))))
    Weight=@n3.OutputParam
   )
   &RigUnit_ParentConstraint(
    AdvancedSettings=(RotationOrderForFilter=XZY)
    Child=(Type=Bone,Name="deform_glute_base_r")
    Parents=((Item=(Type=Bone,Name="dfrm_thigh_r"),Weight=0.25),(Item=(Type=Bone,Name="dfrm_pelvis"),Weight=0.75))
    Weight=1
    ExecutePin=@SequenceExecution_1_1.B
   )
   > Deform_GetMirrorTransform_FNC_1 = RigVMFunctionReferenceNode(transform=(Rotation=(X=0,Y=0,Z=0,W=1),Translation=(X=0,Y=0,Z=0),Scale3D=(X=1,Y=1,Z=1)))
   > &RigUnit_ModifyTransforms(
    ItemToModify=((Item=(Type=Bone,Name="deform_glute_r"),Transform=(Rotation=(X=0,Y=0,Z=0,W=1),Translation=(X=0,Y=0,Z=0),Scale3D=(X=1,Y=1,Z=1))))
    ItemToModify.0.Transform=@Deform_GetMirrorTransform_FNC_1.mirrorTransform
    Weight=@n4.OutputParam
   )
  ]}
 ContainedGraph {nodes [
   n1 = RigVMFunctionEntryNode()
   SequenceExecution = &RigVMFunction_Sequence(ExecuteContext=@n1.ExecuteContext)
   SequenceExecution_1 = &RigVMFunction_Sequence(ExecuteContext=@SequenceExecution.B)
   SequenceExecution_2 = &RigVMFunction_Sequence(ExecuteContext=@SequenceExecution_1.B)
   RigVMFunctionReturnNode(A=@SequenceExecution.A, B=@SequenceExecution_1.A, C=@SequenceExecution_2.A, D=@SequenceExecution_2.B)
  ]}
 ContainedGraph {nodes [
   n1 = RigVMFunctionEntryNode()
   SequenceExecution_1 = &RigVMFunction_Sequence(ExecuteContext=@n1.ExecuteContext)
   SequenceExecution_1_1 = &RigVMFunction_Sequence(ExecuteContext=@SequenceExecution_1.B)
   SequenceExecution_1_2 = &RigVMFunction_Sequence(ExecuteContext=@SequenceExecution_1_1.B)
   RigVMFunctionReturnNode(A=@SequenceExecution_1.A, B=@SequenceExecution_1_1.A, C=@SequenceExecution_1_2.A, D=@SequenceExecution_1_2.B)
  ]}
 ContainedGraph {nodes [
   n1 = RigVMFunctionEntryNode()
   SequenceExecution_1_1 = &RigVMFunction_Sequence(ExecuteContext=@n1.ExecuteContext)
   SequenceExecution_1_1_1 = &RigVMFunction_Sequence(ExecuteContext=@SequenceExecution_1_1.B)
   SequenceExecution_1_1_2 = &RigVMFunction_Sequence(ExecuteContext=@SequenceExecution_1_1_1.B)
   RigVMFunctionReturnNode(A=@SequenceExecution_1_1.A, B=@SequenceExecution_1_1_1.A, C=@SequenceExecution_1_1_2.A, D=@SequenceExecution_1_1_2.B)
  ]}
 New Function_ContainedGraph {nodes [
   n1 = RigVMFunctionEntryNode()
   RigVMFunction_MathBoolToInteger = &RigVMFunction_MathBoolToInteger(Value=@n1.onlyTwist)
   RigVMFunction_MathQuaternionSwingTwist = &RigVMFunction_MathQuaternionSwingTwist(
    TwistAxis=@n1.twistAxis
    Input=&RigUnit_GetRelativeTransformForItem(bChildInitial=False, bParentInitial=False, Child=@n1.child, Parent=@n1.parent).RelativeTransform.Rotation
   )
   n2 = RigVMDispatchNode(
    Values=((),())
    Index=@RigVMFunction_MathBoolToInteger.Result
    Values.0=@RigVMFunction_MathQuaternionSwingTwist.Swing
    Values.1=@RigVMFunction_MathQuaternionSwingTwist.Twist
   ) {template="DISPATCH_RigVMDispatch_SelectInt32(in Index,in Values,out Result)"}
   RigVMFunction_MathQuaternionToEuler = &RigVMFunction_MathQuaternionToEuler(Value=@n2.Result, RotationOrder=@n1.Rotation Order)
   > DISPATCH_RigVMDispatch_Print = RigVMDispatchNode(ScreenColor=(R=1,G=1,B=1,A=1), ScreenDuration=1.5, Enabled=@n1.debugRotation, Value=@RigVMFunction_MathQuaternionToEuler.Result) {template="DISPATCH_RigVMDispatch_Print(in Prefix,in Value,in Enabled,in ScreenDuration,in ScreenColor)"}
   Cast = RigVMDispatchNode(Value=@n1.triggerAxis) {template="DISPATCH_RigVMDispatch_CastEnumToInt(in Value,out Result)"}
   DISPATCH_RigVMDispatch_SelectInt32 = RigVMDispatchNode(
    Values=(0,0,0)
    Index=@Cast.Result
    Values.0=@RigVMFunction_MathQuaternionToEuler.Result.X
    Values.1=@RigVMFunction_MathQuaternionToEuler.Result.Y
    Values.2=@RigVMFunction_MathQuaternionToEuler.Result.Z
   ) {template="DISPATCH_RigVMDispatch_SelectInt32(in Index,in Values,out Result)"}
   Remap = &RigVMFunction_MathDoubleRemap(
    bClamp=False
    Value=@DISPATCH_RigVMDispatch_SelectInt32.Result
    TargetMinimum=@n1.outputMinMax.X
    TargetMaximum=@n1.outputMinMax.Y
    SourceMinimum=@n1.rotationMinMax.X
    SourceMaximum=@n1.rotationMinMax.Y
   )
   Absolute = &RigVMFunction_MathDoubleAbs(Value=@Remap.Result)
   Multiply = &RigVMFunction_MathDoubleMul(
    A=&RigVMFunction_MathDoubleClamp(Value=RigVMDispatchNode(True=@Absolute.Result, Condition=@n1.absoluteValue, False=@Remap.Result) {template="DISPATCH_RigVMDispatch_If(in Condition,in True,in False,out Result)"}.Result).Result
    B=@n1.premult
   )
   > RigVMFunctionReturnNode(outputQuat=@n2.Result, outputWeight=@Multiply.Result)
  ]}
 New Function_ContainedGraph {nodes [
   n1 = RigVMFunctionEntryNode()
   RigVMFunction_MathBoolToInteger = &RigVMFunction_MathBoolToInteger(Value=@n1.onlyTwist)
   RigVMFunction_MathQuaternionSwingTwist = &RigVMFunction_MathQuaternionSwingTwist(
    TwistAxis=@n1.twistAxis
    Input=&RigUnit_GetRelativeTransformForItem(bChildInitial=False, bParentInitial=False, Child=@n1.child, Parent=@n1.parent).RelativeTransform.Rotation
   )
   n2 = RigVMDispatchNode(
    Values=((),())
    Index=@RigVMFunction_MathBoolToInteger.Result
    Values.0=@RigVMFunction_MathQuaternionSwingTwist.Swing
    Values.1=@RigVMFunction_MathQuaternionSwingTwist.Twist
   ) {template="DISPATCH_RigVMDispatch_SelectInt32(in Index,in Values,out Result)"}
   RigVMFunction_MathQuaternionToEuler = &RigVMFunction_MathQuaternionToEuler(Value=@n2.Result, RotationOrder=@n1.Rotation Order)
   > DISPATCH_RigVMDispatch_Print = RigVMDispatchNode(ScreenColor=(R=1,G=1,B=1,A=1), ScreenDuration=1.5, Enabled=@n1.debugRotation, Value=@RigVMFunction_MathQuaternionToEuler.Result) {template="DISPATCH_RigVMDispatch_Print(in Prefix,in Value,in Enabled,in ScreenDuration,in ScreenColor)"}
   Cast = RigVMDispatchNode(Value=@n1.triggerAxis) {template="DISPATCH_RigVMDispatch_CastEnumToInt(in Value,out Result)"}
   DISPATCH_RigVMDispatch_SelectInt32 = RigVMDispatchNode(
    Values=(0,0,0)
    Index=@Cast.Result
    Values.0=@RigVMFunction_MathQuaternionToEuler.Result.X
    Values.1=@RigVMFunction_MathQuaternionToEuler.Result.Y
    Values.2=@RigVMFunction_MathQuaternionToEuler.Result.Z
   ) {template="DISPATCH_RigVMDispatch_SelectInt32(in Index,in Values,out Result)"}
   Remap = &RigVMFunction_MathDoubleRemap(
    bClamp=False
    Value=@DISPATCH_RigVMDispatch_SelectInt32.Result
    TargetMinimum=@n1.outputMinMax.X
    TargetMaximum=@n1.outputMinMax.Y
    SourceMinimum=@n1.rotationMinMax.X
    SourceMaximum=@n1.rotationMinMax.Y
   )
   Absolute = &RigVMFunction_MathDoubleAbs(Value=@Remap.Result)
   Multiply = &RigVMFunction_MathDoubleMul(
    A=&RigVMFunction_MathDoubleClamp(Value=RigVMDispatchNode(True=@Absolute.Result, Condition=@n1.absoluteValue, False=@Remap.Result) {template="DISPATCH_RigVMDispatch_If(in Condition,in True,in False,out Result)"}.Result).Result
    B=@n1.premult
   )
   > RigVMFunctionReturnNode(outputQuat=@n2.Result, outputWeight=@Multiply.Result)
  ]}
 RigVMFunctionLibrary {nodes [
   Deform_Elbow_FNC = RigVMCollapseNode(ExecuteContext=(), bicep=true, elbow_in=true, elbow_out=true)
   Deform_GetMirrorTransform_FNC = RigVMCollapseNode(ExecuteContext=(), transform=(Rotation=(X=0,Y=0,Z=0,W=1),Translation=(X=0,Y=0,Z=0),Scale3D=(X=1,Y=1,Z=1)))
   Deform_GetMirrorTransform_FNC_1 = RigVMCollapseNode(ExecuteContext=(), transform=(Rotation=(X=0,Y=0,Z=0,W=1),Translation=(X=0,Y=0,Z=0),Scale3D=(X=1,Y=1,Z=1)))
   Deform_GetMirrorTransform_FNC_2 = RigVMCollapseNode(ExecuteContext=(), transform=(Rotation=(X=0,Y=0,Z=0,W=1),Translation=(X=0,Y=0,Z=0),Scale3D=(X=1,Y=1,Z=1)))
   Deform_GetMirrorTransform_FNC_3 = RigVMCollapseNode(ExecuteContext=(), transform=(Rotation=(X=0,Y=0,Z=0,W=1),Translation=(X=0,Y=0,Z=0),Scale3D=(X=1,Y=1,Z=1)))
   Deform_GetMirrorTransform_FNC_4 = RigVMCollapseNode(ExecuteContext=(), transform=(Rotation=(X=0,Y=0,Z=0,W=1),Translation=(X=0,Y=0,Z=0),Scale3D=(X=1,Y=1,Z=1)))
   Deform_GetMirrorTransform_FNC_5 = RigVMCollapseNode(ExecuteContext=(), transform=(Rotation=(X=0,Y=0,Z=0,W=1),Translation=(X=0,Y=0,Z=0),Scale3D=(X=1,Y=1,Z=1)))
   Deform_GetMirrorTransform_FNC_6 = RigVMCollapseNode(ExecuteContext=(), transform=(Rotation=(X=0,Y=0,Z=0,W=1),Translation=(X=0,Y=0,Z=0),Scale3D=(X=1,Y=1,Z=1)))
   Deform_GetMirrorTransform_FNC_7 = RigVMCollapseNode(ExecuteContext=(), transform=(Rotation=(X=0,Y=0,Z=0,W=1),Translation=(X=0,Y=0,Z=0),Scale3D=(X=1,Y=1,Z=1)))
   Deform_GetMirrorTransform_FNC_8 = RigVMCollapseNode(ExecuteContext=(), transform=(Rotation=(X=0,Y=0,Z=0,W=1),Translation=(X=0,Y=0,Z=0),Scale3D=(X=1,Y=1,Z=1)))
   Deform_GroinGlue_FNC = RigVMCollapseNode(ExecuteContext=(), debug=False, glute=true, groin=true)
   Deform_GroinGlue_FNC_1 = RigVMCollapseNode(ExecuteContext=(), debug=False, glute=true, groin=true)
   Deform_Knee_FNC = RigVMCollapseNode(ExecuteContext=(), knee_in=true, knee_out=true)
   Deform_Knee_FNC_1 = RigVMCollapseNode(ExecuteContext=(), knee_in=true, knee_out=true)
   Deform_Lat_FNC = RigVMCollapseNode(ExecuteContext=(), debug=False)
   Deform_Pec_FNC = RigVMCollapseNode(ExecuteContext=())
   Deform_UpperArm_FNC = RigVMCollapseNode(ExecuteContext=(), debug_bk=false, debug_fr=false, debug_up=false, upperarm_bk=true, upperarm_fr=true, upperarm_up=true)
   Deform_Wrist_FNC = RigVMCollapseNode(ExecuteContext=(), debug_dn=false, debug_up=false, wrist_dn=true, wrist_up=true)
   GetDeltaTransformFromRefPose_FNC = RigVMCollapseNode(ExecuteContext=(), Item=(Type=Bone))
   GetDeltaTransformFromRefPose_FNC_1 = RigVMCollapseNode(ExecuteContext=(), Item=(Type=Bone))
   GetDeltaTransformFromRefPose_FNC_2 = RigVMCollapseNode(ExecuteContext=(), Item=(Type=Bone))
   GetDeltaTransformFromRefPose_FNC_3 = RigVMCollapseNode(ExecuteContext=(), Item=(Type=Bone))
  ]}
 RigVMFunction_Sequence_1_1_ContainedGraph {nodes [
   n1 = RigVMFunctionEntryNode()
   n2 = &RigVMFunction_Sequence(ExecuteContext=@n1.ExecuteContext)
   n3 = &RigVMFunction_Sequence(ExecuteContext=@n2.B)
   RigVMFunctionReturnNode(A=@n2.A, B=@n3.A, C=@n3.B)
  ]}
 RigVMModel {nodes [
   RigUnit_BeginExecution = &RigUnit_BeginExecution()
   SequenceAggregate_1 = RigVMAggregateNode(ExecuteContext=@RigUnit_BeginExecution.ExecutePin)
   &RigUnit_ParentConstraint(
    AdvancedSettings=(InterpolationType=Average,RotationOrderForFilter=XZY)
    Child=(Type=Bone,Name="dfrm_thigh_l")
    Filter=(TranslationFilter=(bX=false,bY=false,bZ=false),RotationFilter=(bX=True,bY=True,bZ=True),ScaleFilter=(bX=True,bY=True,bZ=True))
    Parents=((Item=(Type=Bone,Name="thigh_l"),Weight=1))
    Weight=1
    ExecutePin=@SequenceAggregate_1.A
   )
   > &RigUnit_ParentConstraint(
    AdvancedSettings=(InterpolationType=Average,RotationOrderForFilter=XZY)
    Child=(Type=Bone,Name="dfrm_thigh_r")
    Filter=(TranslationFilter=(bX=false,bY=false,bZ=false),RotationFilter=(bX=True,bY=True,bZ=True),ScaleFilter=(bX=True,bY=True,bZ=True))
    Parents=((Item=(Type=Bone,Name="thigh_r"),Weight=1))
    Weight=1
   )
   > n1 = &RigUnit_SphericalPoseReader(
    ActiveRegionScaleFactors=(PositiveWidth=3.5,NegativeWidth=0.5,PositiveHeight=3,NegativeHeight=3)
    ActiveRegionSize=0.1
    Debug=(bDrawDebug=false,DebugScale=25,DebugSegments=20,DebugThickness=0.25)
    DriverAxis=(X=1,Y=0,Z=0)
    DriverItem=(Type=Bone,Name="dfrm_thigh_r")
    FalloffRegionScaleFactors=(PositiveWidth=1,NegativeWidth=1,PositiveHeight=1,NegativeHeight=1)
    FalloffSize=0.4
    FlipHeightScaling=false
    FlipWidthScaling=true
    OptionalParentItem=(Type=Bone)
    RotationOffset=(X=-90,Y=0,Z=0)
   )
   > RigVMVariableNode(Value=@n1.OutputParam) {variable=thigh r up}
   > &RigUnit_ModifyTransforms(
    ItemToModify=((Item=(Type=Bone,Name="dfrm_calf_r_PV"),Transform=(Rotation=(X=0,Y=0,Z=0,W=1),Translation=(X=0,Y=0,Z=5),Scale3D=(X=1,Y=1,Z=1))))
    Weight=1
    WeightMinimum=0
   )
   > &RigUnit_ModifyTransforms(
    ItemToModify=((Item=(Type=Bone,Name="dfrm_calf_l_PV"),Transform=(Rotation=(X=0,Y=0,Z=0,W=1),Translation=(X=0,Y=0,Z=-5),Scale3D=(X=1,Y=1,Z=1))))
    Weight=1
    WeightMinimum=0
   )
   > &RigUnit_ParentConstraint(
    AdvancedSettings=(InterpolationType=Average,RotationOrderForFilter=XZY)
    Child=(Type=Control,Name="PV L")
    Filter=(TranslationFilter=(bX=true,bY=true,bZ=true),RotationFilter=(bX=True,bY=True,bZ=True),ScaleFilter=(bX=True,bY=True,bZ=True))
    Parents=((Item=(Type=Bone,Name="dfrm_calf_l_PV"),Weight=1))
    Weight=1
   )
   > &RigUnit_ParentConstraint(
    AdvancedSettings=(InterpolationType=Average,RotationOrderForFilter=XZY)
    Child=(Type=Control,Name="PV R")
    Filter=(TranslationFilter=(bX=true,bY=true,bZ=true),RotationFilter=(bX=True,bY=True,bZ=True),ScaleFilter=(bX=True,bY=True,bZ=True))
    Parents=((Item=(Type=Bone,Name="dfrm_calf_r_PV"),Weight=1))
    Weight=1
   )
   RigUnit_GetTransform = &RigUnit_GetTransform(Item=(Type=Control,Name="dfrm_foot_l_ctrl"), bInitial=false)
   > /Script/ControlRig.RigUnit_TwoBoneIKSimplePerItem(
    DebugSettings=(bEnabled=False,Scale=10,WorldOffset=(Rotation=(X=0,Y=0,Z=0,W=1),Translation=(X=0,Y=0,Z=0),Scale3D=(X=1,Y=1,Z=1)))
    EffectorItem=(Type=Bone,Name="dfrm_foot_l")
    ItemA=(Type=Bone,Name="dfrm_thigh_l")
    ItemALength=0
    ItemB=(Type=Bone,Name="dfrm_calf_l")
    ItemBLength=0
    PoleVectorKind=Direction
    PoleVectorSpace=(Type=Bone,Name="None")
    PrimaryAxis=(X=-1,Y=0,Z=0)
    SecondaryAxis=(X=0,Y=1,Z=0)
    SecondaryAxisWeight=1
    StretchMaximumRatio=1.25
    StretchStartRatio=0.99
    Weight=1
    bEnableStretch=true
    bPropagateToChildren=True
    Effector=@RigUnit_GetTransform.Transform
    PoleVector=&RigUnit_GetTransform(Item=(Type=Control,Name="PV L"), bInitial=false).Transform.Translation
   ) {is=struct RigVMUnitNode}
   > &RigUnit_TwoBoneIKSimplePerItem(
    DebugSettings=(Scale=10,WorldOffset=(Rotation=(X=0,Y=0,Z=0,W=1),Translation=(X=0,Y=0,Z=0),Scale3D=(X=1,Y=1,Z=1)))
    EffectorItem=(Type=Bone,Name="dfrm_foot_r")
    ItemA=(Type=Bone,Name="dfrm_thigh_r")
    ItemB=(Type=Bone,Name="dfrm_calf_r")
    PoleVectorSpace=(Type=Bone,Name="None")
    PrimaryAxis=(X=1,Y=0,Z=0)
    SecondaryAxis=(X=0,Y=-1,Z=0)
    SecondaryAxisWeight=1
    StretchMaximumRatio=1.25
    StretchStartRatio=0.99
    Weight=1
    bEnableStretch=true
    bPropagateToChildren=True
    Effector=&RigUnit_GetTransform(Item=(Type=Control,Name="dfrm_foot_r_ctrl")).Transform
    PoleVector=&RigUnit_GetTransform(Item=(Type=Control,Name="PV R")).Transform.Translation
   )
   > &RigUnit_ParentConstraint(
    AdvancedSettings=(InterpolationType=Average,RotationOrderForFilter=XZY)
    Child=(Type=Bone,Name="dfrm_foot_l")
    Filter=(TranslationFilter=(bX=false,bY=false,bZ=false),RotationFilter=(bX=True,bY=True,bZ=True),ScaleFilter=(bX=True,bY=True,bZ=True))
    Parents=((Item=(Type=Bone,Name="foot_l"),Weight=1))
    Weight=1
   )
   > &RigUnit_ParentConstraint(
    AdvancedSettings=(InterpolationType=Average,RotationOrderForFilter=XZY)
    Child=(Type=Bone,Name="dfrm_foot_r")
    Filter=(TranslationFilter=(bX=false,bY=false,bZ=false),RotationFilter=(bX=True,bY=True,bZ=True),ScaleFilter=(bX=True,bY=True,bZ=True))
    Parents=((Item=(Type=Bone,Name="foot_r"),Weight=1))
    Weight=1
    bMaintainOffset=true
   )
   > &RigUnit_ModifyTransforms(
    ItemToModify=((Item=(Type=Bone,Name="dfrm_foot_l"),Transform=(Rotation=(X=0,Y=0,Z=0,W=1),Translation=(X=0,Y=1.75,Z=0),Scale3D=(X=1,Y=1,Z=1))))
    Weight=1
    WeightMinimum=0
   )
   > &RigUnit_ModifyTransforms(
    ItemToModify=((Item=(Type=Bone,Name="dfrm_foot_r"),Transform=(Rotation=(X=0,Y=0,Z=0,W=1),Translation=(X=0,Y=-1.75,Z=0),Scale3D=(X=1,Y=1,Z=1))))
    Weight=1
    WeightMinimum=0
   )
   > &RigUnit_ParentConstraint(
    AdvancedSettings=(InterpolationType=Average,RotationOrderForFilter=XZY)
    Child=(Type=Bone,Name="dfrm_thigh_twist_01_l")
    Filter=(TranslationFilter=(bX=false,bY=false,bZ=false),RotationFilter=(bX=True,bY=True,bZ=True),ScaleFilter=(bX=True,bY=True,bZ=True))
    Parents=((Item=(Type=Bone,Name="thigh_twist_01_l"),Weight=1))
    Weight=1
   )
   > &RigUnit_ParentConstraint(
    AdvancedSettings=(InterpolationType=Average,RotationOrderForFilter=XZY)
    Child=(Type=Bone,Name="dfrm_thigh_twist_01_r")
    Filter=(TranslationFilter=(bX=false,bY=false,bZ=false),RotationFilter=(bX=True,bY=True,bZ=True),ScaleFilter=(bX=True,bY=True,bZ=True))
    Parents=((Item=(Type=Bone,Name="thigh_twist_01_r"),Weight=1))
    Weight=1
   )
   > &RigUnit_ParentConstraint(
    AdvancedSettings=(InterpolationType=Average,RotationOrderForFilter=XZY)
    Child=(Type=Bone,Name="dfrm_calf_twist_01_l")
    Filter=(TranslationFilter=(bX=false,bY=true,bZ=true),RotationFilter=(bX=True,bY=True,bZ=True),ScaleFilter=(bX=True,bY=True,bZ=True))
    Parents=((Item=(Type=Bone,Name="calf_twist_01_l"),Weight=1))
    Weight=1
   )
   > &RigUnit_ParentConstraint(
    AdvancedSettings=(InterpolationType=Average,RotationOrderForFilter=XZY)
    Child=(Type=Bone,Name="dfrm_calf_twist_01_r")
    Filter=(TranslationFilter=(bX=false,bY=true,bZ=true),RotationFilter=(bX=True,bY=True,bZ=True),ScaleFilter=(bX=True,bY=True,bZ=True))
    Parents=((Item=(Type=Bone,Name="calf_twist_01_r"),Weight=1))
    Weight=1
   )
   > &RigUnit_ParentConstraint(
    AdvancedSettings=(InterpolationType=Average,RotationOrderForFilter=XZY)
    Child=(Type=Bone,Name="dfrm_calf_twist_02_l")
    Filter=(TranslationFilter=(bX=false,bY=false,bZ=true),RotationFilter=(bX=true,bY=true,bZ=true),ScaleFilter=(bX=True,bY=True,bZ=True))
    Parents=((Item=(Type=Bone,Name="calf_twist_02_l"),Weight=1))
    Weight=1
   )
   > &RigUnit_ParentConstraint(
    AdvancedSettings=(InterpolationType=Average,RotationOrderForFilter=XZY)
    Child=(Type=Bone,Name="dfrm_calf_twist_02_r")
    Filter=(TranslationFilter=(bX=false,bY=false,bZ=true),RotationFilter=(bX=True,bY=True,bZ=True),ScaleFilter=(bX=True,bY=True,bZ=True))
    Parents=((Item=(Type=Bone,Name="calf_twist_02_r"),Weight=1))
    Weight=1
   )
   > &RigUnit_RotationConstraintLocalSpaceOffset(
    AdvancedSettings=(InterpolationType=Average,RotationOrderForFilter=XZY)
    Child=(Type=Bone,Name="dyn_ball_l")
    Filter=(bX=True,bY=True,bZ=True)
    Parents=((Item=(Type=Bone,Name="ball_l"),Weight=1))
    Weight=1
   )
   > &RigUnit_RotationConstraintLocalSpaceOffset(
    AdvancedSettings=(InterpolationType=Average,RotationOrderForFilter=XZY)
    Child=(Type=Bone,Name="dyn_ball_r")
    Filter=(bX=True,bY=True,bZ=True)
    Parents=((Item=(Type=Bone,Name="ball_r"),Weight=1))
    Weight=1
   )
   SequenceAggregate_2_1_1 = RigVMAggregateNode(ExecuteContext=@SequenceAggregate_1.B)
   n2 = &RigVMFunction_ControlFlowBranch(
    Condition=RigVMVariableNode() {variable=deform_bool}.Value.knee_9_#1
    ExecuteContext=@SequenceAggregate_2_1_1.A
   )
   Deform_Knee_FNC_1_1 = RigVMFunctionReferenceNode(knee_in=true, knee_out=true, ExecuteContext=@n2.True)
   > &RigUnit_PositionConstraintLocalSpaceOffset(
    Child=(Type=Bone,Name="Foot_l_COL")
    Filter=(bX=false,bY=false,bZ=true)
    Parents=((Item=(Type=Bone,Name="dfrm_foot_l"),Weight=1))
    Weight=1
    bMaintainOffset=True
   )
   > &RigUnit_PositionConstraintLocalSpaceOffset(
    Child=(Type=Bone,Name="Foot_r_COL")
    Filter=(bX=false,bY=false,bZ=true)
    Parents=((Item=(Type=Bone,Name="dfrm_foot_r"),Weight=1))
    Weight=1
    bMaintainOffset=True
   )
   n3 = &RigVMFunction_MathDoubleAdd(B=0, A=&RigUnit_GetTransform(Item=(Type=Bone,Name="Foot_l_COL"), Space=LocalSpace, bInitial=False).Transform.Translation.Z)
   n4 = RigVMVariableNode() {variable=emote}
   > RigVMVariableNode(Value=&RigVMFunction_MathDoubleMul(
     A=&RigVMFunction_MathDoubleRemap(SourceMaximum=30, SourceMinimum=10, TargetMaximum=1, TargetMinimum=0, Value=@n3.Result).Result
     B=&RigVMFunction_MathDoubleSub(A=1, B=@n4.Value).Result
    ).Result) {variable=foot up l}
   n5 = &RigVMFunction_MathDoubleAdd(B=0, A=&RigUnit_GetTransform(Item=(Type=Bone,Name="Foot_r_COL"), bInitial=False).Transform.Translation.Z)
   n6 = RigVMVariableNode() {variable=emote}
   > RigVMVariableNode(Value=&RigVMFunction_MathDoubleMul(
     A=&RigVMFunction_MathDoubleRemap(SourceMaximum=30, SourceMinimum=10, TargetMaximum=1, TargetMinimum=0, Value=@n5.Result).Result
     B=&RigVMFunction_MathDoubleSub(A=1, B=@n6.Value).Result
    ).Result) {variable=foot up r}
   n7 = RigVMVariableNode() {variable=deform_bool}
   n8 = &RigVMFunction_ControlFlowBranch(
    Condition=&RigVMFunction_MathBoolOr(B=@n7.Value.glute_17_#2, A=@n7.Value.groin_8_#3).Result
    ExecuteContext=@SequenceAggregate_2_1_1.B
   )
   Deform_GroinGlue_FNC_1_1 = RigVMFunctionReferenceNode(
    debug=False
    ExecuteContext=@n8.True
    glute=@n7.Value.glute_17_#2
    groin=@n7.Value.groin_8_#3
   )
   n9 = RigVMVariableNode() {variable=foot up l}
   > &RigUnit_ModifyTransforms(
    ItemToModify=((Item=(Type=Bone,Name="dyn_ball_l"),Transform=(Rotation=(X=0,Y=0,Z=0.3007,W=0.9537),Translation=(X=0,Y=0,Z=0),Scale3D=(X=1,Y=1,Z=1))))
    WeightMinimum=0
    Weight=@n9.Value
   )
   > &RigUnit_ModifyTransforms(
    ItemToModify="((Item=(Type=Bone,Name=\"dyn_toe_l_a_01\"),Transform=(Rotation=(X=0.1036,Y=0.6036,Z=-0.0795,W=0.7866),Translation=(X=0,Y=0,Z=0),Scale3D=(X=1,Y=1,Z=1))),(Item=(Type=Bone,Name=\"dyn_toe_l_a_02\"),Transform=(Rotation=(X=0,Y=0.7071,Z=0,W=0.7071),Translation=(X=0,Y ...(truncated)"
    WeightMinimum=0
    Weight=@n9.Value
   )
   > &RigUnit_ModifyTransforms(
    ItemToModify="((Item=(Type=Bone,Name=\"dyn_toe_l_b_01\"),Transform=(Rotation=(X=0,Y=0.6088,Z=0,W=0.7934),Translation=(X=0,Y=0,Z=0),Scale3D=(X=1,Y=1,Z=1))),(Item=(Type=Bone,Name=\"dyn_toe_l_b_02\"),Transform=(Rotation=(X=0,Y=0.7373,Z=0,W=0.6756),Translation=(X=0,Y=0,Z=0),Sca ...(truncated)"
    WeightMinimum=0
    Weight=@n9.Value
   )
   > &RigUnit_ModifyTransforms(
    ItemToModify="((Item=(Type=Bone,Name=\"dyn_toe_l_c_01\"),Transform=(Rotation=(X=0,Y=0.6088,Z=0,W=0.7934),Translation=(X=0,Y=0,Z=0),Scale3D=(X=1,Y=1,Z=1))),(Item=(Type=Bone,Name=\"dyn_toe_l_c_02\"),Transform=(Rotation=(X=0,Y=0.7071,Z=0,W=0.7071),Translation=(X=0,Y=0,Z=0),Sca ...(truncated)"
    WeightMinimum=0
    Weight=@n9.Value
   )
   > &RigUnit_ModifyTransforms(
    ItemToModify="((Item=(Type=Bone,Name=\"dyn_toe_l_d_03\"),Transform=(Rotation=(X=0,Y=0.3007,Z=0,W=0.9537),Translation=(X=0,Y=0,Z=0),Scale3D=(X=1,Y=1,Z=1))),(Item=(Type=Bone,Name=\"dyn_toe_l_d_01\"),Transform=(Rotation=(X=0,Y=0.1305,Z=0,W=0.9914),Translation=(X=0,Y=0,Z=0),Sca ...(truncated)"
    WeightMinimum=0
    Weight=@n9.Value
   )
   n10 = RigVMVariableNode() {variable=foot up r}
   > &RigUnit_ModifyTransforms(
    ItemToModify=((Item=(Type=Bone,Name="dyn_ball_r"),Transform=(Rotation=(X=0,Y=0,Z=0.3007,W=0.9537),Translation=(X=0,Y=0,Z=0),Scale3D=(X=1,Y=1,Z=1))))
    WeightMinimum=0
    Weight=@n10.Value
   )
   > &RigUnit_ModifyTransforms(
    ItemToModify="((Item=(Type=Bone,Name=\"dyn_toe_r_a_01\"),Transform=(Rotation=(X=0,Y=0.6088,Z=0,W=0.7934),Translation=(X=0,Y=0,Z=0),Scale3D=(X=1,Y=1,Z=1))),(Item=(Type=Bone,Name=\"dyn_toe_r_a_02\"),Transform=(Rotation=(X=0,Y=0.7071,Z=0,W=0.7071),Translation=(X=0,Y=0,Z=0),Sca ...(truncated)"
    WeightMinimum=0
    Weight=@n10.Value
   )
   > &RigUnit_ModifyTransforms(
    ItemToModify="((Item=(Type=Bone,Name=\"dyn_toe_r_b_01\"),Transform=(Rotation=(X=0,Y=0.6088,Z=0,W=0.7934),Translation=(X=0,Y=0,Z=0),Scale3D=(X=1,Y=1,Z=1))),(Item=(Type=Bone,Name=\"dyn_toe_r_b_02\"),Transform=(Rotation=(X=0,Y=0.7373,Z=0,W=0.6756),Translation=(X=0,Y=0,Z=0),Sca ...(truncated)"
    WeightMinimum=0
    Weight=@n10.Value
   )
   > &RigUnit_ModifyTransforms(
    ItemToModify="((Item=(Type=Bone,Name=\"dyn_toe_r_c_01\"),Transform=(Rotation=(X=0,Y=0.6088,Z=0,W=0.7934),Translation=(X=0,Y=0,Z=0),Scale3D=(X=1,Y=1,Z=1))),(Item=(Type=Bone,Name=\"dyn_toe_r_c_02\"),Transform=(Rotation=(X=0,Y=0.7071,Z=0,W=0.7071),Translation=(X=0,Y=0,Z=0),Sca ...(truncated)"
    WeightMinimum=0
    Weight=@n10.Value
   )
   > &RigUnit_ModifyTransforms(
    ItemToModify="((Item=(Type=Bone,Name=\"dyn_toe_r_d_03\"),Transform=(Rotation=(X=0,Y=0.3007,Z=0,W=0.9537),Translation=(X=0,Y=0,Z=0),Scale3D=(X=1,Y=1,Z=1))),(Item=(Type=Bone,Name=\"dyn_toe_r_d_01\"),Transform=(Rotation=(X=0,Y=0.1305,Z=0,W=0.9914),Translation=(X=0,Y=0,Z=0),Sca ...(truncated)"
    WeightMinimum=0
    Weight=@n10.Value
   )
   SequenceAggregate_1_2 = RigVMAggregateNode(ExecuteContext=@SequenceAggregate_1.C)
   n11 = &RigUnit_SphericalPoseReader(
    ActiveRegionScaleFactors=(PositiveWidth=1.3,NegativeWidth=1,PositiveHeight=1,NegativeHeight=1)
    ActiveRegionSize=0.1
    Debug=(bDrawDebug=false,bDraw2D=False,bDrawLocalAxes=False,DebugScale=25,DebugSegments=20,DebugThickness=0.25)
    DriverAxis=(X=1,Y=0,Z=0)
    DriverItem=(Type=Bone,Name="upperarm_l")
    FalloffRegionScaleFactors=(PositiveWidth=2,NegativeWidth=1,PositiveHeight=2,NegativeHeight=1.5)
    FalloffSize=0.2
    FlipHeightScaling=true
    FlipWidthScaling=false
    OptionalParentItem=(Type=Bone,Name="spine_04")
    RotationOffset=(X=20,Y=-115,Z=15)
    ExecutePin=@SequenceAggregate_1_2.A
   )
   > RigVMVariableNode(Value=@n11.OutputParam) {variable=SPR_upperarm_l_down_A_alpha}
   > n12 = &RigUnit_SphericalPoseReader(
    ActiveRegionScaleFactors=(PositiveWidth=1.3,NegativeWidth=1,PositiveHeight=1,NegativeHeight=1)
    ActiveRegionSize=0.1
    Debug=(bDrawDebug=false,bDraw2D=False,bDrawLocalAxes=False,DebugScale=25,DebugSegments=20,DebugThickness=0.25)
    DriverAxis=(X=-1,Y=0,Z=0)
    DriverItem=(Type=Bone,Name="upperarm_r")
    FalloffRegionScaleFactors=(PositiveWidth=2,NegativeWidth=1,PositiveHeight=2,NegativeHeight=1.5)
    FalloffSize=0.2
    FlipHeightScaling=False
    FlipWidthScaling=False
    OptionalParentItem=(Type=Bone,Name="spine_04")
    RotationOffset=(X=-20,Y=75,Z=15)
   )
   > RigVMVariableNode(Value=@n12.OutputParam) {variable=SPR_upperarm_r_down_A_alpha}
   > n13 = &RigUnit_SphericalPoseReader(
    ActiveRegionScaleFactors=(PositiveWidth=1,NegativeWidth=1,PositiveHeight=1,NegativeHeight=1)
    ActiveRegionSize=0.05
    Debug=(bDrawDebug=false,bDraw2D=False,bDrawLocalAxes=true,DebugScale=25,DebugSegments=20,DebugThickness=0.25)
    DriverAxis=(X=-1,Y=0,Z=0)
    DriverItem=(Type=Bone,Name="upperarm_r")
    FalloffRegionScaleFactors=(PositiveWidth=0.5,NegativeWidth=1,PositiveHeight=0.75,NegativeHeight=1)
    FalloffSize=0.2
    FlipHeightScaling=False
    FlipWidthScaling=False
    OptionalParentItem=(Type=Bone,Name="None")
    RotationOffset=(X=-30,Y=25,Z=0)
   )
   > RigVMVariableNode(Value=@n13.OutputParam) {variable=SPR_upperarm_r_down_B_alpha}
   SequenceAggregate = &RigVMFunction_Sequence(ExecuteContext=@SequenceAggregate_1_2.C)
   RigUnit_ItemArray = &RigUnit_ItemArray(Items=((Type=Bone,Name="clavicle_l"),(Type=Bone,Name="upperarm_l"),(Type=Bone,Name="lowerarm_l")))
   DISPATCH_RigVMDispatch_ArrayIterator = RigVMDispatchNode(Array=@RigUnit_ItemArray.Items, ExecuteContext=@SequenceAggregate.A) {template="DISPATCH_RigVMDispatch_ArrayIterator(in Array,out Element,out Index,out Count,out Ratio)"}
   DISPATCH_RigVMDispatch_ArrayGetAtIndex = RigVMDispatchNode(
    Array=&RigUnit_ItemArray(Items=((Type=Bone,Name="dfrm_clavicle_l"),(Type=Bone,Name="dfrm_upperarm_l"),(Type=Bone,Name="dfrm_lowerarm_l"))).Items
    Index=@DISPATCH_RigVMDispatch_ArrayIterator.Index
   ) {template="DISPATCH_RigVMDispatch_ArrayGetAtIndex(in Array,in Index,out Element)"}
   > &RigUnit_SetRotation(
    Space=GlobalSpace
    Weight=1
    bInitial=False
    bPropagateToChildren=true
    Item=@DISPATCH_RigVMDispatch_ArrayGetAtIndex.Element
    Value=&RigUnit_GetTransform(Item=@DISPATCH_RigVMDispatch_ArrayIterator.Element).Transform.Rotation
   )
   &RigUnit_ModifyTransforms(
    ItemToModify=((Item=(Type=Bone,Name="dfrm_clavicle_l"),Transform=(Rotation=(X=0,Y=0,Z=0,W=1),Translation=(X=1,Y=0,Z=0),Scale3D=(X=1,Y=1,Z=1))))
    WeightMinimum=0
    Weight=RigVMVariableNode() {variable=SPR_upperarm_l_down_A_alpha}.Value
    ExecutePin=@DISPATCH_RigVMDispatch_ArrayIterator.Completed
   )
   > PBIK = /Script/PBIK.RigUnit_PBIK(
    BoneSettings=((Bone="dfrm_lowerarm_l",X=Locked,Y=Locked,Z=Free,PreferredAngles=(X=0,Y=0,Z=0)))
    Debug=(DrawScale=1)
    Effectors=((Bone="dfrm_hand_l",Transform=(Rotation=(X=0,Y=0,Z=0,W=1),Translation=(X=0,Y=0,Z=0),Scale3D=(X=1,Y=1,Z=1)),PositionAlpha=1,RotationAlpha=1,StrengthAlpha=1,PinRotation=1,ChainDepth=0))
    ExcludedBones=()
    Root=dfrm_clavicle_l
    Settings="(Iterations=20,SubIterations=0,MassMultiplier=1,bAllowStretch=False,RootBehavior=PinToInput,PrePullRootSettings=(RotationAlpha=0,RotationAlphaX=1,RotationAlphaY=1,RotationAlphaZ=1,PositionAlpha=1,PositionAlphaX=1,PositionAlphaY=1,PositionAlphaZ=1),GlobalPu ...(truncated)"
    Effectors.0.Transform=&RigUnit_GetTransform(Item=(Type=Bone,Name="hand_l")).Transform
   ) {is=struct RigVMUnitNode}
   n14 = RigVMDispatchNode(
    Array=&RigUnit_ItemArray(Items=((Type=Bone,Name="clavicle_r"),(Type=Bone,Name="upperarm_r"),(Type=Bone,Name="lowerarm_r"))).Items
    ExecuteContext=@SequenceAggregate.B
   ) {template="DISPATCH_RigVMDispatch_ArrayIterator(in Array,out Element,out Index,out Count,out Ratio)"}
   > &RigUnit_SetRotation(
    Space=GlobalSpace
    Weight=1
    bInitial=False
    bPropagateToChildren=true
    Item=RigVMDispatchNode(
     Array=&RigUnit_ItemArray(Items=((Type=Bone,Name="dfrm_clavicle_r"),(Type=Bone,Name="dfrm_upperarm_r"),(Type=Bone,Name="dfrm_lowerarm_r"))).Items
     Index=@n14.Index
    ) {template="DISPATCH_RigVMDispatch_ArrayGetAtIndex(in Array,in Index,out Element)"}.Element
    Value=&RigUnit_GetTransform(Item=@n14.Element).Transform.Rotation
   )
   &RigUnit_ModifyTransforms(
    ItemToModify=((Item=(Type=Bone,Name="dfrm_clavicle_r"),Transform=(Rotation=(X=0,Y=0,Z=0,W=1),Translation=(X=-2,Y=0,Z=0),Scale3D=(X=1,Y=1,Z=1))))
    WeightMinimum=0
    Weight=RigVMVariableNode() {variable=SPR_upperarm_r_down_B_alpha}.Value
    ExecutePin=@n14.Completed
   )
   > &RigUnit_PBIK(
    BoneSettings=((Bone="dfrm_lowerarm_r",X=Locked,Y=Locked,Z=Free,PreferredAngles=(X=0,Y=0,Z=0)))
    Debug=(DrawScale=1)
    Effectors=((Bone="dfrm_hand_r",Transform=(Rotation=(X=0,Y=0,Z=0,W=1),Translation=(X=0,Y=0,Z=0),Scale3D=(X=1,Y=1,Z=1)),PositionAlpha=1,RotationAlpha=1,StrengthAlpha=1,PinRotation=1,ChainDepth=0))
    ExcludedBones=()
    Root=dfrm_clavicle_l
    Settings="(Iterations=20,SubIterations=0,MassMultiplier=1,bAllowStretch=False,RootBehavior=PinToInput,PrePullRootSettings=(RotationAlpha=0,RotationAlphaX=1,RotationAlphaY=1,RotationAlphaZ=1,PositionAlpha=1,PositionAlphaX=1,PositionAlphaY=1,PositionAlphaZ=1),GlobalPu ...(truncated)"
    Effectors.0.Transform=&RigUnit_GetTransform(Item=(Type=Bone,Name="hand_r")).Transform
   )
   SequenceAggregate_2 = &RigVMFunction_Sequence(ExecuteContext=@SequenceAggregate_1_2.D)
   Subtract = /Script/RigVM.RigVMFunction_MathVectorSub(
    B=&RigUnit_GetTransform(Item=(Type=Bone,Name="dfrm_upperarm_l")).Transform.Translation
    A=&RigUnit_GetTransform(Item=(Type=Bone,Name="dfrm_hand_l")).Transform.Translation
   ) {is=struct RigVMUnitNode, template="Subtract::Execute(in A,in B,out Result)"}
   Unit = /Script/RigVM.RigVMFunction_MathVectorUnit(Value=@Subtract.Result) {is=struct RigVMUnitNode, template="Unit::Execute(in Value,out Result)"}
   RigVMFunction_MathQuaternionFromAxisAndAngle = /Script/RigVM.RigVMFunction_MathQuaternionFromAxisAndAngle(Axis=@Unit.Result, Angle=RigVMVariableNode() {variable=pistol}.Value) {is=struct RigVMUnitNode, template="FromAxisAndAngle::Execute(in Axis,in Angle,out Result)"}
   Set Transform = &RigUnit_SetRotation(
    Item=(Type=Bone,Name="dfrm_upperarm_l")
    Space=GlobalSpace
    bInitial=false
    Value=/Script/RigVM.RigVMFunction_MathQuaternionMul(
     B=&RigUnit_GetTransform(Item=(Type=Bone,Name="dfrm_upperarm_l")).Transform.Rotation
     A=@RigVMFunction_MathQuaternionFromAxisAndAngle.Result
    ) {is=struct RigVMUnitNode, template="Multiply::Execute(in A,in B,out Result)"}.Result
    ExecutePin=@SequenceAggregate_2.A
    Weight=RigVMVariableNode() {variable=mult_spr_dst_lowerarm_r_pelvis_alpha}.Value
   )
   n15 = &RigVMFunction_MathVectorUnit(Value=&RigVMFunction_MathVectorSub(
     B=&RigUnit_GetTransform(Item=(Type=Bone,Name="dfrm_lowerarm_l")).Transform.Translation
     A=&RigUnit_GetTransform(Item=(Type=Bone,Name="dfrm_hand_l")).Transform.Translation
    ) {template="Subtract::Execute(in A,in B,out Result)"}.Result) {template="Unit::Execute(in Value,out Result)"}
   > Set Transform_2 = &RigUnit_SetRotation(
    Item=(Type=Bone,Name="dfrm_lowerarm_l")
    Space=GlobalSpace
    bInitial=false
    Value=&RigVMFunction_MathQuaternionMul(
     B=&RigUnit_GetTransform(Item=(Type=Bone,Name="dfrm_lowerarm_l")).Transform.Rotation
     A=&RigVMFunction_MathQuaternionFromAxisAndAngle(Angle=-1.3, Axis=@n15.Result) {template="FromAxisAndAngle::Execute(in Axis,in Angle,out Result)"}.Result
    ) {template="Multiply::Execute(in A,in B,out Result)"}.Result
    Weight=RigVMVariableNode() {variable=mult_spr_dst_lowerarm_r_pelvis_alpha}.Value
   )
   > n16 = &RigUnit_SphericalPoseReader(
    ActiveRegionScaleFactors=(PositiveWidth=3,NegativeWidth=3,PositiveHeight=1,NegativeHeight=3)
    ActiveRegionSize=0.1
    Debug=(bDrawDebug=false,DebugScale=25,DebugSegments=20,DebugThickness=0.25)
    DriverAxis=(X=0,Y=1,Z=0)
    DriverItem=(Type=Bone,Name="head")
    FalloffRegionScaleFactors=(PositiveWidth=1,NegativeWidth=1,PositiveHeight=0.8,NegativeHeight=1)
    FalloffSize=0.4
    FlipHeightScaling=true
    OptionalParentItem=(Type=Bone,Name="pelvis")
    RotationOffset=(X=125,Y=0,Z=0)
   )
   > n17 = &RigUnit_SphericalPoseReader(
    ActiveRegionScaleFactors=(PositiveWidth=1,NegativeWidth=1.3,PositiveHeight=3,NegativeHeight=2)
    ActiveRegionSize=0.05
    Debug=(bDrawDebug=false,bDraw2D=false,bDrawLocalAxes=true,DebugScale=25,DebugSegments=20,DebugThickness=0.25)
    DriverAxis=(X=1,Y=0,Z=0)
    DriverItem=(Type=Bone,Name="upperarm_l")
    FalloffRegionScaleFactors=(PositiveWidth=0.5,NegativeWidth=0.7,PositiveHeight=1,NegativeHeight=1)
    FalloffSize=0.3
    OptionalParentItem=(Type=Bone,Name="spine_04")
    RotationOffset=(X=90,Y=25,Z=-60)
   )
   n18 = &RigVMFunction_MathVectorUnit(Value=&RigVMFunction_MathVectorSub(
     B=&RigUnit_GetTransform(Item=(Type=Bone,Name="dfrm_upperarm_l")).Transform.Translation
     A=&RigUnit_GetTransform(Item=(Type=Bone,Name="dfrm_hand_l")).Transform.Translation
    ) {template="Subtract::Execute(in A,in B,out Result)"}.Result) {template="Unit::Execute(in Value,out Result)"}
   > Set Transform_3 = &RigUnit_SetRotation(
    Item=(Type=Bone,Name="dfrm_upperarm_l")
    Space=GlobalSpace
    bInitial=false
    Value=&RigVMFunction_MathQuaternionMul(
     B=&RigUnit_GetTransform(Item=(Type=Bone,Name="dfrm_upperarm_l")).Transform.Rotation
     A=&RigVMFunction_MathQuaternionFromAxisAndAngle(Angle=-0.35, Axis=@n18.Result) {template="FromAxisAndAngle::Execute(in Axis,in Angle,out Result)"}.Result
    ) {template="Multiply::Execute(in A,in B,out Result)"}.Result
    Weight=&RigVMFunction_MathDoubleMul(B=@n17.OutputParam, A=@n16.OutputParam).Result
   )
   n19 = &RigVMFunction_MathVectorUnit(Value=&RigVMFunction_MathVectorSub(
     B=&RigUnit_GetTransform(Item=(Type=Bone,Name="dfrm_upperarm_r"), bInitial=False).Transform.Translation
     A=&RigUnit_GetTransform(Item=(Type=Bone,Name="dfrm_hand_r"), bInitial=False).Transform.Translation
    ) {template="Subtract::Execute(in A,in B,out Result)"}.Result) {template="Unit::Execute(in Value,out Result)"}
   Set Transform_1 = &RigUnit_SetRotation(
    Item=(Type=Bone,Name="dfrm_upperarm_r")
    Space=GlobalSpace
    bInitial=False
    Value=&RigVMFunction_MathQuaternionMul(
     B=&RigUnit_GetTransform(Item=(Type=Bone,Name="dfrm_upperarm_r"), bInitial=False).Transform.Rotation
     A=&RigVMFunction_MathQuaternionFromAxisAndAngle(Angle=0.1, Axis=@n19.Result) {template="FromAxisAndAngle::Execute(in Axis,in Angle,out Result)"}.Result
    ) {template="Multiply::Execute(in A,in B,out Result)"}.Result
    Weight=RigVMVariableNode() {variable=SPR_upperarm_r_down_B_alpha}.Value
    ExecutePin=@SequenceAggregate_2.B
   )
   > n20 = &RigUnit_SphericalPoseReader(
    ActiveRegionScaleFactors=(PositiveWidth=0.6,NegativeWidth=0.6,PositiveHeight=1,NegativeHeight=1)
    ActiveRegionSize=0.15
    Debug=(bDrawDebug=false,bDraw2D=false,bDrawLocalAxes=true,DebugScale=25,DebugSegments=20,DebugThickness=0.25)
    DriverAxis=(X=-1,Y=0,Z=0)
    DriverItem=(Type=Bone,Name="upperarm_r")
    FalloffRegionScaleFactors=(PositiveWidth=0.6,NegativeWidth=0.6,PositiveHeight=2,NegativeHeight=1)
    FalloffSize=0.45
    FlipHeightScaling=false
    FlipWidthScaling=true
    OptionalParentItem=(Type=Bone,Name="spine_04")
    RotationOffset=(X=-100,Y=25,Z=40)
   )
   n21 = &RigVMFunction_MathVectorUnit(Value=&RigVMFunction_MathVectorSub(
     A=&RigUnit_GetTransform(Item=(Type=Bone,Name="dfrm_hand_r"), bInitial=False).Transform.Translation
     B=&RigUnit_GetTransform(Item=(Type=Bone,Name="dfrm_upperarm_r"), bInitial=False).Transform.Translation
    ) {template="Subtract::Execute(in A,in B,out Result)"}.Result) {template="Unit::Execute(in Value,out Result)"}
   > Set Transform_1_1 = &RigUnit_SetRotation(
    Item=(Type=Bone,Name="dfrm_upperarm_r")
    Space=GlobalSpace
    bInitial=False
    Value=&RigVMFunction_MathQuaternionMul(
     B=&RigUnit_GetTransform(Item=(Type=Bone,Name="dfrm_upperarm_r"), bInitial=False).Transform.Rotation
     A=&RigVMFunction_MathQuaternionFromAxisAndAngle(Angle=0.5, Axis=@n21.Result) {template="FromAxisAndAngle::Execute(in Axis,in Angle,out Result)"}.Result
    ) {template="Multiply::Execute(in A,in B,out Result)"}.Result
    Weight=@n20.OutputParam
   )
   > n22 = &RigUnit_SphericalPoseReader(
    ActiveRegionScaleFactors=(PositiveWidth=0.6,NegativeWidth=0.6,PositiveHeight=1,NegativeHeight=1)
    ActiveRegionSize=0.15
    Debug=(bDrawDebug=false,bDraw2D=false,bDrawLocalAxes=true,DebugScale=25,DebugSegments=20,DebugThickness=0.25)
    DriverAxis=(X=1,Y=0,Z=0)
    DriverItem=(Type=Bone,Name="upperarm_l")
    FalloffRegionScaleFactors=(PositiveWidth=0.6,NegativeWidth=0.6,PositiveHeight=1,NegativeHeight=1)
    FalloffSize=0.45
    OptionalParentItem=(Type=Bone,Name="spine_04")
    RotationOffset=(X=40,Y=25,Z=-40)
   )
   n23 = &RigVMFunction_MathVectorUnit(Value=&RigVMFunction_MathVectorSub(
     A=&RigUnit_GetTransform(Item=(Type=Bone,Name="dfrm_hand_l"), bInitial=False).Transform.Translation
     B=&RigUnit_GetTransform(Item=(Type=Bone,Name="dfrm_upperarm_l"), bInitial=False).Transform.Translation
    ) {template="Subtract::Execute(in A,in B,out Result)"}.Result) {template="Unit::Execute(in Value,out Result)"}
   > Set Transform_1_1_1 = &RigUnit_SetRotation(
    Item=(Type=Bone,Name="dfrm_upperarm_twist_01_l")
    Space=GlobalSpace
    bInitial=False
    Value=&RigVMFunction_MathQuaternionMul(
     B=&RigUnit_GetTransform(Item=(Type=Bone,Name="dfrm_upperarm_l"), bInitial=False).Transform.Rotation
     A=&RigVMFunction_MathQuaternionFromAxisAndAngle(Angle=-0.4, Axis=@n23.Result) {template="FromAxisAndAngle::Execute(in Axis,in Angle,out Result)"}.Result
    ) {template="Multiply::Execute(in A,in B,out Result)"}.Result
    Weight=@n22.OutputParam
   )
   n24 = &RigVMFunction_MathQuaternionSwingTwist(
    TwistAxis=(X=1,Y=0,Z=0)
    Input=&RigVMFunction_MathTransformMakeRelative(
     Parent=&RigUnit_GetTransform(Item=(Type=Bone,Name="hand_l"), Space=LocalSpace, bInitial=true).Transform
     Global=&RigUnit_GetTransform(Item=(Type=Bone,Name="hand_l"), Space=LocalSpace).Transform
    ).Local.Rotation
   )
   &RigUnit_SetRotation(
    Item=(Type=Bone,Name="dfrm_lowerarm_twist_01_l")
    Weight=0.75
    bPropagateToChildren=false
    Value=@n24.Twist
    ExecutePin=@SequenceAggregate_1_2.E
   )
   > &RigUnit_SetRotation(Item=(Type=Bone,Name="dfrm_lowerarm_twist_02_l"), Weight=0.5, bPropagateToChildren=false, Value=@n24.Twist)
   n25 = &RigVMFunction_MathQuaternionSwingTwist(
    TwistAxis=(X=1,Y=0,Z=0)
    Input=&RigVMFunction_MathTransformMakeRelative(
     Parent=&RigUnit_GetTransform(Item=(Type=Bone,Name="hand_r"), Space=LocalSpace, bInitial=true).Transform
     Global=&RigUnit_GetTransform(Item=(Type=Bone,Name="hand_r"), Space=LocalSpace).Transform
    ).Local.Rotation
   )
   > &RigUnit_SetRotation(Item=(Type=Bone,Name="dfrm_lowerarm_twist_01_r"), Weight=0.75, bPropagateToChildren=false, Value=@n25.Twist)
   > &RigUnit_SetRotation(Item=(Type=Bone,Name="dfrm_lowerarm_twist_02_r"), Weight=0.5, bPropagateToChildren=false, Value=@n25.Twist)
   RigVMFunction_Sequence_1 = RigVMAggregateNode(ExecuteContext=@SequenceAggregate_1_2.F)
   &RigUnit_ModifyTransforms(
    ItemToModify="((Item=(Type=Bone,Name=\"dfrm_stomach_01_r\"),Transform=(Rotation=(X=0,Y=0,Z=0,W=1),Translation=(X=-12,Y=0,Z=0),Scale3D=(X=1,Y=1,Z=1))),(Item=(Type=Bone,Name=\"dfrm_stomach_02_r\"),Transform=(Rotation=(X=0,Y=0,Z=0,W=1),Translation=(X=-10,Y=0,Z=0),Scale3D=(X=1, ...(truncated)"
    WeightMinimum=0
    Weight=&RigVMFunction_MathDoubleMul(
     A=RigVMVariableNode() {variable=SPR_upperarm_r_down_B_alpha}.Value
     B=RigVMVariableNode() {variable=mult_spr_dst_hand_r_thigh_twist_r_alpha}.Value
    ).Result
    ExecutePin=@RigVMFunction_Sequence_1.A
   )
   n26 = &RigVMFunction_MathDoubleRemap(SourceMaximum=1, SourceMinimum=0, TargetMaximum=0, TargetMinimum=1, Value=RigVMVariableNode() {variable=emote}.Value)
   n27 = &RigVMFunction_AlphaInterp(
    InterpSpeedDecreasing=10
    InterpSpeedIncreasing=10
    bClampResult=true
    bInterpResult=False
    bMapRange=true
    Scale=@n26.Result
    Value=RigVMVariableNode() {variable=mult_spr_dst_lowerarm_r_pelvis_alpha}.Value
   )
   > &RigUnit_ModifyTransforms(
    ItemToModify="((Item=(Type=Bone,Name=\"dfrm_stomach_01_l\"),Transform=(Rotation=(X=0,Y=0,Z=0,W=1),Translation=(X=-10,Y=0,Z=0),Scale3D=(X=0.75,Y=1,Z=1))),(Item=(Type=Bone,Name=\"dfrm_stomach_03_l\"),Transform=(Rotation=(X=0,Y=0,Z=0,W=1),Translation=(X=-7.5,Y=6.5,Z=-3),Scale3 ...(truncated)"
    WeightMinimum=0
    Weight=@n27.Result
   )
   > n28 = &RigUnit_ModifyTransforms(
    ItemToModify=((Item=(Type=Bone,Name="dfrm_stomach_02_l"),Transform=(Rotation=(X=0,Y=0,Z=0,W=1),Translation=(X=-6,Y=0,Z=0),Scale3D=(X=0.85,Y=1,Z=1))))
    WeightMinimum=0
    Weight=@n27.Result
   )
   DISPATCH_RigVMDispatch_CoreEquals = RigVMDispatchNode(B=-0.1, A=RigVMVariableNode() {variable=pistol}.Value) {template="DISPATCH_RigVMDispatch_CoreEquals(in A,in B,out Result)"}
   RigVMFunction_ControlFlowBranch = &RigVMFunction_ControlFlowBranch(Condition=@DISPATCH_RigVMDispatch_CoreEquals.Result, ExecuteContext=@n28.ExecutePin)
   n29 = RigVMVariableNode() {variable=mult_spr_dst_lowerarm_r_pelvis_alpha}
   n30 = RigVMVariableNode() {variable=SPR_upperarm_l_down_A_alpha}
   n31 = &RigVMFunction_AlphaInterp(
    InterpSpeedDecreasing=10
    InterpSpeedIncreasing=10
    bClampResult=true
    bInterpResult=False
    bMapRange=true
    Value=&RigVMFunction_MathDoubleMax(A=&RigVMFunction_MathDoubleMul(B=@n29.Value, A=@n30.Value).Result, B=RigVMVariableNode() {variable=R handClose to pelvis}.Value).Result
    Scale=@n26.Result
   )
   &RigUnit_ModifyTransforms(
    ItemToModify="((Item=(Type=Bone,Name=\"dfrm_chest_01_r\"),Transform=(Rotation=(X=0,Y=0,Z=0,W=1),Translation=(X=-3,Y=0,Z=5),Scale3D=(X=1,Y=1,Z=1))),(Item=(Type=Bone,Name=\"dfrm_stomach_01_l\"),Transform=(Rotation=(X=0,Y=0,Z=0,W=1),Translation=(X=-5,Y=0,Z=0),Scale3D=(X=1,Y=1, ...(truncated)"
    WeightMinimum=0
    Weight=@n31.Result
    ExecutePin=@RigVMFunction_ControlFlowBranch.False
   )
   n32 = &RigVMFunction_AlphaInterp(
    InterpSpeedDecreasing=10
    InterpSpeedIncreasing=10
    bClampResult=true
    bInterpResult=False
    bMapRange=true
    Scale=@n26.Result
    Value=RigVMVariableNode() {variable=SPR_upperarm_l_down_A_alpha}.Value
   )
   &RigUnit_ModifyTransforms(
    ItemToModify=((Item=(Type=Bone,Name="dfrm_stomach_01_l"),Transform=(Rotation=(X=0,Y=0,Z=0,W=1),Translation=(X=-5,Y=0,Z=0),Scale3D=(X=1,Y=1,Z=1))))
    WeightMinimum=0
    Weight=@n32.Result
    ExecutePin=@RigVMFunction_ControlFlowBranch.True
   )
   Remap = &RigVMFunction_MathDoubleRemap(
    SourceMaximum=15
    SourceMinimum=35
    TargetMaximum=1
    TargetMinimum=0
    Value=&RigVMFunction_MathVectorDistance(
     A=&RigUnit_GetTransform(Item=(Type=Bone,Name="pelvis"), bInitial=False).Transform.Translation
     B=&RigUnit_GetTransform(Item=(Type=Bone,Name="hand_l"), bInitial=False).Transform.Translation
    ).Result
   )
   Multiply = &RigVMFunction_MathDoubleMul(B=@Remap.Result, A=RigVMVariableNode() {variable=SPR_upperarm_l_down_A_alpha}.Value)
   RigVMVariableNode(Value=@Multiply.Result, ExecuteContext=@SequenceAggregate_1_2.B) {variable=mult_spr_dst_lowerarm_r_pelvis_alpha}
   n33 = &RigVMFunction_MathVectorDistance(
    B=&RigUnit_GetTransform(Item=(Type=Bone,Name="hand_r"), bInitial=False).Transform.Translation
    A=&RigUnit_GetTransform(Item=(Type=Bone,Name="pelvis"), bInitial=False).Transform.Translation
   )
   > RigVMVariableNode(Value=&RigVMFunction_MathDoubleMul(
     B=&RigVMFunction_MathDoubleRemap(SourceMaximum=15, SourceMinimum=35, TargetMaximum=1, TargetMinimum=0, Value=@n33.Result).Result
     A=RigVMVariableNode() {variable=SPR_upperarm_r_down_A_alpha}.Value
    ).Result) {variable=R handClose to pelvis}
   RigVMFunction_MathVectorDistance = &RigVMFunction_MathVectorDistance(
    A=&RigUnit_GetTransform(Item=(Type=Bone,Name="hand_r"), bInitial=False).Transform.Translation
    B=&RigUnit_GetTransform(Item=(Type=Bone,Name="thigh_twist_01_r"), bInitial=False).Transform.Translation
   )
   > RigVMVariableNode(Value=&RigVMFunction_MathDoubleSub(
     A=1
     B=&RigVMFunction_MathDoubleRemap(SourceMaximum=59, SourceMinimum=19.5, TargetMaximum=1, TargetMinimum=0, Value=@RigVMFunction_MathVectorDistance.Result).Result
    ).Result) {variable=mult_spr_dst_hand_r_thigh_twist_r_alpha}
   &RigUnit_ParentConstraint(
    AdvancedSettings=(InterpolationType=Average,RotationOrderForFilter=XZY)
    Child=(Type=Bone,Name="dyn_wing_lower_l_MSTR")
    Filter=(TranslationFilter=(bX=true,bY=true,bZ=true),RotationFilter=(bX=True,bY=True,bZ=True),ScaleFilter=(bX=True,bY=True,bZ=True))
    Parents=((Item=(Type=Bone,Name="dfrm_lowerarm_twist_02_l"),Weight=0.5),(Item=(Type=Bone,Name="dfrm_lowerarm_twist_01_l"),Weight=0.5))
    Weight=1
    ExecutePin=@SequenceAggregate_1.D
   )
   > &RigUnit_ParentConstraint(
    AdvancedSettings=(InterpolationType=Average,RotationOrderForFilter=XZY)
    Child=(Type=Bone,Name="dyn_wing_lower_r_MSTR")
    Filter=(TranslationFilter=(bX=true,bY=true,bZ=true),RotationFilter=(bX=True,bY=True,bZ=True),ScaleFilter=(bX=True,bY=True,bZ=True))
    Parents=((Item=(Type=Bone,Name="dfrm_lowerarm_twist_02_r"),Weight=0.5),(Item=(Type=Bone,Name="dfrm_lowerarm_twist_01_r"),Weight=0.5))
    Weight=1
   )
   > &RigUnit_ParentConstraint(
    AdvancedSettings=(InterpolationType=Average,RotationOrderForFilter=XZY)
    Child=(Type=Bone,Name="dyn_wing_upper_l_MSTR")
    Filter=(TranslationFilter=(bX=true,bY=true,bZ=true),RotationFilter=(bX=True,bY=True,bZ=True),ScaleFilter=(bX=True,bY=True,bZ=True))
    Parents=((Item=(Type=Bone,Name="dfrm_upperarm_twist_02_l"),Weight=0.3),(Item=(Type=Bone,Name="dfrm_upperarm_twist_01_l"),Weight=0.7))
    Weight=1
   )
   > &RigUnit_ParentConstraint(
    AdvancedSettings=(InterpolationType=Average,RotationOrderForFilter=XZY)
    Child=(Type=Bone,Name="dyn_wing_upper_r_MSTR")
    Filter=(TranslationFilter=(bX=true,bY=true,bZ=true),RotationFilter=(bX=True,bY=True,bZ=True),ScaleFilter=(bX=True,bY=True,bZ=True))
    Parents=((Item=(Type=Bone,Name="dfrm_upperarm_twist_02_r"),Weight=0.3),(Item=(Type=Bone,Name="dfrm_upperarm_twist_01_r"),Weight=0.7))
    Weight=1
   )
   > &RigUnit_ParentConstraint(
    AdvancedSettings=(InterpolationType=Average,RotationOrderForFilter=XZY)
    Child=(Type=Bone,Name="dyn_wing_upper_l_01")
    Filter=(TranslationFilter=(bX=true,bY=true,bZ=true),RotationFilter=(bX=True,bY=True,bZ=True),ScaleFilter=(bX=True,bY=True,bZ=True))
    Parents=((Item=(Type=Bone,Name="dfrm_upperarm_twist_01_l"),Weight=0.3),(Item=(Type=Bone,Name="dfrm_lowerarm_twist_02_l"),Weight=0.7))
    Weight=1
   )
   > &RigUnit_ParentConstraint(
    AdvancedSettings=(InterpolationType=Average,RotationOrderForFilter=XZY)
    Child=(Type=Bone,Name="dyn_wing_upper_r_01")
    Filter=(TranslationFilter=(bX=true,bY=true,bZ=true),RotationFilter=(bX=True,bY=True,bZ=True),ScaleFilter=(bX=True,bY=True,bZ=True))
    Parents=((Item=(Type=Bone,Name="dfrm_upperarm_twist_01_r"),Weight=0.3),(Item=(Type=Bone,Name="dfrm_lowerarm_twist_02_r"),Weight=0.7))
    Weight=1
   )
   AlphaInterp = &RigVMFunction_AlphaInterp(
    InterpSpeedDecreasing=10
    InterpSpeedIncreasing=10
    Scale=1
    bClampResult=true
    bMapRange=true
    Value=RigVMVariableNode() {variable=sprint}.Value
   )
   > &RigUnit_RotationConstraintLocalSpaceOffset(
    AdvancedSettings=(InterpolationType=Average,RotationOrderForFilter=XZY)
    Child=(Type=Bone,Name="dyn_wing_upper_r_twist_MSTR")
    Filter=(bX=True,bY=True,bZ=True)
    Parents=((Item=(Type=Bone,Name="upperarm_twist_01_r"),Weight=1))
    Weight=@AlphaInterp.Result
   )
   > &RigUnit_RotationConstraintLocalSpaceOffset(
    AdvancedSettings=(InterpolationType=Average,RotationOrderForFilter=XZY)
    Child=(Type=Bone,Name="dyn_wing_upper_l_twist_MSTR")
    Filter=(bX=True,bY=True,bZ=True)
    Parents=((Item=(Type=Bone,Name="upperarm_twist_01_l"),Weight=1))
    Weight=@AlphaInterp.Result
   )
   DeltaRotation v4_1 = RigVMCollapseNode(
    Rotation Order=ZYX
    child=(Type=Bone,Name="spine_04")
    debugRotation=false
    outputMinMax=(X=0,Y=1)
    parent=(Type=Null,Name="spine_05_Spine04Delta_null")
    premult=1
    rotationMinMax=(X=0,Y=30)
    triggerAxis=Z
    twistAxis=(X=1,Y=0,Z=0)
    ExecuteContext=@SequenceAggregate_1.F
   )
   > &RigUnit_ModifyTransforms(
    ItemToModify=((Item=(Type=Bone,Name="attach_backpack"),Transform=(Rotation=(X=-0.2588,Y=0,Z=0,W=0.9659),Translation=(X=0,Y=0,Z=0),Scale3D=(X=1,Y=1,Z=1))))
    WeightMinimum=0
    Weight=@"DeltaRotation v4_1".outputWeight
   )
   n34 = &RigUnit_SphericalPoseReader(
    ActiveRegionScaleFactors=(PositiveWidth=2.5,NegativeWidth=3,PositiveHeight=3,NegativeHeight=3)
    ActiveRegionSize=0.1
    Debug=(bDrawDebug=false,DebugScale=25,DebugSegments=20,DebugThickness=0.25)
    DriverAxis=(X=1,Y=0,Z=0)
    DriverItem=(Type=Bone,Name="upperarm_l")
    FalloffRegionScaleFactors=(PositiveWidth=1,NegativeWidth=1,PositiveHeight=1,NegativeHeight=1)
    FalloffSize=0.5
    OptionalParentItem=(Type=Bone)
    RotationOffset=(X=0,Y=0,Z=0)
    ExecutePin=@SequenceAggregate_1.E
   )
   > &RigUnit_ModifyTransforms(
    ItemToModify="((Item=(Type=Bone,Name=\"dyn_wing_upper_l_twist_MSTR\"),Transform=(Rotation=(X=-0.3827,Y=0,Z=0,W=0.9239),Translation=(X=0,Y=0,Z=0),Scale3D=(X=1,Y=1,Z=1))),(Item=(Type=Bone,Name=\"dyn_wing_upper_l_MSTR\"),Transform=(Rotation=(X=0,Y=-0.1305,Z=0,W=0.9914),Transla ...(truncated)"
    WeightMinimum=0
    Weight=@n34.OutputParam
   )
   > n35 = &RigUnit_SphericalPoseReader(
    ActiveRegionScaleFactors=(PositiveWidth=2.5,NegativeWidth=3,PositiveHeight=3,NegativeHeight=3)
    ActiveRegionSize=0.1
    Debug=(bDrawDebug=false,DebugScale=25,DebugSegments=20,DebugThickness=0.25)
    DriverAxis=(X=-1,Y=0,Z=0)
    DriverItem=(Type=Bone,Name="upperarm_r")
    FalloffRegionScaleFactors=(PositiveWidth=1,NegativeWidth=1,PositiveHeight=1,NegativeHeight=1)
    FalloffSize=0.5
    FlipHeightScaling=true
    OptionalParentItem=(Type=Bone)
    RotationOffset=(X=180,Y=0,Z=180)
   )
   > &RigUnit_ModifyTransforms(
    ItemToModify="((Item=(Type=Bone,Name=\"dyn_wing_upper_r_twist_MSTR\"),Transform=(Rotation=(X=-0.3827,Y=0,Z=0,W=0.9239),Translation=(X=0,Y=0,Z=0),Scale3D=(X=1,Y=1,Z=1))),(Item=(Type=Bone,Name=\"dyn_wing_upper_r_MSTR\"),Transform=(Rotation=(X=0,Y=-0.1305,Z=0,W=0.9914),Transla ...(truncated)"
    WeightMinimum=0
    Weight=@n35.OutputParam
   )
   > n36 = &RigUnit_SphericalPoseReader(
    ActiveRegionScaleFactors=(PositiveWidth=1.5,NegativeWidth=1,PositiveHeight=2,NegativeHeight=2)
    ActiveRegionSize=0.08
    Debug=(bDrawDebug=false,bDraw2D=false,bDrawLocalAxes=true,DebugScale=25,DebugSegments=20,DebugThickness=0.25)
    DriverAxis=(X=-1,Y=0,Z=0)
    DriverItem=(Type=Bone,Name="upperarm_r")
    FalloffRegionScaleFactors=(PositiveWidth=1,NegativeWidth=1,PositiveHeight=2,NegativeHeight=1)
    FalloffSize=0.45
    FlipHeightScaling=false
    FlipWidthScaling=true
    OptionalParentItem=(Type=Bone,Name="spine_04")
    RotationOffset=(X=-80,Y=40,Z=25)
   )
   > &RigUnit_ModifyTransforms(
    ItemToModify="((Item=(Type=Bone,Name=\"dyn_wing_upper_r_twist_MSTR\"),Transform=(Rotation=(X=-0.1305,Y=0,Z=0,W=0.9914),Translation=(X=0,Y=0,Z=0),Scale3D=(X=1,Y=1,Z=1))),(Item=(Type=Bone,Name=\"dyn_wing_upper_r_MSTR\"),Transform=(Rotation=(X=0,Y=0,Z=0,W=1),Translation=(X=0,Y ...(truncated)"
    WeightMinimum=0
    Weight=@n36.OutputParam
   )
   > n37 = &RigUnit_SphericalPoseReader(
    ActiveRegionScaleFactors=(PositiveWidth=0.5,NegativeWidth=1,PositiveHeight=1,NegativeHeight=2)
    ActiveRegionSize=0.08
    Debug=(bDrawDebug=false,bDraw2D=false,bDrawLocalAxes=true,DebugScale=25,DebugSegments=20,DebugThickness=0.25)
    DriverAxis=(X=1,Y=0,Z=0)
    DriverItem=(Type=Bone,Name="upperarm_l")
    FalloffRegionScaleFactors=(PositiveWidth=0.5,NegativeWidth=1,PositiveHeight=1,NegativeHeight=1)
    FalloffSize=0.45
    OptionalParentItem=(Type=Bone,Name="spine_04")
    RotationOffset=(X=80,Y=30,Z=-25)
   )
   > &RigUnit_ModifyTransforms(
    ItemToModify="((Item=(Type=Bone,Name=\"dyn_wing_upper_l_twist_MSTR\"),Transform=(Rotation=(X=-0.1305,Y=0,Z=0,W=0.9914),Translation=(X=0,Y=0,Z=0),Scale3D=(X=1,Y=1,Z=1))),(Item=(Type=Bone,Name=\"dyn_wing_upper_l_MSTR\"),Transform=(Rotation=(X=0,Y=-0.1305,Z=0,W=0.9914),Transla ...(truncated)"
    WeightMinimum=0
    Weight=@n37.OutputParam
   )
   DeltaRotation v4 = RigVMCollapseNode(
    ExecuteContext=()
    Rotation Order=ZYX
    child=(Type=Bone,Name="None")
    outputMinMax=(X=0,Y=1)
    parent=(Type=Bone,Name="None")
    premult=1
    rotationMinMax=(X=0,Y=0)
    twistAxis=(X=1,Y=0,Z=0)
   )
   &RigVMFunction_MathDoubleSub(A=1, B=@Remap.Result)
   RigVMVariableNode() {variable=SPR_upperarm_r_down_B_alpha}
   RigVMVariableNode() {variable=SPR_upperarm_r_down_B_alpha}
  ]}
 SequenceAggregate_1_1_ContainedGraph {nodes [
   n1 = RigVMFunctionEntryNode()
   SequenceAggregate_1 = &RigVMFunction_Sequence(ExecuteContext=@n1.ExecuteContext)
   SequenceAggregate_1_1 = &RigVMFunction_Sequence(ExecuteContext=@SequenceAggregate_1.B)
   SequenceAggregate_1_2 = &RigVMFunction_Sequence(ExecuteContext=@SequenceAggregate_1_1.B)
   SequenceAggregate_1_3 = &RigVMFunction_Sequence(ExecuteContext=@SequenceAggregate_1_2.B)
   SequenceAggregate_2 = &RigVMFunction_Sequence(ExecuteContext=@SequenceAggregate_1_3.B)
   RigVMFunctionReturnNode(
    A=@SequenceAggregate_1.A
    B=@SequenceAggregate_1_1.A
    C=@SequenceAggregate_1_2.A
    D=@SequenceAggregate_1_3.A
    E=@SequenceAggregate_2.A
    F=@SequenceAggregate_2.B
   )
  ]}
 SequenceAggregate_1_1_ContainedGraph {nodes [
   n1 = RigVMFunctionEntryNode()
   SequenceAggregate_1 = &RigVMFunction_Sequence(ExecuteContext=@n1.ExecuteContext)
   SequenceAggregate_1_1 = &RigVMFunction_Sequence(ExecuteContext=@SequenceAggregate_1.B)
   SequenceAggregate_1_2 = &RigVMFunction_Sequence(ExecuteContext=@SequenceAggregate_1_1.B)
   SequenceAggregate_1_3 = &RigVMFunction_Sequence(ExecuteContext=@SequenceAggregate_1_2.B)
   SequenceAggregate_1_4 = &RigVMFunction_Sequence(ExecuteContext=@SequenceAggregate_1_3.B)
   RigVMFunctionReturnNode(
    A=@SequenceAggregate_1.A
    B=@SequenceAggregate_1_1.A
    C=@SequenceAggregate_1_2.A
    D=@SequenceAggregate_1_3.A
    E=@SequenceAggregate_1_4.A
    F=@SequenceAggregate_1_4.B
   )
  ]}
]
hierarchy [
 root {loc=<0 0 0>, rot=<0 0 0 1>}
 pelvis {parent=root, loc=<0 -0.5301 98.0571>, rot=<0.0127 0.707 -0.0127 -0.707>}
 spine_01 {parent=pelvis, loc=<6.1208 -0.9328 0>, rot=<0 0 0.034 -0.9994>}
 spine_02 {parent=spine_01, loc=<6.5996 -0.2095 0>, rot=<0 0 0.0238 -0.9997>}
 spine_03 {parent=spine_02, loc=<6.4914 -0.1237 0>, rot=<0 0 0.0149 -0.9999>}
 spine_04 {parent=spine_03, loc=<6.9598 0.0794 0>, rot=<0 0 -0.0132 -0.9999>}
 spine_05 {parent=spine_04, loc=<6.9987 0 0>, rot=<0 0 -0.0015 -1>}
 attach_cape {parent=spine_05, loc=<9.0184 -11.7141 -0.0001>, rot=<0.0616 -0.7044 -0.0616 -0.7044>}
 neck_01 {parent=spine_05, loc=<11.3138 0.1529 0>, rot=<0 0 -0.2003 -0.9797>}
 neck_02 {parent=neck_01, loc=<4.5627 0 0>, rot=<0 0 0.0504 -0.9987>}
 head {parent=neck_02, loc=<5.3914 0 0>, rot=<0 0 0.115 -0.9934>}
 attach_backpack {parent=spine_05, loc=<-0.1213 -13.4731 -0.0001>, rot=<-0.0093 -0.707 0.0093 -0.707>}
 clavicle_l {parent=spine_05, loc=<5.4831 -0.1053 -3.2946>, rot=<0.0465 -0.7289 -0.009 -0.6829>}
 upperarm_l {parent=clavicle_l, loc=<9.8658 -1.2136 -0.4105>, rot=<0.0164 -0.3245 -0.0269 -0.9454>}
 lowerarm_l {parent=upperarm_l, loc=<26.465 -0.8345 0.259>, rot=<0.0036 0.0161 -0.2298 -0.9731>}
 hand_l {parent=lowerarm_l, loc=<22.7355 -0.3203 -0.6105>, rot=<-0.6784 -0.0223 0.0735 -0.7307>}
 thumb_01_l {parent=hand_l, loc=<1.3118 -0.3191 -2.459>, rot=<0.6546 -0.4042 0.0664 -0.6354>}
 thumb_02_l {parent=thumb_01_l, loc=<5.9336 -0.0448 0.0401>, rot=<-0.0037 0.0213 0.298 -0.9543>}
 thumb_03_l {parent=thumb_02_l, loc=<2.2223 0.3682 -0.0907>, rot=<0.0118 -0.013 0.0066 -0.9998>}
 index_metacarpal_l {parent=hand_l, loc=<1.6535 0.0183 -2.96>, rot=<0.0158 -0.0124 -0.0173 -0.9996>}
 index_01_l {parent=index_metacarpal_l, loc=<8.3008 0.0357 -0.0524>, rot=<-0.0044 0.0485 0.0924 -0.9945>}
 index_02_l {parent=index_01_l, loc=<3.3587 -0.3004 0.0452>, rot=<-0.002 0.0027 0.3225 -0.9466>}
 index_03_l {parent=index_02_l, loc=<2.5578 -0.123 -0.0252>, rot=<-0.002 0.0067 0.2558 -0.9667>}
 middle_metacarpal_l {parent=hand_l, loc=<2.1856 -0.1725 -0.5984>, rot=<0.006 0.0168 -0.0329 -0.9993>}
 middle_01_l {parent=middle_metacarpal_l, loc=<7.3275 -0.2447 0.0129>, rot=<-0.0208 0.0042 0.0858 -0.9961>}
 middle_02_l {parent=middle_01_l, loc=<4.0123 0.1401 0.1175>, rot=<-0.0218 0.0258 0.3512 -0.9357>}
 middle_03_l {parent=middle_02_l, loc=<2.6077 0.0534 0.1373>, rot=<0.0082 0.0188 0.1708 -0.9851>}
 ring_metacarpal_l {parent=hand_l, loc=<2.2074 -0.3625 1.3638>, rot=<-0.0284 0.0722 -0.0458 -0.9959>}
 ring_01_l {parent=ring_metacarpal_l, loc=<7.0295 -0.6197 -0.1446>, rot=<-0.0489 -0.0252 0.1892 -0.9804>}
 ring_02_l {parent=ring_01_l, loc=<3.6174 0.4647 -0.1027>, rot=<0 0 0.2806 -0.9598>}
 ring_03_l {parent=ring_02_l, loc=<2.4152 0.114 0.0199>, rot=<0 0 0.1189 -0.9929>}
 pinky_metacarpal_l {parent=hand_l, loc=<1.7616 -0.9753 3.3814>, rot=<-0.0397 0.0814 -0.069 -0.9935>}
 pinky_01_l {parent=pinky_metacarpal_l, loc=<7.1709 -0.4831 -0.1094>, rot=<-0.0654 -0.0035 0.2242 -0.9723>}
 pinky_02_l {parent=pinky_01_l, loc=<2.8274 0.0893 -0.1043>, rot=<0.0037 0.0011 0.2854 -0.9584>}
 pinky_03_l {parent=pinky_02_l, loc=<2.2046 0.0307 -0.0078>, rot=<-0.0002 0.0098 0.1445 -0.9895>}
 weapon_l {parent=hand_l, loc=<-0.1974 -3.4111 0.3811>, rot=<0 0 0 1>}
 hand_attach_l {parent=hand_l, loc=<8.5 -3 0>, rot=<0.7071 0.7071 0 0>}
 lowerarm_twist_02_l {parent=lowerarm_l, loc=<10.6707 0 0>, rot=<0 0 0 1>}
 lowerarm_twist_01_l {parent=lowerarm_l, loc=<18.316 0 0>, rot=<0 0 0 1>}
 upperarm_twist_01_l {parent=upperarm_l, loc=<7.2135 -0.1134 0>, rot=<0 0 0 1>}
 upperarm_twist_02_l {parent=upperarm_l, loc=<19.7998 -0.1134 0>, rot=<0 0 0 1>}
 clavicle_r {parent=spine_05, loc=<5.4831 -0.1053 3.2946>, rot=<0.7289 0.0465 -0.6829 0.009>}
 upperarm_r {parent=clavicle_r, loc=<-9.8658 1.2136 0.4105>, rot=<0.0164 -0.3245 -0.0269 -0.9454>}
 lowerarm_r {parent=upperarm_r, loc=<-26.465 0.8345 -0.259>, rot=<0.0036 0.0161 -0.2298 -0.9731>}
 hand_r {parent=lowerarm_r, loc=<-22.7355 0.3203 0.6106>, rot=<-0.6784 -0.0223 0.0735 -0.7307>}
 thumb_01_r {parent=hand_r, loc=<-1.3118 0.3191 2.459>, rot=<0.6546 -0.4042 0.0664 -0.6354>}
 thumb_02_r {parent=thumb_01_r, loc=<-5.9336 0.0448 -0.0401>, rot=<-0.0037 0.0213 0.298 -0.9543>}
 thumb_03_r {parent=thumb_02_r, loc=<-2.2223 -0.3682 0.0907>, rot=<0.0118 -0.013 0.0066 -0.9998>}
 index_metacarpal_r {parent=hand_r, loc=<-1.6535 -0.0183 2.96>, rot=<0.0158 -0.0124 -0.0173 -0.9996>}
 index_01_r {parent=index_metacarpal_r, loc=<-8.3008 -0.0357 0.0524>, rot=<-0.0044 0.0485 0.0924 -0.9945>}
 index_02_r {parent=index_01_r, loc=<-3.3587 0.3004 -0.0452>, rot=<-0.002 0.0027 0.3225 -0.9466>}
 index_03_r {parent=index_02_r, loc=<-2.5578 0.123 0.0252>, rot=<-0.002 0.0067 0.2558 -0.9667>}
 middle_metacarpal_r {parent=hand_r, loc=<-2.1856 0.1725 0.5984>, rot=<0.006 0.0168 -0.0329 -0.9993>}
 middle_01_r {parent=middle_metacarpal_r, loc=<-7.3275 0.2447 -0.0129>, rot=<-0.0208 0.0042 0.0858 -0.9961>}
 middle_02_r {parent=middle_01_r, loc=<-4.0123 -0.1401 -0.1175>, rot=<-0.0218 0.0258 0.3512 -0.9357>}
 middle_03_r {parent=middle_02_r, loc=<-2.6077 -0.0534 -0.1373>, rot=<0.0082 0.0188 0.1708 -0.9851>}
 ring_metacarpal_r {parent=hand_r, loc=<-2.2074 0.3625 -1.3638>, rot=<-0.0284 0.0722 -0.0458 -0.9959>}
 ring_01_r {parent=ring_metacarpal_r, loc=<-7.0295 0.6197 0.1446>, rot=<-0.0489 -0.0252 0.1892 -0.9804>}
 ring_02_r {parent=ring_01_r, loc=<-3.6174 -0.4647 0.1027>, rot=<0 0 0.2806 -0.9598>}
 ring_03_r {parent=ring_02_r, loc=<-2.4152 -0.114 -0.0199>, rot=<0 0 0.1189 -0.9929>}
 pinky_metacarpal_r {parent=hand_r, loc=<-1.7616 0.9753 -3.3814>, rot=<-0.0397 0.0814 -0.069 -0.9935>}
 pinky_01_r {parent=pinky_metacarpal_r, loc=<-7.1709 0.4831 0.1094>, rot=<-0.0654 -0.0035 0.2242 -0.9723>}
 pinky_02_r {parent=pinky_01_r, loc=<-2.8274 -0.0893 0.1043>, rot=<0.0037 0.0011 0.2854 -0.9584>}
 pinky_03_r {parent=pinky_02_r, loc=<-2.2046 -0.0307 0.0078>, rot=<-0.0002 0.0098 0.1445 -0.9895>}
 weapon_r {parent=hand_r, loc=<0.1974 3.4111 -0.3811>, rot=<0 0 0 1>}
 hand_attach_r {parent=hand_r, loc=<-8.5001 2.9999 0>, rot=<0 0 -0.7071 -0.7071>}
 lowerarm_twist_02_r {parent=lowerarm_r, loc=<-10.6707 0 0>, rot=<0 0 0 -1>}
 lowerarm_twist_01_r {parent=lowerarm_r, loc=<-18.316 0 0>, rot=<0 0 0 -1>}
 upperarm_twist_01_r {parent=upperarm_r, loc=<-7.2135 0.1134 0>, rot=<0 0 0 -1>}
 upperarm_twist_02_r {parent=upperarm_r, loc=<-19.7998 0.1134 0>, rot=<0 0 0 -1>}
 dfrm_clavicle_l {parent=spine_05, loc=<5.4831 -0.1053 -3.2946>, rot=<0.0465 -0.7289 -0.009 -0.6829>}
 dfrm_upperarm_l {parent=dfrm_clavicle_l, loc=<9.8658 -1.2136 -0.4105>, rot=<0.0164 -0.3245 -0.0269 -0.9454>}
 dfrm_lowerarm_l {parent=dfrm_upperarm_l, loc=<26.465 -0.8345 0.259>, rot=<0.0036 0.0161 -0.2298 -0.9731>}
 dfrm_lowerarm_twist_02_l {parent=dfrm_lowerarm_l, loc=<10.6707 0 0>, rot=<0 0 0 1>}
 dyn_wing_lower_l_twist_MSTR {parent=dfrm_lowerarm_twist_02_l, loc=<0 0 0>, rot=<0 0 0 1>}
 dyn_wing_lower_l_MSTR {parent=dyn_wing_lower_l_twist_MSTR, loc=<-7.658 -2.048 -0.4383>, rot=<0.0097 -0.1651 0.9668 0.1948>}
 dyn_wing_upper_l_04 {parent=dyn_wing_lower_l_MSTR, loc=<-1.1461 3.2663 -0.3853>, rot=<0.0433 -0.0165 -0.0185 -0.9988>}
 dyn_wing_upper_l_04_end {parent=dyn_wing_upper_l_04, loc=<-9.3847 0.0001 -0.0002>, rot=<0.0023 0.0263 0.0172 -0.9995>}
 dyn_wing_upper_l_05 {parent=dyn_wing_lower_l_MSTR, loc=<-10.9451 -1.8257 -0.4175>, rot=<0 0 -0.039 -0.9992>}
 dyn_wing_upper_l_05_end {parent=dyn_wing_upper_l_05, loc=<-8.0176 0 0>, rot=<-0.004 -0.0349 0.0426 -0.9985>}
 dfrm_lowerarm_twist_01_l {parent=dfrm_lowerarm_l, loc=<18.316 0 0>, rot=<0 0 0 1>}
 dfrm_hand_l {parent=dfrm_lowerarm_l, loc=<22.7355 -0.3203 -0.6105>, rot=<-0.6784 -0.0223 0.0735 -0.7307>}
 dfrm_upperarm_twist_01_l {parent=dfrm_upperarm_l, loc=<7.2135 -0.1134 0>, rot=<0.0015 0 0 -1>}
 dyn_wing_upper_l_twist_MSTR {parent=dfrm_upperarm_twist_01_l, loc=<0 0 0>, rot=<0 0 0 1>}
 dyn_wing_upper_l_MSTR {parent=dyn_wing_upper_l_twist_MSTR, loc=<2.858 -3.3049 -3.137>, rot=<-0.0599 -0.3671 0.8988 0.2321>}
 dyn_wing_upper_l_01 {parent=dyn_wing_upper_l_MSTR, loc=<-13.8175 -2.74 -1.3027>, rot=<-0.0902 0.1199 -0.1208 -0.9813>}
 dyn_wing_upper_l_01_end {parent=dyn_wing_upper_l_01, loc=<-6.8789 0 -0.0001>, rot=<0.0079 -0.0534 0.0092 -0.9985>}
 dyn_wing_upper_l_02 {parent=dyn_wing_upper_l_MSTR, loc=<-7.1189 1.3474 -0.9037>, rot=<0 0 0 1>}
 dyn_wing_upper_l_02_end {parent=dyn_wing_upper_l_02, loc=<-6.8685 0 0>, rot=<-0.0087 0.0178 -0.0208 -0.9996>}
 dyn_wing_upper_l_03 {parent=dyn_wing_upper_l_MSTR, loc=<-1.5405 4.0704 -0.5145>, rot=<0.0076 0.0232 0.1055 -0.9941>}
 dyn_wing_upper_l_03_end {parent=dyn_wing_upper_l_03, loc=<-6.0908 0 0>, rot=<-0.0222 0.0342 -0.0093 -0.9991>}
 dfrm_upperarm_twist_02_l {parent=dfrm_upperarm_l, loc=<19.7998 -0.1134 0>, rot=<0 0 0 1>}
 dfrm_clavicle_r {parent=spine_05, loc=<5.4831 -0.1053 3.2946>, rot=<0.7289 0.0465 -0.6829 0.009>}
 dfrm_upperarm_r {parent=dfrm_clavicle_r, loc=<-9.8658 1.2136 0.4105>, rot=<0.0164 -0.3245 -0.0269 -0.9454>}
 dfrm_lowerarm_r {parent=dfrm_upperarm_r, loc=<-26.465 0.8345 -0.259>, rot=<0.0036 0.0161 -0.2298 -0.9731>}
 dfrm_lowerarm_twist_02_r {parent=dfrm_lowerarm_r, loc=<-10.6707 0 0>, rot=<0 0 0 -1>}
 dyn_wing_lower_r_twist_MSTR {parent=dfrm_lowerarm_twist_02_r, loc=<0 0 0>, rot=<0 0 0 1>}
 dyn_wing_lower_r_MSTR {parent=dyn_wing_lower_r_twist_MSTR, loc=<7.6577 2.0482 0.4387>, rot=<0.0097 -0.1651 0.9668 0.1948>}
 dyn_wing_upper_r_04 {parent=dyn_wing_lower_r_MSTR, loc=<1.1459 -3.2663 0.3851>, rot=<0.0433 -0.0165 -0.0185 -0.9988>}
 dyn_wing_upper_r_04_end {parent=dyn_wing_upper_r_04, loc=<9.3845 0 0>, rot=<0.0023 0.0263 0.0172 -0.9995>}
 dyn_wing_upper_r_05 {parent=dyn_wing_lower_r_MSTR, loc=<10.9446 1.8259 0.417>, rot=<0 0 -0.039 -0.9992>}
 dyn_wing_upper_r_05_end {parent=dyn_wing_upper_r_05, loc=<8.0175 0 0>, rot=<-0.004 -0.0349 0.0426 -0.9985>}
 dfrm_lowerarm_twist_01_r {parent=dfrm_lowerarm_r, loc=<-18.316 0 0>, rot=<0 0 0 -1>}
 dfrm_hand_r {parent=dfrm_lowerarm_r, loc=<-22.7355 0.3203 0.6106>, rot=<-0.6784 -0.0223 0.0735 -0.7307>}
 dfrm_upperarm_twist_01_r {parent=dfrm_upperarm_r, loc=<-7.2135 0.1134 0>, rot=<0 0 0 -1>}
 dyn_wing_upper_r_twist_MSTR {parent=dfrm_upperarm_twist_01_r, loc=<0 0 0>, rot=<0 0 0 1>}
 dyn_wing_upper_r_MSTR {parent=dyn_wing_upper_r_twist_MSTR, loc=<-2.8583 3.3049 3.1373>, rot=<-0.0599 -0.3671 0.8988 0.2321>}
 dyn_wing_upper_r_01 {parent=dyn_wing_upper_r_MSTR, loc=<13.8171 2.7401 1.3022>, rot=<-0.0902 0.1199 -0.1208 -0.9813>}
 dyn_wing_upper_r_01_end {parent=dyn_wing_upper_r_01, loc=<6.8788 0 0>, rot=<0.0079 -0.0534 0.0092 -0.9985>}
 dyn_wing_upper_r_02 {parent=dyn_wing_upper_r_MSTR, loc=<7.1186 -1.3473 0.9034>, rot=<0 0 0 1>}
 dyn_wing_upper_r_02_end {parent=dyn_wing_upper_r_02, loc=<6.8685 0 0>, rot=<-0.0087 0.0178 -0.0208 -0.9996>}
 dyn_wing_upper_r_03 {parent=dyn_wing_upper_r_MSTR, loc=<1.5403 -4.0703 0.5142>, rot=<0.0076 0.0232 0.1055 -0.9941>}
 dyn_wing_upper_r_03_end {parent=dyn_wing_upper_r_03, loc=<6.0908 0 0>, rot=<-0.0222 0.0342 -0.0093 -0.9991>}
 dfrm_upperarm_twist_02_r {parent=dfrm_upperarm_r, loc=<-19.7998 0.1134 0>, rot=<0 0 0 -1>}
 dyn_necklace_MSTR {parent=spine_05, loc=<10.0018 7.517 -0.1601>, rot=<0.0284 -0.7065 -0.0284 -0.7065>}
 dyn_necklace_01 {parent=dyn_necklace_MSTR, loc=<0.1111 3.9372 -2.644>, rot=<-0.3254 -0.6108 -0.3394 -0.637>}
 dyn_necklace_end {parent=dyn_necklace_01, loc=<8.9981 0 0>, rot=<0.9436 -0.33 0.0221 0.0166>}
 dyn_dogtags_MSTR {parent=dyn_necklace_end, loc=<0 0 0>, rot=<0 0 0 1>}
 dyn_dogtag_01 {parent=dyn_dogtags_MSTR, loc=<3.183 -0.1819 -0.0265>, rot=<0.0099 -0.0327 -0.013 -0.9993>}
 dyn_dogtag_02 {parent=dyn_dogtags_MSTR, loc=<3.1401 0.3128 -0.0264>, rot=<-0.0394 0.1336 -0.019 -0.9901>}
 dyn_necklace_r_01 {parent=dyn_necklace_MSTR, loc=<-10.2414 0.4135 -0.3989>, rot=<0.7238 0.3592 -0.4695 -0.3559>}
 dyn_necklace_r_02 {parent=dyn_necklace_r_01, loc=<5.5772 0 0>, rot=<-0.004 -0.0667 -0.0828 -0.9943>}
 dyn_necklace_r_03 {parent=dyn_necklace_r_02, loc=<5.6711 0 0>, rot=<0.0635 -0.3223 -0.0969 -0.9395>}
 dyn_necklace_l_01 {parent=dyn_necklace_MSTR, loc=<10.2414 -0.4135 0.3989>, rot=<-0.5403 -0.3868 0.6771 0.3162>}
 dyn_necklace_l_02 {parent=dyn_necklace_l_01, loc=<5.9563 0 0>, rot=<-0.0743 0.1288 -0.0662 -0.9867>}
 dyn_necklace_l_03 {parent=dyn_necklace_l_02, loc=<6.0347 0 0>, rot=<-0.0617 0.2834 -0.0556 -0.9554>}
 dyn_neckalce_back_MSTR {parent=dyn_necklace_MSTR, loc=<-0.7146 -15.0284 3.4711>, rot=<0 0 0.7071 -0.7071>}
 dyn_neckalce_back_01 {parent=dyn_neckalce_back_MSTR, loc=<0 0 0>, rot=<0 0 0 1>}
 dyn_necklace_r_back_01 {parent=dyn_necklace_MSTR, loc=<-10.8157 -5.0429 1.213>, rot=<0.5473 -0.3132 0.5091 -0.5859>}
 dyn_necklace_r_back_02 {parent=dyn_necklace_r_back_01, loc=<6.3634 0 0>, rot=<0.0105 0.3881 -0.0878 -0.9174>}
 dyn_necklace_l_back_01 {parent=dyn_necklace_MSTR, loc=<10.1846 -5.824 2.6244>, rot=<0.4471 -0.5613 0.6072 -0.3411>}
 dyn_necklace_l_back_02 {parent=dyn_necklace_l_back_01, loc=<5.7747 0 0>, rot=<0.013 -0.3733 -0.0831 -0.9239>}
 spine_05_Spine04Delta_null {parent=spine_05, type=Null, loc=<0 0 0>, rot=<0 0 0 1>}
 dfrm_chest_01_l {parent=spine_04, loc=<0.5425 15.2169 -7.801>, rot=<-0.5908 -0.3886 0.621 0.3381>}
 dfrm_chest_01_r {parent=spine_04, loc=<2.3593 15.1631 6.5797>, rot=<0.3475 0.6186 -0.395 -0.5835>}
 thigh_l {parent=pelvis, loc=<-1.0855 -0.2109 -8.7053>, rot=<0.0009 0.0355 0.0155 -0.9993>}
 calf_l {parent=thigh_l, loc=<-42.5284 -0.0019 -0.7297>, rot=<0.0006 -0.0161 -0.0391 -0.9991>}
 foot_l {parent=calf_l, loc=<-46.7401 -2.5014 -0.592>, rot=<0.0427 -0.0238 0.0353 -0.9982>}
 ball_l {parent=foot_l, loc=<-3.5658 10.8066 -0.1544>, rot=<0.0001 -0.0001 -0.7063 -0.7079>}
 dfrm_foot_aim_l {parent=foot_l, loc=<-4.054 -7.761 -1.1648>, rot=<0 0 0 -1>}
 dfrm_foot_l_ctrl {controlType=EULER_TRANSFORM, parent=foot_l, type=Control, loc=<14.6217 -3.4245 3.2306>, rot=<-0.5229 -0.4735 0.4801 0.5214>}
 calf_twist_02_l {parent=calf_l, loc=<-19.1546 -1.7943 -0.2638>, rot=<0.0237 0.0095 -0.0035 -0.9997>}
 calf_twist_01_l {parent=calf_l, loc=<-35.8483 -1.9197 -0.5794>, rot=<0.0237 0.0095 -0.0035 -0.9997>}
 dfrm_calf_aim_l {parent=calf_l, loc=<-40.7601 -20.4095 -0.4263>, rot=<0.0017 0.0216 0.0001 -0.9998>}
 PV L {controlType=EULER_TRANSFORM, parent=calf_l, type=Control, loc=<0 50 0>, rot=<0 0 0 1>}
 dfrm_calf_l_PV {parent=calf_l, loc=<-5.4607 50.5823 -0.0167>, rot=<0.0017 0.0216 0.0001 -0.9998>}
 thigh_twist_01_l {parent=thigh_l, loc=<-27.952 0 0>, rot=<0.0013 0 0 -1>}
 thigh_r {parent=pelvis, loc=<-1.0855 -0.2109 8.7053>, rot=<0.0355 -0.0009 0.9993 0.0155>}
 calf_r {parent=thigh_r, loc=<42.5284 0.0019 0.7297>, rot=<0.0006 -0.0161 -0.0391 -0.9991>}
 foot_r {parent=calf_r, loc=<46.7401 2.5014 0.592>, rot=<0.0427 -0.0238 0.0353 -0.9982>}
 ball_r {parent=foot_r, loc=<3.5658 -10.8066 0.1544>, rot=<0.0001 -0.0001 -0.7063 -0.7079>}
 dfrm_foot_aim_r {parent=foot_r, loc=<4.054 7.761 1.1648>, rot=<0 0 0 -1>}
 dfrm_foot_r_ctrl {controlType=EULER_TRANSFORM, parent=foot_r, type=Control, loc=<0 0 0>, rot=<0 0 0 1>}
 calf_twist_02_r {parent=calf_r, loc=<19.1546 1.7943 0.2638>, rot=<0.0237 0.0095 -0.0035 -0.9997>}
 calf_twist_01_r {parent=calf_r, loc=<35.8483 1.9197 0.5794>, rot=<0.0237 0.0095 -0.0035 -0.9997>}
 dfrm_calf_aim_r {parent=calf_r, loc=<42.7689 20.0655 0.5145>, rot=<0.0017 0.0216 0.0001 -0.9998>}
 PV R {controlType=EULER_TRANSFORM, parent=calf_r, type=Control, loc=<0 -50 0>, rot=<0 0 0 1>}
 dfrm_calf_r_PV {parent=calf_r, loc=<5.4562 -50.5821 0.0765>, rot=<0.0017 0.0216 0.0001 -0.9998>}
 thigh_twist_01_r {parent=thigh_r, loc=<27.952 0 0>, rot=<0.0013 0 0 -1>}
 dfrm_pelvis {parent=pelvis, loc=<-12.6365 0 0>, rot=<0 0 0 1>}
 dyn_tail_MSTR {parent=dfrm_pelvis, loc=<-12.0283 -8.1087 0>, rot=<0.2668 -0.9638 0 0>}
 dyn_tail_01 {parent=dyn_tail_MSTR, loc=<0 0 0>, rot=<0 0 0 1>}
 dyn_tail_02 {parent=dyn_tail_01, loc=<12.3834 0 0>, rot=<0 0 0 1>}
 dyn_tail_03 {parent=dyn_tail_02, loc=<12.3834 0 0>, rot=<0 0 0 1>}
 dfrm_thigh_r {parent=dfrm_pelvis, loc=<-3.2425 0.2481 8.7122>, rot=<0.0572 -0.0012 0.9982 0.0155>}
 dfrm_calf_r {parent=dfrm_thigh_r, loc=<37.1895 -0.1455 0.06>, rot=<0.0006 -0.0161 -0.0391 -0.9991>}
 dfrm_calf_twist_01_r {parent=dfrm_calf_r, loc=<29.2812 0.7545 -0.3098>, rot=<0.0237 0.0095 -0.0035 -0.9997>}
 dfrm_calf_twist_02_r {parent=dfrm_calf_r, loc=<12.2745 0.7545 -0.3098>, rot=<0.0237 0.0095 -0.0035 -0.9997>}
 dyn_calf_feather_01 {parent=dfrm_calf_twist_02_r, loc=<-2.0783 -2.7413 -0.5344>, rot=<-0.724 -0.0054 0.0675 -0.6865>}
 dyn_calf_feather_02 {parent=dfrm_calf_twist_02_r, loc=<-7.448 -2.9439 -1.5042>, rot=<-0.8183 0.0047 0.0676 -0.5708>}
 dyn_calf_feather_03 {parent=dfrm_calf_twist_02_r, loc=<-2.7347 0.5821 3.7067>, rot=<-0.0446 -0.0508 0.0449 -0.9967>}
 dyn_calf_feather_04 {parent=dfrm_calf_twist_02_r, loc=<-8.4089 0.7852 3.5059>, rot=<-0.0446 -0.0508 0.0449 -0.9967>}
 dyn_calf_feather_05 {parent=dfrm_calf_twist_02_r, loc=<-8.3312 4.92 0.8996>, rot=<0.6751 -0.0676 -0.0043 -0.7346>}
 dyn_calf_feather_06 {parent=dfrm_calf_twist_02_r, loc=<-3.1964 4.7196 2.2151>, rot=<0.4667 -0.0665 0.013 -0.8818>}
 dyn_calf_feather_07 {parent=dfrm_calf_twist_02_r, loc=<-2.5695 2.1758 -3.1154>, rot=<0.9841 -0.0506 -0.045 -0.1639>}
 dyn_calf_feather_08 {parent=dfrm_calf_twist_02_r, loc=<-7.5771 2.7904 -3.1329>, rot=<0.9841 -0.0506 -0.045 -0.1639>}
 deform_knee_in_base_r {parent=dfrm_calf_r, loc=<0.3084 7.9762 -0.8686>, rot=<0 0 0 -1>}
 deform_knee_in_r {parent=deform_knee_in_base_r, loc=<0 0 0>, rot=<0 0 0 1>}
 deform_knee_out_base_r {parent=dfrm_calf_r, loc=<-0.8551 -8.4142 -0.7848>, rot=<0 0 0 -1>}
 deform_knee_out_r {parent=deform_knee_out_base_r, loc=<0 0 0>, rot=<0 0 0 1>}
 dfrm_foot_r {parent=dfrm_calf_r, loc=<42.065 -0.9449 -1.8502>, rot=<-0.0025 -0.0618 -0.6802 -0.7304>}
 dyn_ball_r {parent=dfrm_foot_r, loc=<-5.8402 -1.077 0.421>, rot=<0 0 0 1>}
 dyn_toe_r_c_01 {parent=dyn_ball_r, loc=<0.5132 -0.201 3.0849>, rot=<-0.2023 -0.6858 -0.6661 0.2124>}
 dyn_toe_r_c_02 {parent=dyn_toe_r_c_01, loc=<3.8303 0 0>, rot=<0 -0.0255 0.0482 -0.9985>}
 joint22 {parent=dyn_toe_r_c_02, loc=<5.0692 0 0>, rot=<0 0 0 1>}
 dyn_toe_r_b_01 {parent=dyn_ball_r, loc=<-0.702 -0.157 0.2106>, rot=<0.0098 0.7004 0.7136 0.0112>}
 dyn_toe_r_b_02 {parent=dyn_toe_r_b_01, loc=<5.8662 0 0>, rot=<0 -0.0345 0.0155 -0.9993>}
 joint23 {parent=dyn_toe_r_b_02, loc=<5.7004 0 0>, rot=<0 0 0 1>}
 dyn_toe_r_a_01 {parent=dyn_ball_r, loc=<0.3668 -0.1038 -2.6579>, rot=<-0.156 0.673 0.6981 0.188>}
 dyn_toe_r_a_02 {parent=dyn_toe_r_a_01, loc=<3.64 0 0>, rot=<0 -0.0797 -0.0135 -0.9967>}
 joint24 {parent=dyn_toe_r_a_02, loc=<4.3265 0 0>, rot=<0 0 0 1>}
 dyn_toe_r_d_03 {parent=dfrm_foot_r, loc=<2.6964 0.0647 0.4623>, rot=<0.7071 -0.1005 0.0447 -0.6985>}
 dyn_toe_r_d_01 {parent=dyn_toe_r_d_03, loc=<2.0486 0 0>, rot=<-0.0333 -0.0352 0.0011 -0.9988>}
 dyn_toe_r_d_02 {parent=dyn_toe_r_d_01, loc=<1.8873 0 0>, rot=<0 0.066 0 -0.9978>}
 joint15 {parent=dyn_toe_r_d_02, loc=<3.4635 0 0>, rot=<0 0 0 1>}
 dfrm_thigh_twist_01_r {parent=dfrm_thigh_r, loc=<13.2027 0 0>, rot=<0.0013 0 0 -1>}
 deform_groin_base_r {parent=dfrm_thigh_r, loc=<-1.1526 -11.1928 1.1578>, rot=<0 0 0 -1>}
 deform_groin_r {parent=deform_groin_base_r, loc=<0 0 0>, rot=<0 0 0 -1>}
 dfrm_thigh_l {parent=dfrm_pelvis, loc=<-3.2425 0.2481 -8.7122>, rot=<0.0012 0.0572 0.0155 -0.9982>}
 dfrm_calf_l {parent=dfrm_thigh_l, loc=<-37.1895 0.1455 -0.06>, rot=<0.0006 -0.0161 -0.0391 -0.9991>}
 dfrm_calf_twist_01_l {parent=dfrm_calf_l, loc=<-29.2811 -0.7545 0.3098>, rot=<0.0237 0.0095 -0.0035 -0.9997>}
 dfrm_calf_twist_02_l {parent=dfrm_calf_l, loc=<-12.2744 -0.7545 0.3098>, rot=<0.0237 0.0095 -0.0035 -0.9997>}
 dyn_calf_feather_l_10 {parent=dfrm_calf_twist_02_l, loc=<9.8282 -0.5861 -3.9566>, rot=<-0.0439 -0.9974 -0.0261 0.0516>}
 dyn_calf_feather_l_09 {parent=dfrm_calf_twist_02_l, loc=<3.3487 -0.2994 -3.7679>, rot=<-0.0528 -0.9746 -0.2135 0.0424>}
 dyn_calf_feather_l_08 {parent=dfrm_calf_twist_02_l, loc=<10.0707 -4.1928 -1.4122>, rot=<-0.0042 -0.8134 0.5777 0.0676>}
 dyn_calf_feather_l_07 {parent=dfrm_calf_twist_02_l, loc=<3.903 -4.7153 -0.8825>, rot=<-0.0042 -0.8134 0.5777 0.0676>}
 dyn_calf_feather_l_06 {parent=dfrm_calf_twist_02_l, loc=<9.6695 -3.0726 2.1359>, rot=<0.0396 -0.2639 0.9622 0.055>}
 dyn_calf_feather_l_05 {parent=dfrm_calf_twist_02_l, loc=<3.4028 -2.4962 2.7341>, rot=<0.0396 -0.2639 0.9622 0.055>}
 dyn_calf_feather_l_04 {parent=dfrm_calf_twist_02_l, loc=<9.6734 1.9401 1.9952>, rot=<0.0651 0.3862 0.9199 0.0188>}
 dyn_calf_feather_l_03 {parent=dfrm_calf_twist_02_l, loc=<3.7433 0.9814 3.1347>, rot=<0.0596 0.1812 0.9811 0.0323>}
 dyn_calf_feather_l_02 {parent=dfrm_calf_twist_02_l, loc=<3.5916 2.5427 0.4456>, rot=<0.0675 0.565 0.8223 0.0052>}
 dyn_calf_feather_l_01 {parent=dfrm_calf_twist_02_l, loc=<10.6202 2.431 -1.2337>, rot=<-0.0646 -0.8309 -0.5522 0.0203>}
 deform_knee_in_base_l {parent=dfrm_calf_l, loc=<-0.3084 -7.9762 0.8686>, rot=<0 0 0 -1>}
 deform_knee_in_l {parent=deform_knee_in_base_l, loc=<0 0 0>, rot=<0 0 0 1>}
 deform_knee_out_base_l {parent=dfrm_calf_l, loc=<0.8551 8.4142 0.7849>, rot=<0 0 0 -1>}
 deform_knee_out_l {parent=deform_knee_out_base_l, loc=<0 0 0>, rot=<0 0 0 1>}
 dfrm_foot_l {parent=dfrm_calf_l, loc=<-42.065 0.9449 1.8503>, rot=<-0.0025 -0.0618 -0.6802 -0.7304>}
 dyn_ball_l {parent=dfrm_foot_l, loc=<5.8402 1.077 -0.421>, rot=<0 0 0 1>}
 dyn_toe_l_c_01 {parent=dyn_ball_l, loc=<-0.3668 0.1038 2.6579>, rot=<-0.6896 0.1883 -0.1422 -0.6847>}
 dyn_toe_l_c_02 {parent=dyn_toe_l_c_01, loc=<5.0037 0 0>, rot=<0.0036 -0.1325 -0.0096 -0.9911>}
 joint9 {parent=dyn_toe_l_c_02, loc=<3.0034 0 0>, rot=<-0.5168 -0.8206 -0.13 0.2064>}
 dyn_toe_l_b_01 {parent=dyn_ball_l, loc=<0.7011 0.1507 -0.1672>, rot=<-0.7053 0.0209 0.0214 -0.7083>}
 dyn_toe_l_b_02 {parent=dyn_toe_l_b_01, loc=<7.3191 0 0>, rot=<0.0132 -0.0869 -0.0211 -0.9959>}
 joint6 {parent=dyn_toe_l_b_02, loc=<4.2827 0 0>, rot=<0 0 0 1>}
 dyn_toe_l_a_01 {parent=dyn_ball_l, loc=<-0.4612 0.1551 -3.2374>, rot=<-0.6727 -0.187 0.1986 -0.6878>}
 dyn_toe_l_a_02 {parent=dyn_toe_l_a_01, loc=<4.8931 0 0>, rot=<-0.1023 -0.0736 -0.0192 -0.9918>}
 joint3 {parent=dyn_toe_l_a_02, loc=<3.8979 0 0>, rot=<0 0 0 1>}
 dyn_toe_l_d_03 {parent=dfrm_foot_l, loc=<-2.6963 -0.0647 -0.4626>, rot=<-0.0696 -0.6943 0.7053 0.1251>}
 dyn_toe_l_d_01 {parent=dyn_toe_l_d_03, loc=<2.0436 0 0.1444>, rot=<0 0 0 1>}
 dyn_toe_l_d_02 {parent=dyn_toe_l_d_01, loc=<1.8873 0 0>, rot=<-0.0107 0.0658 -0.0051 -0.9978>}
 joint12 {parent=dyn_toe_l_d_02, loc=<3.4635 0 0>, rot=<0 0 0 1>}
 dfrm_thigh_twist_01_l {parent=dfrm_thigh_l, loc=<-13.2027 0 0>, rot=<0.0013 0 0 -1>}
 deform_groin_base_l {parent=dfrm_thigh_l, loc=<1.1526 11.1928 -1.1578>, rot=<0 0 0 -1>}
 deform_groin_l {parent=deform_groin_base_l, loc=<0 0 0>, rot=<0 0 0 1>}
 deform_glute_base_r {parent=dfrm_pelvis, loc=<-8.4059 -13.215 8.4287>, rot=<0.0501 -0.0275 0.8926 -0.4472>}
 deform_glute_r {parent=deform_glute_base_r, loc=<0 0 0>, rot=<0 0 0 1>}
 deform_glute_base_l {parent=dfrm_pelvis, loc=<-8.4058 -13.215 -8.4287>, rot=<0.0275 0.0501 -0.4472 -0.8926>}
 deform_glute_l {parent=deform_glute_base_l, loc=<0 0 0>, rot=<0 0 0 1>}
 dyn_tail_r_MSTR {parent=dfrm_pelvis, loc=<-13.226 -5.434 10.419>, rot=<0.2668 -0.9638 0 0>}
 dyn_tail_r_01 {parent=dyn_tail_r_MSTR, loc=<0 0 0>, rot=<0 0 0 1>}
 dyn_tail_r_02 {parent=dyn_tail_r_01, loc=<12.3834 0 0>, rot=<0 0 0 1>}
 dyn_tail_r_03 {parent=dyn_tail_r_02, loc=<12.3834 0 0>, rot=<0 0 0 1>}
 dyn_tail_l_MSTR {parent=dfrm_pelvis, loc=<-13.2257 -5.4342 -10.4193>, rot=<0.2668 -0.9638 0 0>}
 dyn_tail_l_01 {parent=dyn_tail_l_MSTR, loc=<0 0 0>, rot=<0 0 0 1>}
 dyn_tail_l_02 {parent=dyn_tail_l_01, loc=<12.3834 0 0>, rot=<0 0 0 1>}
 dyn_tail_l_03 {parent=dyn_tail_l_02, loc=<12.3834 0 0>, rot=<0 0 0 1>}
 dfrm_stomach_01_r {parent=pelvis, loc=<5.3295 19.3732 7.4074>, rot=<0.7074 0.6825 0.1322 -0.1275>}
 dfrm_stomach_02_r {parent=pelvis, loc=<5.6777 12.2012 15.2478>, rot=<0.584 0.5634 0.4206 -0.4058>}
 dfrm_stomach_03_r {parent=pelvis, loc=<5.8017 2.8742 16.9445>, rot=<0.5063 0.4885 0.5114 -0.4934>}
 dfrm_stomach_03_l {parent=pelvis, loc=<5.8017 2.8742 -16.9445>, rot=<0.5599 0.5402 -0.4521 0.4362>}
 dfrm_stomach_02_l {parent=pelvis, loc=<5.6777 12.2012 -15.248>, rot=<0.6319 0.6096 -0.3445 0.3323>}
 dfrm_stomach_01_l {parent=pelvis, loc=<5.3295 19.3732 -7.407>, rot=<0.7041 0.6793 -0.1487 0.1435>}
 Floor_COL {parent=root, loc=<0 0 0>, rot=<0 0 0 1>}
 Foot_r_COL {parent=root, loc=<0 0 0>, rot=<0 0 0 1>}
 Foot_l_COL {parent=root, loc=<0 0 0>, rot=<0 0 0 1>}
 ik_hand_root {parent=root, loc=<0 0 0>, rot=<0 0 0 1>}
 ik_hand_gun {parent=ik_hand_root, loc=<-48.1743 3.3545 104.4381>, rot=<0.6276 0.1309 0.4183 -0.6434>}
 ik_hand_r {parent=ik_hand_gun, loc=<0 0 0>, rot=<0 0 0 1>}
 ik_hand_l {parent=ik_hand_gun, loc=<59.3271 67.6997 34.3524>, rot=<-0.6158 -0.7027 -0.3565 0>}
 ik_foot_root {parent=root, loc=<0 0 0>, rot=<0 0 0 1>}
 ik_foot_l {parent=ik_foot_root, loc=<14.8262 -7.3545 8.296>, rot=<0.0357 0.7045 0.0284 -0.7082>}
 ik_foot_r {parent=ik_foot_root, loc=<-14.8262 -7.3545 8.296>, rot=<0.7082 0.0284 -0.7045 0.0357>}
 attach {parent=root, loc=<0 0 0>, rot=<0 0 0 1>}
 attach_fp {parent=root, loc=<0 0 0>, rot=<0 0 0 1>}
 camera_root {parent=root, loc=<0 0 0>, rot=<0 0 0 1>}
 camera {parent=camera_root, loc=<0 0 0>, rot=<0 0 0 1>}
 warp_point_01 {parent=root, loc=<0 0 0>, rot=<0 0 0 1>}
 warp_point_02 {parent=root, loc=<0 0 0>, rot=<0 0 0 1>}
 warp_point_03 {parent=root, loc=<0 0 0>, rot=<0 0 0 1>}
]
variables [
 R handClose to pelvis {type=double, default=0}
 SPR_upperarm_l_down_A_alpha {type=double, default=0}
 SPR_upperarm_r_down_A_alpha {type=double, default=0}
 SPR_upperarm_r_down_B_alpha {type=double, default=0}
 deform_bool {
  type=FUserDefinedStruct_#4
  default="(upperarm_1_#5=True,elbow_3_#6=True,wrist_7_#7=True,groin_8_#3=True,glute_17_#2=True,knee_9_40C408324CA261 ...(truncated)"
 }
 deform_value {
  type=FUserDefinedStruct_#8
  default="(upperarm_up_18_#9=(Rotation=(X=0,Y=0,Z=0,W=1),Translation=(X=0,Y=0,Z=8),Scale3D=(X=0.7,Y=1,Z=1)),upperarm_fwd_fr_19_#10=(Rotation=(X=0,Y=0,Z=0,W=1),Translation=(X=0,Y=10,Z=0),Scale3D=(X=0.5,Y=1,Z= ...(truncated)"
 }
 emote {type=double, default=0}
 foot up l {type=double, default=0}
 foot up r {type=double, default=0}
 mult_spr_dst_hand_r_thigh_twist_r_alpha {type=double, default=0}
 mult_spr_dst_lowerarm_r_pelvis_alpha {type=double, default=0}
 pistol {type=double, default=0}
 sprint {type=double, default=0}
 thigh r up {type=double, default=0}
]
