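# Physics asset for the tail mesh of the MallardVantage medium female player body, defining collision and joint limits for its tail.
# Three capsule tail bodies chain from a tiny master capsule through constraints that allow wider swing on one axis, with tight side swing and twist.
# Thigh, calf and foot shapes supply leg collision, with feet linked to calves by angularly free constraints.
# A large offset sphere on Floor_COL approximates the ground.
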PhysicsAsset /Game/Characters/Player/Female/Medium/Bodies/F_MED_MallardVantage/Meshes/F_MED_MallardVantage_Tail_Physics
bodies {
 Floor_COL [{shapeType=Sphere, radius=300, shapeName=root_sphere, center=<0 0 -298>}]
 dfrm_calf_l [{shapeType=Capsule, length=32, radius=6.75, shapeName=dfrm_calf_l_capsule, center=<-20 0 0>, rotation=<rot 90 0 0>}]
 dfrm_calf_r [{shapeType=Capsule, length=32, radius=6.75, shapeName=dfrm_calf_l_capsule, center=<20 0 0>, rotation=<rot 90 0 0>}]
 dfrm_foot_l [{shapeType=Sphere, radius=8, shapeName=dfrm_foot_l_sphere}]
 dfrm_foot_r [{shapeType=Sphere, radius=8, shapeName=dfrm_foot_r_sphere}]
 dfrm_thigh_l [{shapeType=Capsule, length=23.5, radius=6.5, shapeName=dfrm_thigh_r_capsule, center=<-18 2.5 0>, rotation=<rot 90 0 -5>}]
 dfrm_thigh_r [{shapeType=Capsule, length=23.5, radius=6.5, shapeName=dfrm_thigh_r_capsule, center=<18 -2.5 0>, rotation=<rot 90 0 -5>}]
 dyn_tail_01 [{shapeType=Capsule, length=25.25, radius=9, shapeName=dyn_tail_01_capsule, center=<2 -3 0>}]
 dyn_tail_02 [{shapeType=Capsule, length=22, radius=9, shapeName=dyn_tail_02_capsule, center=<2 -3 0>}]
 dyn_tail_03 [{shapeType=Capsule, length=10, radius=9, shapeName=dyn_tail_03_capsule, center=<5 -5 0>}]
 dyn_tail_MSTR [{shapeType=Capsule, length=0.505, radius=0.505, shapeName=dyn_tail_MSTR_capsule, center=<2 0 0>}]
}
constraints [
 {bone1Name=dyn_tail_01, bone2Name=dyn_tail_MSTR, swing1LimitDegrees=35, swing2LimitDegrees=5, twistLimitDegrees=5}
 {bone1Name=dyn_tail_02, bone2Name=dyn_tail_01, swing1LimitDegrees=35, swing2LimitDegrees=5, twistLimitDegrees=5}
 {bone1Name=dyn_tail_03, bone2Name=dyn_tail_02, swing1LimitDegrees=35, swing2LimitDegrees=5, twistLimitDegrees=5}
 {
  bone1Name=dfrm_foot_l
  bone2Name=dfrm_calf_l
  swing1LimitDegrees=45
  swing1Motion=Free
  swing2LimitDegrees=45
  swing2Motion=Free
  twistLimitDegrees=45
  twistMotion=Free
 }
 {
  bone1Name=dfrm_foot_r
  bone2Name=dfrm_calf_r
  swing1LimitDegrees=45
  swing1Motion=Free
  swing2LimitDegrees=45
  swing2Motion=Free
  twistLimitDegrees=45
  twistMotion=Free
 }
]
